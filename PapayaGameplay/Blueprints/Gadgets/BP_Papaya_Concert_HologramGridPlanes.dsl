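; Concert gadget Blueprint for hologram grid planes in the Papaya environment.
; The construction script creates two dynamic material instances from M_HologramCeiling, one each for the SR and SL hologram plane components.
; Each instance is stored in its own variable, SRHolo_MID and SLHolo_MID.
; The BeginPlay, overlap and Tick events are empty, and nothing here changes the material parameters.
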(event EventBeginPlay)

(event Collision|EventActorBeginOverlap
  (params (OtherActor :type Actor%20Object%20Reference)))

(event EventTick
  (params (DeltaSeconds :type Float%20%28single-precision%29)))

(fn ConstructionScript
  (bind _returnvalue (Rendering|Material|CreateDynamicMaterialInstance (Variables|Default|GetSR_HologramPlane) 0 "/Game/Environments/Papaya/Materials/M_HologramCeiling.M_HologramCeiling" "SR_HoloPlane" :!class /Script/Engine.PrimitiveComponent))
  (Variables|Default|SetSRHolo_MID _returnvalue)
  (bind _returnvalue_2 (Rendering|Material|CreateDynamicMaterialInstance (Variables|Default|GetSL_HologramPlane) 0 "/Game/Environments/Papaya/Materials/M_HologramCeiling.M_HologramCeiling" "SL_HoloPlane" :!class /Script/Engine.PrimitiveComponent))
  (Variables|Default|SetSLHolo_MID _returnvalue_2))
