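; BP_Papaya_DiscoBalls is a gadget Blueprint that switches a disco ball effect on and off through Start and Stop custom events.
; Start deactivates the Niagara component, waits briefly, then reactivates it with a reset and reveals the decal.
; Stop deactivates the Niagara component and hides the decal.
; Nothing in the graph calls these events, and the construction script is empty.
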(event Custom|Start
  (bind _niagara (Variables|Default|GetNiagara))
  (Components|Activation|Deactivate _niagara)
  (Utilities|FlowControl|Delay 0.15)
  (Components|Activation|Activate _niagara true)
  (Development|SetHiddeninGame (Variables|Default|GetDecal) false false))

(event Custom|Stop
  (Components|Activation|Deactivate (Variables|Default|GetNiagara))
  (Development|SetHiddeninGame (Variables|Default|GetDecal) true false))

(fn ConstructionScript)
