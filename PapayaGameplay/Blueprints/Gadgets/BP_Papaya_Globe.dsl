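; Globe gadget Blueprint that starts hidden and animates a masked reveal or hide when toggled on or off.
; Construction creates normal and masked dynamic materials for the globe mesh.
; Toggles swap to the masked material and play or reverse GlobeRevealTimeline, which writes PapayaGlobeMaskParameter to the shared FortniteMaterialParameters collection.
; On finish it restores the normal material, hiding the hierarchy if reversed; ChangeGlobeVisibility sets visibility directly.
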(event EventBeginPlay)

(event Collision|EventActorBeginOverlap
  (params (OtherActor :type Actor%20Object%20Reference)))

(event EventTick
  (params (DeltaSeconds :type Float%20%28single-precision%29)))

(event Custom|ChangeGlobeVisibility
  (params (Visible :type Boolean))
  (Rendering|SetVisibility (Variables|Default|GetDefaultSceneRoot) Visible true))

(event Custom|ToggleGlobeOn
  (Rendering|Material|SetMaterial (Variables|Default|Getsm_papaya_globe) 0 (Variables|Default|GetMID_Masked))
  (Rendering|SetVisibility (Variables|Default|GetDefaultSceneRoot) true true)
  (bind (_direction _maskvalue) (Timeline|GlobeRevealTimeline Play)
    (:Update
      (Rendering|Material|SetScalarParameterValue "/Game/Packages/Fortress_SharedMaterials/GlobalMaterialParameters/FortniteMaterialParameters.FortniteMaterialParameters" "PapayaGlobeMaskParameter" _maskvalue :!class /Script/Engine.KismetMaterialLibrary))
    (:Finished
      (Rendering|Material|SetMaterial (Variables|Default|Getsm_papaya_globe) 0 (Variables|Default|GetMID))
      (if (Utilities|Enum|Equal(Enum) _direction "Backward")
        (Rendering|SetVisibility (Variables|Default|GetDefaultSceneRoot) false true)))))

(event Custom|ToggleGlobeOff
  (Rendering|Material|SetMaterial (Variables|Default|Getsm_papaya_globe) 0 (Variables|Default|GetMID_Masked))
  (Timeline|GlobeRevealTimeline Reverse))

(fn ConstructionScript
  (bind _sm_papaya_globe (Variables|Default|Getsm_papaya_globe))
  (Utilities|FlowControl|Sequence
    (:then_0
      (bind _returnvalue (Rendering|Material|CreateDynamicMaterialInstance _sm_papaya_globe 0 "/PapayaGameplay/Blueprints/Gadgets/FX/Materials/MI_Papaya_Globe.MI_Papaya_Globe" "None" :!class /Script/Engine.PrimitiveComponent))
      (Variables|Default|SetMID _returnvalue)
      (bind _returnvalue_2 (Rendering|Material|CreateDynamicMaterialInstance _sm_papaya_globe 0 "/PapayaGameplay/Blueprints/Gadgets/FX/Materials/MI_Papaya_Globe_Masked.MI_Papaya_Globe_Masked" "None" :!class /Script/Engine.PrimitiveComponent))
      (Variables|Default|SetMID_Masked _returnvalue_2))
    (:then_1
      (Rendering|SetVisibility (Variables|Default|GetDefaultSceneRoot) false true))))
