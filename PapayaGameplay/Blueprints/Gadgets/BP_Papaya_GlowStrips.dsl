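; Glow strip gadget Blueprint that prepares swirly spline-mesh materials for three entry glow shapes: SR, SL and Center.
; On BeginPlay it creates a named dynamic material instance on element 0 of each shape component and stores each in its own variable.
; SR and SL share the 02 variant of the swirl material, while Center uses the base variant.
; No material parameters are changed here, and the construction script is empty.
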(event EventBeginPlay
  (bind _returnvalue (Rendering|Material|CreateDynamicMaterialInstance (Variables|Default|GetSR_Entry_GlowShape) 0 "/PapayaGameplay/Blueprints/Gadgets/FX/Materials/MI_Papaya_Swirly_SplineMesh_02.MI_Papaya_Swirly_SplineMesh_02" "SREntryGlowShape_MID" :!class /Script/Engine.PrimitiveComponent))
  (Variables|Default|SetSREntry_MID _returnvalue)
  (bind _returnvalue_2 (Rendering|Material|CreateDynamicMaterialInstance (Variables|Default|GetSL_Entry_GlowShape) 0 "/PapayaGameplay/Blueprints/Gadgets/FX/Materials/MI_Papaya_Swirly_SplineMesh_02.MI_Papaya_Swirly_SplineMesh_02" "SLEntryGlowShape_MID" :!class /Script/Engine.PrimitiveComponent))
  (Variables|Default|SetSLEntry_MID _returnvalue_2)
  (bind _returnvalue_3 (Rendering|Material|CreateDynamicMaterialInstance (Variables|Default|GetCenter_Entry_GlowShape) 0 "/PapayaGameplay/Blueprints/Gadgets/FX/Materials/MI_Papaya_Swirly_SplineMesh.MI_Papaya_Swirly_SplineMesh" "CenterEntryGlowShape_MID" :!class /Script/Engine.PrimitiveComponent))
  (Variables|Default|SetCenterEntry_MID _returnvalue_3))

(fn ConstructionScript)
