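; Glow strip screen gadget Blueprint for Papaya gameplay that prepares dynamic materials for four screen curve components.
; On BeginPlay, it creates a separately named dynamic instance of the swirly spline mesh top material on each curve's first slot.
; Each instance is stored in its own variable, but no parameters are changed here.
; The overlap, tick and construction script entry points are empty.
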(event EventBeginPlay
  (bind _returnvalue (Rendering|Material|CreateDynamicMaterialInstance (Variables|Default|GetScreenCurve_1) 0 "/PapayaGameplay/Blueprints/Gadgets/FX/Materials/MI_Papaya_Swirly_Spline_Mesh_Top.MI_Papaya_Swirly_Spline_Mesh_Top" "ScreenCurve1MaterialInstDyanmic" :!class /Script/Engine.PrimitiveComponent))
  (Variables|Default|SetScreenCurve1_MID _returnvalue)
  (bind _returnvalue_2 (Rendering|Material|CreateDynamicMaterialInstance (Variables|Default|GetScreenCurve_2) 0 "/PapayaGameplay/Blueprints/Gadgets/FX/Materials/MI_Papaya_Swirly_Spline_Mesh_Top.MI_Papaya_Swirly_Spline_Mesh_Top" "ScreenCurve2MaterialInstDyanmic" :!class /Script/Engine.PrimitiveComponent))
  (Variables|Default|SetScreenCurve2_MID _returnvalue_2)
  (bind _returnvalue_3 (Rendering|Material|CreateDynamicMaterialInstance (Variables|Default|GetScreenCurve_3) 0 "/PapayaGameplay/Blueprints/Gadgets/FX/Materials/MI_Papaya_Swirly_Spline_Mesh_Top.MI_Papaya_Swirly_Spline_Mesh_Top" "ScreenCurve3MaterialInstDyanmic" :!class /Script/Engine.PrimitiveComponent))
  (Variables|Default|SetScreenCurve3_MID _returnvalue_3)
  (bind _returnvalue_4 (Rendering|Material|CreateDynamicMaterialInstance (Variables|Default|GetScreenCurve_4) 0 "/PapayaGameplay/Blueprints/Gadgets/FX/Materials/MI_Papaya_Swirly_Spline_Mesh_Top.MI_Papaya_Swirly_Spline_Mesh_Top" "ScreenCurve4MaterialInstDyanmic" :!class /Script/Engine.PrimitiveComponent))
  (Variables|Default|SetScreenCurve4_MID _returnvalue_4))

(event Collision|EventActorBeginOverlap
  (params (OtherActor :type Actor%20Object%20Reference)))

(event EventTick
  (params (DeltaSeconds :type Float%20%28single-precision%29)))

(fn ConstructionScript)
