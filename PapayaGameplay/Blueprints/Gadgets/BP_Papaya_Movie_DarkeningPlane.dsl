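; Blueprint actor for a movie screen-darkening plane, built on a static mesh.
; Both the construction script and BeginPlay create a dynamic instance of the M_ScreenDarkening material on the mesh's first slot and store it in a DarkeningPlaneMID variable.
; No darkening parameters are set here; the stored instance is only made available for use.
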(event EventBeginPlay
  (bind _returnvalue (Rendering|Material|CreateDynamicMaterialInstance (Variables|Default|GetStaticMesh) 0 "/Game/Environments/Papaya/FX/M_ScreenDarkening.M_ScreenDarkening" "DarkeningMID" :!class /Script/Engine.PrimitiveComponent))
  (Variables|Default|SetDarkeningPlaneMID _returnvalue))

(fn ConstructionScript
  (bind _returnvalue (Rendering|Material|CreateDynamicMaterialInstance (Variables|Default|GetStaticMesh) 0 "/Game/Environments/Papaya/FX/M_ScreenDarkening.M_ScreenDarkening" "DarkeningMID" :!class /Script/Engine.PrimitiveComponent))
  (Variables|Default|SetDarkeningPlaneMID _returnvalue))
