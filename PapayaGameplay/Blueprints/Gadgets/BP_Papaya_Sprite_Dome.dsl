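; Papaya gadget Blueprint that renders a configurable dome of sprites through a Niagara component.
; UpdateSystem, run from the construction script, pushes spawn count, ray percentage, radius and flip speed to Niagara, scaling bounds with radius.
; It also writes alpha mode and solid-sprite values to the shared MPC_PapayaGadgetControls parameter collection, so those are not per-instance.
; ShowDome activates or deactivates the effect; OnRep_AlphaMode is empty.
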(event Custom|UpdateSystem
  (bind _radius (Variables|Default|GetRadius))
  (Utilities|FlowControl|Sequence
    (:then_0
      (Niagara|SetNiagaraVariable(Int32) (Variables|Default|GetSpriteDome) "SpawnCount" (Variables|Default|GetSpawn%20Count))
      (Niagara|SetNiagaraVariable(Float) (Variables|Default|GetSpriteDome) "SpawnCountRaysPct" (Variables|Default|GetSpawn%20Count%20Rays%20Pct))
      (Niagara|SetNiagaraVariable(Float) (Variables|Default|GetSpriteDome) "Radius" _radius)
      (Rendering|SetBoundsScale (Variables|Default|GetSpriteDome) (/ _radius 1e+02)))
    (:then_1
      (Niagara|SetNiagaraVariable(Float) (Variables|Default|GetSpriteDome) "FlipSpeed" (Variables|Default|GetFlip%20Speed)))
    (:then_2
      (Rendering|Material|SetScalarParameterValue "/Game/Environments/Papaya/Materials/MPC_PapayaGadgetControls.MPC_PapayaGadgetControls" "SpriteGlobeAlphaMode" (Variables|Default|GetAlpha%20Mode) :!class /Script/Engine.KismetMaterialLibrary)
      (Rendering|Material|SetScalarParameterValue "/Game/Environments/Papaya/Materials/MPC_PapayaGadgetControls.MPC_PapayaGadgetControls" "SpriteGlobeSolid" (Variables|Default|GetSolid%20Sprites) :!class /Script/Engine.KismetMaterialLibrary))))

(event Custom|ShowDome
  (params (bShow :type Boolean))
  (bind _spritedome (Variables|Default|GetSpriteDome))
  (if bShow
    (Components|Activation|Activate _spritedome true)
    (else
      (Components|Activation|Deactivate _spritedome))))

(fn OnRep_AlphaMode)

(fn ConstructionScript
  (CallFunction|UpdateSystem))
