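; BP_Papaya_TetraScreen is a Papaya gadget Blueprint for a tetrahedral screen mesh that can be hidden and spun.
; The Hidden event sets the mesh's hidden-in-game state and casts its shadow only while shown.
; RotateSpeed applies the Rotation Speed variable as the yaw rate of a RotatingMovement component, which StartRotate and EndRotate activate and deactivate.
; Nothing in the graph calls these custom events, and the construction script is empty.
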(event Custom|Hidden
  (params (Hidden :type Boolean))
  (bind _sm_papaya_tetrahedral_screen (Variables|Default|Getsm_papaya_tetrahedral_screen))
  (Development|SetHiddeninGame _sm_papaya_tetrahedral_screen Hidden false)
  (Rendering|SetCastShadow _sm_papaya_tetrahedral_screen (not Hidden)))

(event Custom|RotateSpeed
  (Variables|/Script/Engine.RotatingMovementComponent|SetRotationRate (Math|Rotator|MakeRotator 0.0 0.0 (Variables|Default|GetRotation%20Speed)) (Variables|Default|GetRotatingMovement)))

(event Custom|StartRotate
  (Components|Activation|Activate (Variables|Default|GetRotatingMovement) false))

(event Custom|EndRotate
  (Components|Activation|Deactivate (Variables|Default|GetRotatingMovement)))

(fn ConstructionScript)
