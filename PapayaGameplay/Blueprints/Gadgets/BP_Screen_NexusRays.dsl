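; Screen gadget Blueprint that displays Nexus ray effects on a center screen and toggleable shard side screens.
; Construction script and BeginPlay both create and store dynamic material instances for the nexus core-crack mesh and the side-screen rays mesh.
; Construction also shows the Center component and hides Sides, propagating to children.
; Custom events activate or deactivate the MainScreenTransition component and show or hide the side screens; no material parameters are set here.
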(event EventBeginPlay
  (bind _returnvalue (Rendering|Material|CreateDynamicMaterialInstance (Variables|Default|Getsm_papaya_screen_nexus_rays) 0 "/PapayaGameplay/Blueprints/Gadgets/FX/Materials/M_Papaya_Core_Crack.M_Papaya_Core_Crack" "NexusRaysMaterialInstDynamic" :!class /Script/Engine.PrimitiveComponent))
  (Variables|Default|SetNexusRays_MID _returnvalue)
  (bind _returnvalue_2 (Rendering|Material|CreateDynamicMaterialInstance (Variables|Default|Getsm_papaya_sidescreen_rays) 0 "/PapayaGameplay/Blueprints/Gadgets/FX/Materials/M_Papaya_SideScreenRays.M_Papaya_SideScreenRays" "NexusRaysSideMaterialInstDynamic" :!class /Script/Engine.PrimitiveComponent))
  (Variables|Default|SetNexusRaysSide_MID _returnvalue_2))

(event Custom|ActivateTransition
  (Components|Activation|Activate (Variables|Default|GetMainScreenTransition) true))

(event Custom|DeactivateTransition
  (Components|Activation|Deactivate (Variables|Default|GetMainScreenTransition)))

(event Custom|ShowShardSideScreens
  (Development|SetHiddeninGame (Variables|Default|GetSides) false true))

(event Custom|HideShardSideScreens
  (Development|SetHiddeninGame (Variables|Default|GetSides) true true))

(fn ConstructionScript
  (Utilities|FlowControl|Sequence
    (:then_0
      (bind _returnvalue (Rendering|Material|CreateDynamicMaterialInstance (Variables|Default|Getsm_papaya_screen_nexus_rays) 0 "/PapayaGameplay/Blueprints/Gadgets/FX/Materials/M_Papaya_Core_Crack.M_Papaya_Core_Crack" "NexusRaysMaterialInstDynamic" :!class /Script/Engine.PrimitiveComponent))
      (Variables|Default|SetNexusRays_MID _returnvalue))
    (:then_1
      (bind _returnvalue_2 (Rendering|Material|CreateDynamicMaterialInstance (Variables|Default|Getsm_papaya_sidescreen_rays) 0 "/PapayaGameplay/Blueprints/Gadgets/FX/Materials/M_Papaya_SideScreenRays.M_Papaya_SideScreenRays" "NexusRaysSideMaterialInstDynamic" :!class /Script/Engine.PrimitiveComponent))
      (Variables|Default|SetNexusRaysSide_MID _returnvalue_2))
    (:then_2
      (Development|SetHiddeninGame (Variables|Default|GetCenter) false true)
      (Development|SetHiddeninGame (Variables|Default|GetSides) true true))))
