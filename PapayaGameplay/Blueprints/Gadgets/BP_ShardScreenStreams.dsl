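; BP_ShardScreenStreams is a gadget Blueprint whose name points to a shard screen stream effect with a haze mesh.
; Its construction script creates a dynamic material instance from MI_Papaya_NoTick_Haze on slot 0 of the SM_NoTickHaze component.
; It stores that instance in the Haze_MID variable, and nothing shown here changes its parameters.
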(fn ConstructionScript
  (bind _returnvalue (Rendering|Material|CreateDynamicMaterialInstance (Variables|Default|GetSM_NoTickHaze) 0 "/PapayaGameplay/Blueprints/Gadgets/FX/Materials/MI_Papaya_NoTick_Haze.MI_Papaya_NoTick_Haze" "NoTickHaze_MID" :!class /Script/Engine.PrimitiveComponent))
  (Variables|Default|SetHaze_MID _returnvalue))
