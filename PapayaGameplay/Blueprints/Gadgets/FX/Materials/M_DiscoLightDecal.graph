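# Translucent deferred decal material that projects animated disco ball light around its actor, reacting to event audio.
# The actor-to-pixel direction spins around Z over time, then drives axis-weighted grid texture samples, each using one repeated component as UV.
# The pattern fades with distance, gets hue-shifted and partly desaturated.
# EventAudio_Amplitude from the Fortnite parameter collection blends in the MF_Papaya_GadgetColor tint and scales brightness.
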Material /PapayaGameplay/Blueprints/Gadgets/FX/Materials/M_DiscoLightDecal
graph {
 properties {
  bCastRayTracedShadows=false
  blendMode=BLEND_Translucent
  materialDomain=MD_DeferredDecal
  pixelDepthOffsetMode=PDOM_Legacy
  usedShadingModels=Default Lit
 }
 nodes [
  n1 = Multiply(A=Time(), B=ScalarParameter() {defaultValue=0.02, parameterName=Rotation Speed}) {constB=0.05}
  n2 = Cosine(None=@n1)
  n3 = Subtract(A=WorldPosition().XYZ, B=ActorPositionWS())
  n4 = MaterialFunctionCall(Float3=Normalize(VectorInput=@n3)) {materialFunction=&BreakOutFloat3Components}
  n5 = Sine(None=@n1)
  n6 = MaterialFunctionCall(X=Subtract(A=Multiply(A=@n2, B=@n4.R), B=Multiply(A=@n5, B=@n4.G)), Y=Add(A=Multiply(A=@n5, B=@n4.R), B=Multiply(A=@n2, B=@n4.G)), Z=@n4.B) {materialFunction=&MakeFloat3}
  n7 = MaterialFunctionCall(Float3=@n6.Result) {materialFunction=&BreakOutFloat3Components}
  n8 = MaterialFunctionCall(X=@n7.R, Y=@n7.R) {materialFunction=&MakeFloat2}
  n9 = TextureObject() {texture=/PapayaGameplay/Blueprints/Gadgets/FX/Textures/disco_ball_grid}
  n10 = Constant()
  n11 = DotProduct(A=@n6.Result, B=Constant3Vector() {constant=<rgba 0 1 0 0>})
  LinearInterpolate(
   B=ShadingPathSwitch(
    Default=TextureSample(UVs=@n8.Result, Tex=@n9).RGB
    Mobile=TextureSample!(UVs=@n8.Result, Tex=@n9, Level=@n10, Apply View MipBias) {constMipValue=0, mipValueMode=TMVM_MipLevel}.RGB
   )
   Alpha=Power(Base=Abs(None=@n11)) {constExponent=2.5}
  ) {constB=0}
  n12 = Time()
  n13 = MaterialFunctionCall(X=@n7.G, Y=@n7.G) {materialFunction=&MakeFloat2}
  n14 = DotProduct(A=@n6.Result, B=Constant3Vector() {constant=<rgba 1 0 0 0>})
  n15 = LinearInterpolate(
   B=ShadingPathSwitch(
    Default=TextureSample!(UVs=@n13.Result, Tex=@n9, Apply View MipBias) {constMipValue=0}.RGB
    Mobile=TextureSample!(UVs=@n13.Result, Tex=@n9, Level=@n10, Apply View MipBias) {constMipValue=0, mipValueMode=TMVM_MipLevel}.RGB
   )
   Alpha=Power(Base=Abs(None=@n14)) {constExponent=2.5}
  ) {constB=0}
  n16 = MaterialFunctionCall(X=@n7.R, Y=@n7.R) {materialFunction=&MakeFloat2}
  n17 = TextureSample!(UVs=@n16.Result, Tex=@n9, Apply View MipBias)
  n18 = TextureSample!(UVs=@n16.Result, Tex=@n9, Level=@n10, Apply View MipBias) {constMipValue=0, mipValueMode=TMVM_MipLevel}
  n19 = Abs(None=DotProduct(A=@n6.Result, B=Constant3Vector() {constant=<rgba 0 0 1 0>}))
  n20 = Add(
   A=Add(A=@n15, B=@n15)
   B=LinearInterpolate(B=ShadingPathSwitch(Default=@n17.RGB, Mobile=@n18.RGB), Alpha=Power(Base=@n19) {constExponent=2.5}) {constB=0}
  )
  n21 = MaterialFunctionCall(Vector 3=@n3) {materialFunction=/Engine/Functions/Engine_MaterialFunctions02/Utility/VectorLength}
  n22 = ScalarParameter() {defaultValue=3000, parameterName=Falloff Distance}
  n23 = OneMinus(None=Saturate(None=Divide(A=@n21.V3 Length, B=@n22)))
  n24 = Desaturation(
   None=MaterialFunctionCall(Hue Shift Percentage=Multiply(A=@n12), Texture=Multiply(A=@n20, B=@n23)) {materialFunction=&HueShift}.Result
   Fraction=Constant() {r=0.75}
  )
  n25 = MaterialFunctionCall!(
   Media Player TexCoord
   Media Color Code Coordinate
   Media Texture Resolution
   Base Color Parameter
   Base Color Mix
   Media Color Code Mix
   Intermission Color Parameter
   Intermission Color Mix
   Desaturation
  ) {materialFunction=/Game/Environments/Papaya/Materials/MaterialFunctions/MF_Papaya_GadgetColor}
  n26 = CollectionParameter() {collection=&FortniteMaterialParameters, parameterName=EventAudio_Amplitude}
  Multiply(A=LinearInterpolate(A=@n24, B=Multiply(A=@n24, B=@n25.Result), Alpha=@n26), B=Add(A=Multiply(A=@n26) {constB=1.9}) {constB=0.1})
 ]
}
