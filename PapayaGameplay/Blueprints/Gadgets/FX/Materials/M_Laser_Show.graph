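# M_Laser_Show is an additive, unlit, two-sided material for Niagara mesh-particle laser show effects in the Papaya gadgets.
# Color blends the shared gadget color function with particle color via a dynamic parameter, then pulses with the EventAudio_Spectrum collection's alpha.
# An IsBeam switch selects thin, edge-faded lines thickened by SplineThicken, or a softer vertical gradient without offset.
# Both get HDR boosting, particle-alpha and depth fades, then shared emissive adjustments.
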Material /PapayaGameplay/Blueprints/Gadgets/FX/Materials/M_Laser_Show
graph {
 properties {
  bUsedWithNiagaraMeshParticles=true
  blendMode=BLEND_Additive
  pixelDepthOffsetMode=PDOM_Legacy
  shadingModel=MSM_Unlit
  twoSided=true
  usedShadingModels=Unlit
 }
 nodes [
  n1 = Constant3Vector()
  n2 = MaterialFunctionCall(Float3=Multiply(A=TransformPosition(None=@n1) {transformType=TRANSFORMPOSSOURCE_World}) {constB=0.0001}) {materialFunction=&BreakOutFloat3Components}
  MaterialFunctionCall(
   Hue Shift Percentage=TextureSample(UVs=MaterialFunctionCall(R=@n2.R, G=@n2.R) {materialFunction=&AppendMany}.RG) {texture=/Game/Global/Textures/Default/Noise/3xPerlin_M}.RGB
   Texture=Constant3Vector() {constant=<rgba 1 0 0 0>}
  ) {materialFunction=&HueShift}
  n3 = TextureCoordinate()
  n4 = ComponentMask(None=@n3) {r=true}
  n5 = Multiply(A=@n4, B=OneMinus(None=@n4))
  n6 = ScalarParameter() {defaultValue=1.25, parameterName=Edges Fade Power}
  n7 = ComponentMask(None=@n3) {g=true}
  n8 = ScalarParameter() {defaultValue=3, parameterName=Sharp Line End Fade}
  n9 = ParticleColor()
  n10 = CollectionParameter() {collection=&FortniteMaterialParameters, parameterName=EventAudio_Spectrum}
  n11 = Multiply(
   A=LinearInterpolate(
    A=MaterialFunctionCall!(
     Media Player TexCoord
     Media Color Code Coordinate
     Media Texture Resolution
     Base Color Parameter
     Base Color Mix
     Media Color Code Mix
     Intermission Color Parameter
     Intermission Color Mix
     Desaturation
    ) {materialFunction=/Game/Environments/Papaya/Materials/MaterialFunctions/MF_Papaya_GadgetColor}.Result
    B=@n9.RGB
    Alpha=DynamicParameter() {defaultValue=<rgba 0 1 1 1>, paramNames [UseCustomColor, Param2, Param3, Param4]}.UseCustomColor
   )
   B=Multiply(A=ComponentMask(None=@n10) {a=true}) {constB=2}
  )
  n12 = Multiply(A=Multiply(A=Power(Base=@n5, Exp=@n6), B=Power(Base=@n7, Exp=@n8)), B=@n11)
  n13 = ComponentMask(None=TextureCoordinate()) {g=true}
  n14 = Power(Base=@n13) {constExponent=4}
  n15 = Multiply(A=Power(Base=@n13)) {constB=0.05}
  n16 = Multiply(A=@n11, B=Multiply(A=Add(A=@n14, B=@n15)) {constB=0.25})
  n17 = StaticSwitchParameter(
   True=MaterialFunctionCall(
    LDR Input=Saturate(None=@n12)
    HDR Tint=@n12
    HDR Intensity Multiplier=ScalarParameter() {defaultValue=10, parameterName=HDR Thin_Lines_Multiplier}
   ) {materialFunction=&DeriveHDRfromLDR}.RemappedRange
   False=MaterialFunctionCall(LDR Input=Saturate(None=@n16), HDR Tint=@n16, HDR Intensity Multiplier=ScalarParameter() {defaultValue=1.4, parameterName=HDR Multiplier}) {materialFunction=&DeriveHDRfromLDR}.RemappedRange
  ) {parameterName=IsBeam}
  n18 = ScalarParameter() {defaultValue=150, parameterName=Depth Fade}
  MaterialFunctionCall(Emissive=Multiply(A=Multiply(A=@n17, B=@n9.A), B=DepthFade(FadeDistance=@n18))) {materialFunction=&MF_FX_EmissiveAdjustments}
  StaticSwitchParameter(
   True=MaterialFunctionCall(
    WidthBase=ScalarParameter() {defaultValue=16, parameterName=LaserThicknessBase}
    WidthTip=ScalarParameter() {defaultValue=16, parameterName=LaserThicknessTip}
   ) {materialFunction=&SplineThicken}.WorldPositionOffset
   False=Constant()
  ) {parameterName=IsBeam}
 ]
}
