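# Additive, unlit, two-sided material for Papaya gadget core crack effects, shaped as light rays.
# Vertex blue blends between layered-sine flicker vibration and a UV1-driven push, both along the direction away from the object pivot.
# Color runs Papaya parameter-collection values, offset along UV1, through MF_Papaya_GadgetColor.
# Intensity multiplies panning noise, UV1 falloffs, non-mobile depth fade, optional time-of-day dimming, camera-distance fade and Alpha.
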Material /PapayaGameplay/Blueprints/Gadgets/FX/Materials/M_Papaya_Core_Crack
graph {
 properties {blendMode=BLEND_Additive, pixelDepthOffsetMode=PDOM_Legacy, shadingModel=MSM_Unlit, twoSided=true, usedShadingModels=Unlit}
 nodes [
  n1 = Normalize(VectorInput=Subtract(
    A=WorldPosition().XYZ
    B=MaterialFunctionCall() {materialFunction=&ObjectPivotPoint}.Object Pivot Location
   ))
  n2 = ScalarParameter() {defaultValue=0.8, parameterName=FlickerSpeed}
  n3 = Multiply(A=Time(), B=LinearInterpolate(A=@n2, B=Multiply(A=@n2) {constB=2}, Alpha=Constant() {r=0.8}))
  n4 = Add(A=Multiply(B=Sine(None=@n3) {period=0.55}), B=Multiply(A=Sine(None=@n3) {period=0.11}) {constB=0.2})
  n5 = Multiply(A=Sine(None=@n3) {period=0.175}) {constB=0.3}
  n6 = Sine(None=@n3) {period=3.25}
  n7 = Sine(None=@n3) {period=2.325}
  n8 = Multiply(A=Multiply(A=Add(A=@n4, B=@n5), B=Multiply(A=@n6, B=@n7)), B=Constant() {r=0.8})
  n9 = ScalarParameter() {parameterName=WPOVibrate}
  n10 = ScalarParameter() {parameterName=PushDistanceMin}
  n11 = ScalarParameter() {defaultValue=2500, parameterName=PushDistance}
  n12 = Clamp(None=ComponentMask(None=TextureCoordinate() {coordinateIndex=1}) {g=true})
  n13 = VertexColor()
  LinearInterpolate(A=Multiply(A=@n1, B=Multiply(A=@n8, B=@n9)), B=Multiply(A=LinearInterpolate(A=@n10, B=@n11, Alpha=@n12), B=@n1), Alpha=@n13.B)
  n14 = CollectionParameter() {
   collection=&FortniteMaterialParameters
   parameterName=PapayaBaseColorParameter
  }
  n15 = CollectionParameter() {
   collection=&FortniteMaterialParameters
   parameterName=PapayaBaseColorParameterOffset
  }
  n16 = TextureCoordinate() {coordinateIndex=1}
  n17 = OneMinus(None=ComponentMask(None=@n16) {g=true})
  n18 = OneMinus(None=@n17)
  MaterialFunctionCall!(
   Media Player TexCoord
   Media Color Code Coordinate
   Media Texture Resolution
   Base Color Parameter=Add(A=@n14, B=Multiply(A=@n15, B=@n18))
   Base Color Mix
   Media Color Code Mix
   Intermission Color Parameter
   Intermission Color Mix
   Desaturation
  ) {materialFunction=/Game/Environments/Papaya/Materials/MaterialFunctions/MF_Papaya_GadgetColor}
  n19 = Panner(
   Coordinate=Multiply(A=TextureCoordinate(), B=ScalarParameter() {defaultValue=2, parameterName=TilingNoise})
   Speed=AppendVector(A=ScalarParameter() {defaultValue=0.1, parameterName=Pan1}, B=Constant())
  )
  n20 = ScalarParameter() {defaultValue=3, parameterName=OpacityRaysFalloff}
  n21 = Multiply(A=ScalarParameter() {defaultValue=7, parameterName=RayBottomPowerMult}, B=@n20) {constB=7}
  n22 = Add(
   A=Multiply(A=Power(Base=@n17, Exp=@n21), B=ScalarParameter() {defaultValue=5, parameterName=RayBottomMult}) {constB=5}
   B=Multiply(A=Power(Base=@n17, Exp=@n20)) {constB=0.5}
  )
  n23 = Multiply(A=Saturate(None=ComponentMask(None=@n16) {r=true}), B=ScalarParameter() {defaultValue=0.3, parameterName=OpacityRays})
  n24 = DepthFade(FadeDistance=LinearInterpolate(A=ScalarParameter() {defaultValue=500, parameterName=FadeDistance}, Alpha=@n17) {constB=0})
  n25 = Constant() {r=1}
  n26 = Multiply(
   A=Multiply(
    A=TextureSample(UVs=@n19) {samplerType=SAMPLERTYPE_Grayscale, texture=/Game/Global/Textures/Default/Noise/TilingNoise06}.G
    B=Multiply(A=@n22, B=@n23)
   )
   B=LinearInterpolate(B=ShadingPathSwitch(Default=@n24, Mobile=@n25) {desc=Migrated from FeatureLevelSwitch}, Alpha=@n13.B) {constA=1}
  )
  n27 = Multiply(A=@n26) {constB=0.26}
  n28 = MaterialFunctionCall(Color_Morning=@n27, Color_Day=@n26, Color_Evening=@n27, Color_Night=@n27) {materialFunction=&Color_Time_of_Day}
  n29 = MaterialFunctionCall(
   Fade Length=ScalarParameter() {defaultValue=4096, parameterName=Fade Length}
   Fade Offset=ScalarParameter() {defaultValue=7500, parameterName=Fade Offset}
  ) {materialFunction=&CameraDepthFade}
  Multiply(
   A=Multiply(A=StaticSwitchParameter(True=@n28.Result, False=@n26) {parameterName=Use color time of day}, B=OneMinus(None=@n29.Result))
   B=ScalarParameter() {defaultValue=1, parameterName=Alpha}
  )
 ]
}
