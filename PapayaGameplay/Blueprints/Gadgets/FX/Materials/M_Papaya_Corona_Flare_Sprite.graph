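# Additive unlit particle sprite material for a flickering corona flare on Papaya gadgets, tinted by particle color and MF_Papaya_GadgetColor.
# Layered sines of time, offset by vertex red, drive flicker in opacity and in corona size through sphere-mask hardness.
# That mask indexes a halo gradient, combined with edge glow, a flare texture and a center hotspot omitted on the Forward path.
# Depth fade (skipped on mobile) and near-camera fade soften opacity.
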Material /PapayaGameplay/Blueprints/Gadgets/FX/Materials/M_Papaya_Corona_Flare_Sprite
graph {
 properties {
  bUsedWithNiagaraSprites=true
  bUsedWithParticleSprites=true
  blendMode=BLEND_Additive
  pixelDepthOffsetMode=PDOM_Legacy
  shadingModel=MSM_Unlit
  usedShadingModels=Unlit
 }
 nodes [
  n1 = Constant() {r=0.5}
  n2 = SphereMask(A=TextureCoordinate(), B=@n1, Radius=@n1) {hardnessPercent=0}
  n3 = Power(Base=@n2) {constExponent=25}
  n4 = ScalarParameter() {defaultValue=0.25, parameterName=CenterBrightness}
  n5 = Constant() {r=0.5}
  n6 = Multiply(A=Add(A=VertexColor().R, B=Time()), B=ScalarParameter() {defaultValue=1, parameterName=FlickerIntensity})
  n7 = Sine(None=@n6) {period=0.075}
  n8 = Multiply(A=Multiply(A=Sine(None=@n6) {period=0.155}, B=Sine(None=@n6) {period=0.2}), B=Sine(None=@n6) {period=0.43})
  n9 = DynamicParameter() {paramNames [FlickerFade,CoronaSize,Param3,Size]}
  n10 = LinearInterpolate(
   A=Clamp(None=Multiply(A=@n7, B=@n8)) {maxDefault=3}
   Alpha=Multiply(A=ScalarParameter() {defaultValue=1, parameterName=FlickerLerp}, B=@n9.FlickerFade)
  ) {constA=1}
  n11 = LinearInterpolate(Alpha=@n10) {constA=-0.4, constB=0}
  n12 = Multiply(A=@n9.CoronaSize, B=ScalarParameter() {defaultValue=1, parameterName=CoronaSize})
  n13 = SphereMask(A=TextureCoordinate(), B=@n5, Radius=@n5, Hardness=LinearInterpolate(Alpha=Add(A=@n11, B=@n12) {constB=0}) {constA=0.7, constB=-0.3}) {hardnessPercent=25}
  n14 = Multiply(
   A=Desaturation(
    None=TextureSample(UVs=@n13) {texture=/Game/Effects/Fort_Effects/Textures/Grad/T_Halo_Corona_Gradient}.RGB
    Fraction=ScalarParameter() {parameterName=Desaturate}
   )
   B=ScalarParameter() {defaultValue=0.5, parameterName=CoronaBrightness}
  ) {constB=0.125}
  n15 = ShadingPathSwitch(Default=Add(A=Multiply(A=@n3, B=@n4) {constB=2}, B=@n14), Forward=@n14)
  n16 = ParticleColor()
  n17 = Multiply(
   A=@n16.RGB
   B=MaterialFunctionCall!(
    Media Player TexCoord
    Media Color Code Coordinate
    Media Texture Resolution
    Base Color Parameter
    Base Color Mix
    Media Color Code Mix
    Intermission Color Parameter
    Intermission Color Mix
    Desaturation
   ) {materialFunction=/Game/Environments/Papaya/Materials/MaterialFunctions/MF_Papaya_GadgetColor}.Result
  )
  n18 = LinearInterpolate(B=@n17) {constA=1, constAlpha=0.7}
  n19 = Add(
   A=Multiply(A=Power(Base=@n2) {constExponent=4}, B=ScalarParameter() {defaultValue=1, parameterName=EdgeBrightness}) {constB=0.15}
   B=TextureSampleParameter2D() {parameterName=FlareMask, texture=/Game/Effects/Fort_Effects/Textures/Flares/T_Flare_Spikeball}.RGB
  )
  MaterialFunctionCall(Emissive=Add(A=Multiply(A=@n15, B=@n18), B=Multiply(A=@n19, B=@n17))) {materialFunction=&MF_FX_EmissiveAdjustments}
  n20 = DepthFade(Opacity=@n16.A, FadeDistance=ScalarParameter() {defaultValue=20, parameterName=DepthFade})
  n21 = MaterialFunctionCall(Distance=ScalarParameter() {defaultValue=150, parameterName=NearCameraShrink}) {materialFunction=&Near_Camera_Fade}
  Multiply(A=@n10, B=Multiply(A=ShadingPathSwitch(Default=@n20, Mobile=@n16.A), B=Power(Base=@n21.Result) {constExponent=3}))
 ]
}
