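# Lit disco ball material for Papaya gadget FX, usable on Niagara mesh particles, with audio-reactive coloring.
# An equalizer function on UV1 offsets the collection-driven base color fed into the gadget color function, then desaturates the result as amplitude drops.
# A dynamic-parameter random value slightly hue-shifts that color, which tints the gained disco ball texture.
# The tint, scaled by raw amplitude, is remapped from LDR to HDR for emissive glow.
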Material /PapayaGameplay/Blueprints/Gadgets/FX/Materials/M_Papaya_DiscoBall
graph {
 properties {bUsedWithNiagaraMeshParticles=true, pixelDepthOffsetMode=PDOM_Legacy, usedShadingModels=Default Lit}
 nodes [
  CollectionParameter() {collection=&FortniteMaterialParameters, parameterName=EventAudio_Amplitude}
  Constant() {r=0.9}
  Constant() {r=1}
  Constant() {r=0.1}
  n1 = Multiply(
   A=TextureSample() {texture=/Game/Environments/Props/Retail/Textures/T_DiscoBall_D}.RGB
   B=ScalarParameter() {defaultValue=3, parameterName=Base Color Gain}
  )
  n2 = Multiply(A=DynamicParameter() {paramNames [MaterialRandom, Param2, Param3, Param4]}.MaterialRandom) {constB=0.1}
  n3 = CollectionParameter() {
   collection=&FortniteMaterialParameters
   parameterName=PapayaBaseColorParameter
  }
  n4 = TextureCoordinate() {coordinateIndex=1}
  n5 = MaterialFunctionCall!(
   Audio Lookup Param=Constant() {r=0.5}
   Average Amplitude Mix=Constant()
   Ramp Texture=TextureObject() {samplerType=SAMPLERTYPE_LinearGrayscale, texture=/PapayaGameplay/Blueprints/Gadgets/FX/Textures/EqualizerComponents_v2_CenterRamp}
   Ramp Exponent=Constant() {r=1}
   Noise Texture
   Noise Scale=Constant2Vector() {g=2, r=4}
   Noise Gain=Constant() {r=0.5}
   Strip Noise Gain
   Grid Tiles X/Y=Constant2Vector() {g=15, r=30}
   Grid Speed
   Output Solid Bars
   TexCoord Override=@n4
  ) {materialFunction=/Game/Environments/Papaya/Materials/MaterialFunctions/MF_Papaya_Equalizer_Data}
  n6 = CollectionParameter() {
   collection=&FortniteMaterialParameters
   parameterName=PapayaBaseColorParameterOffset
  }
  n7 = Multiply(A=@n5.EQ Value, B=Multiply(A=@n6) {constB=2})
  n8 = Desaturation(
   None=MaterialFunctionCall!(
    Media Player TexCoord=@n4
    Media Color Code Coordinate
    Media Texture Resolution
    Base Color Parameter=Add(A=@n3, B=@n7)
    Base Color Mix
    Media Color Code Mix
    Intermission Color Parameter
    Intermission Color Mix
    Desaturation
   ) {materialFunction=/Game/Environments/Papaya/Materials/MaterialFunctions/MF_Papaya_GadgetColor}.Result
   Fraction=OneMinus(None=@n5.EQ Raw Amplitude)
  )
  n9 = Saturate(None=Multiply(
    A=Saturate(None=@n1)
    B=MaterialFunctionCall(Hue Shift Percentage=@n2, Texture=@n8) {materialFunction=&HueShift}.Result
   ))
  n10 = Multiply(A=@n9, B=@n5.EQ Raw Amplitude)
  n11 = MaterialFunctionCall(LDR Input=@n10, HDR Tint=Multiply(A=@n10, B=ScalarParameter() {parameterName=HDR Gain})) {materialFunction=&DeriveHDRfromLDR}
  MaterialFunctionCall(Emissive=@n11.RemappedRange) {materialFunction=&MF_FX_EmissiveAdjustments}
  Reroute(None=@n9)
  Reroute(None=@n11.RemappedRange)
  TextureSample() {samplerType=SAMPLERTYPE_Normal, texture=/Game/Environments/Props/Retail/Textures/T_DiscoBall_N}
 ]
}
