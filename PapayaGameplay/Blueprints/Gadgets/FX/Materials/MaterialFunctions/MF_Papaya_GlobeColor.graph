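# Material function for Papaya gadget sprite globe particles, outputting RGB, Alpha and an audio-reactive WPO.
# Color blends panned MF_Papaya_GadgetColor media, sampled at particle dynamic-parameter UVs, with hue-cycled high-contrast particle color.
# EventAudio_Spectrum, masked by that particle color, drives brightness and pushes vertices outward from the actor position.
# MPC_PapayaGadgetControls toggles solid versus per-sprite UVs and selects alpha: opaque, matte, animated noise, lattice or dots.
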MaterialFunction /PapayaGameplay/Blueprints/Gadgets/FX/Materials/MaterialFunctions/MF_Papaya_GlobeColor
graph {
 properties {bExposeToLibrary=true}
 nodes [
  Fmod(A=Multiply(A=Time()) {constB=0.1}, B=Constant() {r=5})
  n1 = Time()
  n2 = MaterialFunctionCall(X=ScalarParameter() {defaultValue=0.005, parameterName=Panner Speed Color}, Y=Constant()) {materialFunction=&MakeFloat2}
  n3 = DynamicParameter() {paramNames [Random,U,V,Param4]}
  n4 = MaterialFunctionCall(X=@n3.U, Y=@n3.U) {materialFunction=&MakeFloat2}
  n5 = TextureCoordinate()
  n6 = DynamicParameter() {parameterIndex=2, defaultValue=<rgba 0 0 0 0>, paramNames [UFraction,VFraction,Param3,Param4]}
  n7 = MaterialFunctionCall(Float2=Multiply(
    A=Subtract(A=@n5) {constB=0.5}
    B=MaterialFunctionCall(X=@n6.UFraction, Y=@n6.UFraction) {materialFunction=&MakeFloat2}.Result
   )) {materialFunction=&BreakOutFloat2Components}
  n8 = Add(
   A=@n4.Result
   B=MaterialFunctionCall(X=@n7.G, Y=@n7.G) {materialFunction=&MakeFloat2}.Result
  )
  n9 = If(
   A=CollectionParameter() {collection=/Game/Environments/Papaya/Materials/MPC_PapayaGadgetControls, parameterName=SpriteGlobeSolid}
   "A > B"=@n4.Result
   "A == B"=@n4.Result
   "A < B"=@n8
  ) {constB=1}
  n10 = Constant() {r=2}
  n11 = Constant() {r=1}
  n12 = Frac(None=Multiply(
    A=Panner(Coordinate=@n9, Time=@n1, Speed=@n2.Result)
    B=MaterialFunctionCall(X=@n10, Y=@n11) {materialFunction=&MakeFloat2}.Result
   ))
  n13 = CollectionParameter() {
   collection=&FortniteMaterialParameters
   parameterName=PapayaBaseColorParameter
  }
  n14 = CollectionParameter() {
   collection=&FortniteMaterialParameters
   parameterName=PapayaBaseColorParameterOffset
  }
  n15 = Add(A=@n13, B=Multiply(A=@n14, B=ParticleColor().R))
  n16 = CollectionParameter() {collection=&FortniteMaterialParameters, parameterName=PapayaBaseColorMix}
  n17 = Min(A=@n16, B=ScalarParameter() {defaultValue=0.2, parameterName=BaseColorOverride})
  n18 = Multiply(A=Time()) {constB=0.05}
  n19 = MaterialFunctionCall(Hue Shift Percentage=Multiply(A=DynamicParameter() {paramNames [Random,U,V,Param4]}.Random) {constB=0.25}, Texture=ParticleColor().RGB) {materialFunction=&HueShift}
  n20 = Saturate(None=MaterialFunctionCall!(
    In=MaterialFunctionCall(Hue Shift Percentage=@n18, Texture=@n19.Result) {materialFunction=&HueShift}.Result
    Contrast=Constant() {r=100}
   ) {materialFunction=/Engine/Functions/Engine_MaterialFunctions01/ImageAdjustment/CheapContrast_RGB}.Result)
  n21 = DynamicParameter() {paramNames [Random,U,V,Param4]}
  n22 = Multiply(
   A=LinearInterpolate(
    A=MaterialFunctionCall!(
     Media Player TexCoord=@n12
     Media Color Code Coordinate
     Media Texture Resolution
     Base Color Parameter=@n15
     Base Color Mix=@n17
     Media Color Code Mix
     Intermission Color Parameter
     Intermission Color Mix
     Desaturation
    ) {materialFunction=/Game/Environments/Papaya/Materials/MaterialFunctions/MF_Papaya_GadgetColor}.Result
    B=@n20
    Alpha=ScalarParameter() {parameterName=RGB Mix}
   )
   B=Add(A=Multiply(A=@n21.Random) {constB=0.5}) {constB=0.5}
  )
  n23 = MaterialFunctionCall(Float3=@n20) {materialFunction=&BreakOutFloat3Components}
  n24 = CollectionParameter() {collection=&FortniteMaterialParameters, parameterName=EventAudio_Spectrum}
  n25 = MaterialFunctionCall(Float4=@n24) {materialFunction=&BreakOutFloat4Components}
  n26 = Multiply(A=@n23.R, B=@n25.R)
  n27 = Multiply(A=@n23.G, B=@n25.G)
  n28 = Saturate(None=Max(A=Max(A=@n26, B=@n27), B=Multiply(A=@n23.B, B=@n25.B)))
  n29 = Add(A=Multiply(A=Power(Base=@n28) {constExponent=5}, B=ScalarParameter() {defaultValue=2, parameterName=Audio Brightness Gain}) {constB=50})
  FunctionOutput(None=Multiply(A=Multiply(A=@n22, B=@n29), B=ScalarParameter() {defaultValue=0.25, parameterName=Brightness Gain})) {outputName=RGB}
  n30 = Panner(
   Coordinate=StaticSwitch(True=@n8, False=@n4.Result, Value=StaticBoolParameter() {parameterName=Smooth Alpha})
   Time=@n1
   Speed=MaterialFunctionCall(X=ScalarParameter() {defaultValue=0.01, parameterName=Panner Speed Alpha}, Y=Constant()) {materialFunction=&MakeFloat2}.Result
  )
  n31 = TextureSampleParameter2D(UVs=@n30) {parameterName=Alpha Texture, samplerType=SAMPLERTYPE_LinearGrayscale, texture=/PapayaGameplay/Blueprints/Gadgets/FX/Textures/globe_matte}
  n32 = Constant2Vector() {g=1, r=2}
  n33 = MaterialFunctionCall(
   Hue Shift Percentage=Multiply(A=Time()) {constB=0.04}
   Texture=TextureSample(UVs=Multiply(A=@n30, B=@n32)) {samplerType=SAMPLERTYPE_LinearColor, texture=&HighContrastNoise}.RGB
  ) {materialFunction=&HueShift}
  n34 = Constant() {r=10}
  n35 = TextureSample(UVs=@n30) {samplerType=SAMPLERTYPE_LinearGrayscale, texture=/PapayaGameplay/Blueprints/Gadgets/FX/Textures/GlobeDotsPolar}
  n36 = CollectionParameter() {collection=&MPC_PapayaGadgetControls, parameterName=SpriteGlobeAlphaMode}
  n37 = If(
   A=@n36
   "A > B"=@n35.R
   "A == B"=@n35.R
   "A < B"=TextureSample(UVs=@n30) {samplerType=SAMPLERTYPE_LinearGrayscale, texture=/PapayaGameplay/Blueprints/Gadgets/FX/Textures/GlobeLatticePolar}.R
  ) {constB=4}
  n38 = If(
   A=@n36
   "A > B"=@n37
   "A == B"=@n37
   "A < B"=ComponentMask(None=MaterialFunctionCall(In=@n33.Result, Contrast=@n34) {materialFunction=&CheapContrast}.Result) {r=true}
  ) {constB=3}
  n39 = If(A=@n36, "A > B"=@n38, "A == B"=@n38, "A < B"=@n31.R) {constB=2}
  FunctionOutput(None=If(A=@n36, "A > B"=@n39, "A == B"=@n39, "A < B"=Constant() {r=1}) {constB=1}) {outputName=Alpha, sortPriority=10}
  n40 = Normalize(VectorInput=Subtract(A=WorldPosition().XYZ, B=ActorPositionWS()))
  FunctionOutput(None=Multiply(A=Multiply(A=@n40, B=@n28), B=ScalarParameter() {defaultValue=50, parameterName=WPO Scale})) {outputName=WPO, sortPriority=20}
 ]
}
