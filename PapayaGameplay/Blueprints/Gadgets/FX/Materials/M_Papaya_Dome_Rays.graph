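# Additive unlit material for Niagara sprites that draws light rays for the Papaya dome gadget effect.
# A high-contrast noise sampled only along U, offset by a per-particle random value, forms streaks masked by side and vertical UV falloffs.
# MF_Papaya_GlobeColor supplies color, alpha and world position offset, which grows along V.
# Emissive adjustments process the color, and depth fade softens the scaled ray opacity.
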Material /PapayaGameplay/Blueprints/Gadgets/FX/Materials/M_Papaya_Dome_Rays
graph {
 properties {bUsedWithNiagaraSprites=true, blendMode=BLEND_Additive, pixelDepthOffsetMode=PDOM_Legacy, shadingModel=MSM_Unlit, usedShadingModels=Unlit}
 nodes [
  n1 = MaterialFunctionCall(Float2=TextureCoordinate()) {materialFunction=&BreakOutFloat2Components}
  n2 = MaterialFunctionCall() {materialFunction=/PapayaGameplay/Blueprints/Gadgets/FX/Materials/MaterialFunctions/MF_Papaya_GlobeColor}
  Add(A=@n2.WPO, B=Multiply(A=OneMinus(None=@n1.G), B=@n2.WPO))
  n3 = ScalarParameter() {defaultValue=2, parameterName=Alpha Scale}
  n4 = MaterialFunctionCall(Float2=TextureCoordinate()) {materialFunction=&BreakOutFloat2Components}
  n5 = Multiply(A=Add(A=@n4.R, B=DynamicParameter() {paramNames [Random,U,V,Param4]}.Random)) {constB=0.5}
  n6 = Constant()
  n7 = Multiply(A=Abs(None=Subtract(A=@n4.R) {constB=0.5})) {constB=2}
  n8 = Saturate(None=Multiply(A=OneMinus(None=@n4.G)) {constB=5})
  n9 = Multiply(
   A=TextureSample(UVs=MaterialFunctionCall(X=@n5, Y=@n6) {materialFunction=&MakeFloat2}.Result) {samplerType=SAMPLERTYPE_LinearColor, texture=&HighContrastNoise}.R
   B=Multiply(A=OneMinus(None=@n7), B=Multiply(A=@n4.G, B=@n8))
  )
  LinearInterpolate(B=Multiply(A=Multiply(A=@n2.Alpha, B=@n3) {constB=0.5}, B=Power(Base=@n9) {constExponent=1.5}), Alpha=DepthFade() {fadeDistanceDefault=500})
  MaterialFunctionCall(Emissive=@n2.RGB) {materialFunction=&MF_FX_EmissiveAdjustments}
 ]
}
