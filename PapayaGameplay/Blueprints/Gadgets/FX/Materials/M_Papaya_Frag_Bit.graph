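# Translucent, two-sided unlit material for Niagara mesh particles, apparently cracked-glass bits for a Papaya gadget effect.
# Glass-crack normals and masks, sampled at two tilings, drive a Fresnel rim, bright cracks and refraction scaled by particle alpha.
# Geode crack and cloud colors blend toward Papaya gadget colors via "Use Concert Colors".
# That glow, tinted by particle color and Strength, adds to a geode cubemap before emissive adjustment.
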Material /PapayaGameplay/Blueprints/Gadgets/FX/Materials/M_Papaya_Frag_Bit
graph {
 properties {
  bUsedWithNiagaraMeshParticles=true
  blendMode=BLEND_Translucent
  pixelDepthOffsetMode=PDOM_Legacy
  refractionMethod=RM_IndexOfRefraction
  shadingModel=MSM_Unlit
  twoSided=true
  usedShadingModels=Unlit
 }
 nodes [
  n1 = TextureCoordinate() {uTiling=3.5, vTiling=4.2}
  n2 = Add(
   A=TextureSample() {samplerType=SAMPLERTYPE_Normal, texture=/Game/Effects/Fort_Effects/Textures/Tile/T_Glass_Cracks_N}.RGB
   B=Multiply(
    A=TextureSample(UVs=@n1) {samplerType=SAMPLERTYPE_Normal, texture=&T_Glass_Cracks_N}.RGB
    B=Constant3Vector() {constant=<rgba 1 1 0 0>}
   )
  )
  n3 = Fresnel(Normal=Multiply(A=Transform(None=@n2), B=TwoSidedSign())) {baseReflectFraction=0, exponent=4}
  n4 = Add(
   A=TextureSample() {texture=/Game/Effects/Fort_Effects/Textures/Tile/T_Glass_Cracks_M}.G
   B=Multiply(A=TextureSample(UVs=@n1) {texture=&T_Glass_Cracks_M}.G) {constB=0.1}
  )
  n5 = LinearInterpolate(Alpha=@n4) {constA=0.25}
  n6 = Add(A=@n3, B=@n5)
  n7 = ParticleColor()
  LinearInterpolate(
   B=ScalarParameter() {defaultValue=1.1, parameterName=Refraction}
   Alpha=Multiply(A=ScalarParameter() {defaultValue=1, parameterName=Opacity}, B=Multiply(A=@n6, B=@n7.A))
  ) {constA=1}
  n8 = MaterialFunctionCall!(Use Alternate UVs, Alternate UVs) {materialFunction=/Game/Athena/Effects/Materials/Geode/Functions/MF_Geode_Crack_Colors}
  n9 = CollectionParameter() {
   collection=&FortniteMaterialParameters
   parameterName=PapayaBaseColorParameter
  }
  n10 = CollectionParameter() {
   collection=&FortniteMaterialParameters
   parameterName=PapayaBaseColorParameterOffset
  }
  n11 = MaterialFunctionCall!(
   Media Player TexCoord
   Media Color Code Coordinate
   Media Texture Resolution
   Base Color Parameter=Add(A=@n9, B=@n10)
   Base Color Mix
   Media Color Code Mix
   Intermission Color Parameter
   Intermission Color Mix
   Desaturation
  ) {materialFunction=/Game/Environments/Papaya/Materials/MaterialFunctions/MF_Papaya_GadgetColor}
  n12 = ScalarParameter() {parameterName=Use Concert Colors}
  n13 = LinearInterpolate(
   A=@n8.Crack Color
   B=MaterialFunctionCall() {materialFunction=&MF_Papaya_GadgetColor}.Result
   Alpha=@n12
  )
  n14 = Add(
   A=Multiply(A=Multiply(A=@n3, B=@n5), B=LinearInterpolate(A=@n8.Cloud Color, B=@n11.Result, Alpha=@n12))
   B=Multiply(A=@n4, B=Multiply(A=@n13) {constB=8})
  )
  n15 = Multiply(A=ScalarParameter() {defaultValue=1, parameterName=Strength}, B=@n7.RGB)
  MaterialFunctionCall(Emissive=Add(
    A=MaterialFunctionCall!(UV add) {materialFunction=/Game/Athena/Effects/Materials/Geode/Functions/MF_Geode_Cubemap}.Result
    B=Multiply(A=@n14, B=@n15)
   )) {materialFunction=&MF_FX_EmissiveAdjustments}
 ]
}
