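# Material function that produces an audio-reactive, pulsing color for the Papaya gadget globe lights.
# A polar noise texture gets a contrast adjustment and a time-driven hue shift, cycling its three channels as animated masks.
# Three EventAudio_Spectrum components from FortniteMaterialParameters are remapped and clamped, then multiplied into those masks.
# The masks weight the MF_Papaya_GadgetColor result and two slightly hue-shifted variants, which are summed as output.
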MaterialFunction /PapayaGameplay/Blueprints/Gadgets/FX/Materials/MaterialFunctions/MF_Papaya_GlobeLightsPulse
graph {
 properties {bExposeToLibrary=true}
 nodes [
  n1 = MaterialFunctionCall!(
   Media Player TexCoord
   Media Color Code Coordinate
   Media Texture Resolution
   Base Color Parameter
   Base Color Mix
   Media Color Code Mix
   Intermission Color Parameter
   Intermission Color Mix
   Desaturation
  ) {materialFunction=/Game/Environments/Papaya/Materials/MaterialFunctions/MF_Papaya_GadgetColor}
  n2 = Multiply(
   A=Time()
   B=FunctionInput(Preview=Constant() {r=0.5}) {bUsePreviewValueAsDefault=true, inputName=Noise Speed, inputType=FunctionInput_Scalar}
  )
  n3 = TextureObject() {texture=/PapayaGameplay/Blueprints/Gadgets/FX/Textures/GlobeNoisePolar}
  n4 = MaterialFunctionCall!(
   In=TextureSample(Tex=FunctionInput(Preview=@n3) {bUsePreviewValueAsDefault=true, inputType=FunctionInput_Texture2D}) {texture=&GlobeNoisePolar}.RGB
   Contrast=FunctionInput(Preview=Constant() {r=0.5}) {bUsePreviewValueAsDefault=true, inputName=Noise Contrast, inputType=FunctionInput_Scalar}
  ) {materialFunction=/Engine/Functions/Engine_MaterialFunctions01/ImageAdjustment/CheapContrast_RGB}
  n5 = MaterialFunctionCall(Float3=MaterialFunctionCall(Hue Shift Percentage=Frac(None=@n2), Texture=Saturate(None=@n4.Result)) {materialFunction=&HueShift}.Result) {materialFunction=&BreakOutFloat3Components}
  n6 = CollectionParameter() {collection=&FortniteMaterialParameters, parameterName=EventAudio_Spectrum}
  n7 = ComponentMask(None=@n6) {a=true, b=true, g=true}
  n8 = Constant3Vector() {constant=<rgba 0.5 0.5 0.5 0>}
  n9 = Constant3Vector() {constant=<rgba 0.8 0.8 0.8 0>}
  n10 = Constant3Vector()
  n11 = Constant3Vector() {constant=<rgba 1 1 1 0>}
  n12 = MaterialFunctionCall(Float3=Saturate(None=MaterialFunctionCall!(Input=@n7, Input Low=@n8, Input High=@n9, Target Low=@n10, Target Height=@n11) {materialFunction=&/RemapValueRangeVector}.Result)) {materialFunction=&BreakOutFloat3Components}
  n13 = Multiply(A=@n1.Result, B=Multiply(A=@n5.R, B=@n12.R))
  n14 = Multiply(
   A=MaterialFunctionCall(Hue Shift Percentage=Constant() {r=0.05}, Texture=@n1.Result) {materialFunction=&HueShift}.Result
   B=Multiply(A=@n5.G, B=@n12.G)
  )
  n15 = MaterialFunctionCall(Hue Shift Percentage=Constant() {r=0.1}, Texture=@n1.Result) {materialFunction=&HueShift}
  n16 = Multiply(A=@n5.B, B=@n12.B)
  FunctionOutput(None=Add(A=Add(A=@n13, B=@n14), B=Multiply(A=@n15.Result, B=@n16)))
 ]
}
