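# Two-sided lit material for the Papaya globe gadget, with globe color and normal maps, a tinted Fresnel rim, and roughness from a remapped mask varied per segment.
# Emissive blends pulsing lattice and segment glow with pulsing globe lights by mask, using MF_Papaya_GlobeLightsPulse.
# It adds MF_Papaya_GadgetColor through a narrow threshold on a noise- and segment-jittered gradient ramp, driven by PapayaGlobeMaskParameter in the FortniteMaterialParameters collection.
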Material /PapayaGameplay/Blueprints/Gadgets/FX/Materials/M_Papaya_Globe
graph {
 properties {pixelDepthOffsetMode=PDOM_Legacy, twoSided=true, usedShadingModels=Default Lit}
 nodes [
  n1 = CameraVectorWS()
  n2 = Multiply(
   A=TextureSample() {samplerType=SAMPLERTYPE_Normal, texture=/PapayaGameplay/Blueprints/Gadgets/FX/Textures/globe_normal}.RGB
   B=Constant3Vector() {constant=<rgba 1 -1 1 0>}
  )
  n3 = Transform(None=@n2)
  n4 = OneMinus(None=Saturate(None=DotProduct(A=@n1, B=@n3)))
  Add(
   A=TextureSample() {texture=/PapayaGameplay/Blueprints/Gadgets/FX/Textures/globe_color}.RGB
   B=Multiply(A=VectorParameter() {parameterName=Fresnel Color, defaultValue=<rgba 0.2452 0.3166 0.3594 1>}.RGB, B=Power(Base=@n4) {constExponent=5})
  )
  n5 = TextureSample() {texture=/PapayaGameplay/Blueprints/Gadgets/FX/Textures/globe_masks_2k}
  n6 = Constant()
  n7 = Constant() {r=0.6}
  n8 = Constant() {r=0.45}
  n9 = Constant() {r=0.9}
  n10 = TextureSample() {samplerType=SAMPLERTYPE_LinearGrayscale, texture=/PapayaGameplay/Blueprints/Gadgets/FX/Textures/GlobeSegments2Polar}
  Add(
   A=Saturate(None=MaterialFunctionCall(Input=@n5.R, Input Low=@n6, Input High=@n7, Target Low=@n8, Target High=@n9) {materialFunction=&RemapValueRange}.Result)
   B=LinearInterpolate(Alpha=@n10.G) {constA=-0.1, constB=0.1}
  )
  n11 = MaterialFunctionCall!(
   Media Player TexCoord
   Media Color Code Coordinate
   Media Texture Resolution
   Base Color Parameter
   Base Color Mix
   Media Color Code Mix
   Intermission Color Parameter
   Intermission Color Mix
   Desaturation
  ) {materialFunction=/Game/Environments/Papaya/Materials/MaterialFunctions/MF_Papaya_GadgetColor}
  n12 = LinearInterpolate(Alpha=@n10.G) {constA=-0.01, constB=0.01}
  n13 = Add(
   A=TextureSample() {samplerType=SAMPLERTYPE_LinearGrayscale, texture=/Game/Global/Textures/Default/Masks_and_Gradients/EqualizerComponents_v2_Ramp}.R
   B=LinearInterpolate(Alpha=TextureSample() {texture=/PapayaGameplay/Blueprints/Gadgets/FX/Textures/GlobeNoisePolar}.R) {constA=-0.05, constB=0.05}
  )
  n14 = CollectionParameter() {
   collection=&FortniteMaterialParameters
   parameterName=PapayaGlobeMaskParameter
  }
  n15 = LinearInterpolate(Alpha=@n14) {constA=-0.025}
  n16 = Saturate(None=MaterialFunctionCall(Input=Add(A=@n12, B=@n13), Input Low=@n15, Input High=Add(A=@n15) {constB=0.025}, Target Low=Constant(), Target High=Constant() {r=1}) {materialFunction=&RemapValueRange}.Result)
  n17 = TextureSample() {samplerType=SAMPLERTYPE_LinearGrayscale, texture=/PapayaGameplay/Blueprints/Gadgets/FX/Textures/GlobeLattice2Polar}
  n18 = Multiply(A=OneMinus(None=@n5.B)) {constB=0.1}
  n19 = Multiply(
   A=MaterialFunctionCall!(In=TextureObject() {texture=/PapayaGameplay/Blueprints/Gadgets/FX/Textures/GlobeSegmentsPolar}, Noise Contrast=Constant(), Noise Speed) {materialFunction=/PapayaGameplay/Blueprints/Gadgets/FX/Materials/MaterialFunctions/MF_Papaya_GlobeLightsPulse}.Result
   B=Max(A=@n5.G, B=Max(A=@n17.R, B=@n18))
  )
  n20 = Multiply(
   A=MaterialFunctionCall(In=TextureObject() {texture=&GlobeNoisePolar}) {materialFunction=&MF_Papaya_GlobeLightsPulse}.Result
   B=TextureSample() {texture=/PapayaGameplay/Blueprints/Gadgets/FX/Textures/globe_lights_2k}.R
  )
  Add(
   A=LinearInterpolate(A=Multiply(A=@n19) {constB=0.8}, B=Multiply(A=@n20) {constB=2.5}, Alpha=@n5.B)
   B=Multiply(A=Multiply(A=@n11.Result) {constB=3}, B=OneMinus(None=@n16))
  )
  Constant()
  Reroute(None=@n2)
 ]
}
