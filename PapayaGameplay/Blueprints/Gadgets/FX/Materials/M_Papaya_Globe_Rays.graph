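# Additive, unlit, two-sided material for the light rays of the Papaya globe gadget effect.
# Brightness comes from the globe lights pulse function, multiplied by cubed vertex red and a Gain parameter.
# A mask offsets a gradient ramp with polar noise, then remaps it through a narrow, saturated threshold window.
# The PapayaGlobeMaskParameter value from a material parameter collection slides that window, producing a noisy-edged wipe.
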Material /PapayaGameplay/Blueprints/Gadgets/FX/Materials/M_Papaya_Globe_Rays
graph {
 properties {
  bCastRayTracedShadows=false
  bUsedWithNanite=true
  blendMode=BLEND_Additive
  pixelDepthOffsetMode=PDOM_Legacy
  shadingModel=MSM_Unlit
  twoSided=true
  usedShadingModels=Unlit
 }
 nodes [
  n1 = VertexColor()
  Multiply(
   A=Multiply(
    A=MaterialFunctionCall!(In, Noise Contrast, Noise Speed) {materialFunction=/PapayaGameplay/Blueprints/Gadgets/FX/Materials/MaterialFunctions/MF_Papaya_GlobeLightsPulse}.Result
    B=Power(Base=@n1.R) {constExponent=3}
   )
   B=ScalarParameter() {defaultValue=0.1, parameterName=Gain}
  )
  n2 = TextureSample() {samplerType=SAMPLERTYPE_LinearGrayscale, texture=/Game/Global/Textures/Default/Masks_and_Gradients/EqualizerComponents_v2_Ramp}
  n3 = LinearInterpolate(Alpha=TextureSample() {texture=/PapayaGameplay/Blueprints/Gadgets/FX/Textures/GlobeNoisePolar}.R) {constA=-0.05, constB=0.05}
  n4 = CollectionParameter() {
   collection=&FortniteMaterialParameters
   parameterName=PapayaGlobeMaskParameter
  }
  n5 = LinearInterpolate(Alpha=@n4) {constA=-0.025}
  Saturate(None=MaterialFunctionCall(Input=Add(A=@n2.R, B=@n3), Input Low=@n5, Input High=Add(A=@n5) {constB=0.025}, Target Low=Constant(), Target High=Constant() {r=1}) {materialFunction=&RemapValueRange}.Result)
 ]
}
