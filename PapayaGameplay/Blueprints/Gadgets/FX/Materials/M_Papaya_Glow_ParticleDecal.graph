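# Additive unlit Niagara sprite material for Papaya gadget FX: a decal-like glow tinted by gadget color and particle color.
# Panning, optionally posterized smoke noise, a radial mask and depth fades shape opacity, with optional time-of-day scaling.
# On deferred rendering, a Fresnel from scene world normals adds surface-aware edges; other paths use constants, and mobile replaces depth fade with fixed opacity.
# World position offset shrinks sprites near the camera.
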Material /PapayaGameplay/Blueprints/Gadgets/FX/Materials/M_Papaya_Glow_ParticleDecal
graph {
 properties {bUsedWithNiagaraSprites=true, blendMode=BLEND_Additive, pixelDepthOffsetMode=PDOM_Legacy, shadingModel=MSM_Unlit, usedShadingModels=Unlit}
 nodes [
  n1 = ShadingPathSwitch(Default=ParticlePositionWS(), Mobile=ObjectPositionWS())
  n2 = WorldPosition()
  n3 = Divide(A=Distance(A=@n1, B=CameraPositionWS()), B=ScalarParameter() {defaultValue=250, group=WPO, parameterName=NearFadeDistance, sortPriority=1})
  LinearInterpolate(
   B=LinearInterpolate(A=Subtract(A=@n1, B=@n2.XYZ), Alpha=Clamp(None=@n3)) {constB=0}
   Alpha=ScalarParameter() {defaultValue=1, group=WPO, parameterName=AdjustSizeWithCamera}
  )
  n4 = ScalarParameter() {group=Opacity, parameterName=OpacityMorning, sortPriority=1}
  n5 = ScalarParameter() {group=Opacity, parameterName=OpacityDay, sortPriority=2}
  n6 = ScalarParameter() {defaultValue=1, group=Opacity, parameterName=OpacityEvening, sortPriority=3}
  n7 = ScalarParameter() {defaultValue=1, group=Opacity, parameterName=OpacityNight, sortPriority=4}
  n8 = ParticleColor()
  n9 = ComponentMask(None=SceneTexture() {sceneTextureId=PPI_WorldNormal}.Color) {b=true, g=true, r=true}
  n10 = Constant() {r=0.2}
  n11 = ShadingPathSwitch(
   Default=Constant() {r=0.3}
   Deferred=Fresnel(
    ExponentIn=DynamicParameter() {paramNames [FresnelExponent, Refraction, PanAdd, Random]}.FresnelExponent
    Normal=MaterialFunctionCall(Normal=@n9, Flatness=@n10) {materialFunction=&FlattenNormal}.Result
   ) {baseReflectFraction=0, exponent=1}
  )
  n12 = SphereMask(A=TextureCoordinate(), B=Constant() {r=0.5}, Radius=ScalarParameter() {defaultValue=0.5, group=Opacity, parameterName=SphereMaskRadius}) {hardnessPercent=0}
  n13 = ScalarParameter() {defaultValue=2, group=Opacity, parameterName=SphereMaskPower, sortPriority=5}
  n14 = TextureCoordinate()
  n15 = AppendVector(
   A=ScalarParameter() {defaultValue=1, group=Tiling Texture, parameterName=NoiseTilingU, sortPriority=3}
   B=ScalarParameter() {defaultValue=0.6, group=Tiling Texture, parameterName=NoiseTilingV, sortPriority=4}
  )
  n16 = DynamicParameter() {paramNames [FresnelExponent, Refraction, PanAdd, Random]}
  n17 = Multiply(A=Time(), B=LinearInterpolate(Alpha=@n16.Random) {constA=0.4, constB=1.4})
  n18 = Add(A=@n16.PanAdd, B=@n16.PanAdd)
  n19 = ScalarParameter() {defaultValue=0.05, group=Tiling Texture, parameterName=PannerSpeedX, sortPriority=5}
  n20 = ScalarParameter() {defaultValue=0.5, group=Tiling Texture, parameterName=PannerSpeedY, sortPriority=6}
  n21 = TextureSampleParameter2D(UVs=Panner(Coordinate=Multiply(A=@n14, B=@n15), Time=Add(A=@n17, B=@n18), Speed=AppendVector(A=@n19, B=@n20)) {speedX=0.05, speedY=0.5}) {group=Tiling Texture, parameterName=TilingTexture, texture=&T_Inky_Smoke_Tile_Chromatic}
  n22 = ScalarParameter() {defaultValue=5, group=Tiling Texture, parameterName=Floor, sortPriority=2}
  n23 = LinearInterpolate(
   A=@n21.RGB
   B=Floor(None=Multiply(A=@n21.RGB, B=@n22))
   Alpha=ScalarParameter() {group=Tiling Texture, parameterName=Floor Tiling Texture Values, sortPriority=1}
  )
  n24 = ScalarParameter() {defaultValue=1, group=Opacity, parameterName=TextureOpacityAdd, sortPriority=8}
  n25 = DepthFade(FadeDistance=ScalarParameter() {defaultValue=44, group=Opacity, parameterName=VisibleDepthDistance, sortPriority=9}) {fadeDistanceDefault=80}
  n26 = Multiply(
   A=LinearInterpolate(B=@n23, Alpha=ScalarParameter() {defaultValue=1, group=Opacity, parameterName=TextureOpacityMult, sortPriority=7}) {constA=1}
   B=Add(A=Multiply(A=@n23, B=@n24), B=OneMinus(None=@n25))
  )
  n27 = ScalarParameter() {defaultValue=20, group=Opacity, parameterName=DepthFade, sortPriority=6}
  n28 = Multiply(
   A=Multiply(A=Multiply(A=@n8.A, B=@n11) {constA=1}, B=Power(Base=@n12, Exp=@n13))
   B=ShadingPathSwitch(
    Default=DepthFade(Opacity=@n26, FadeDistance=@n27) {fadeDistanceDefault=15}
    Mobile=ScalarParameter() {defaultValue=0.15, group=Opacity, parameterName=Mobile Opacity, sortPriority=10}
   )
  )
  LinearInterpolate(
   A=@n28
   B=Multiply(
    A=MaterialFunctionCall(Color_Morning=@n4, Color_Day=@n5, Color_Evening=@n6, Color_Night=@n7) {materialFunction=&Color_Time_of_Day}.Result
    B=@n28
   )
   Alpha=ScalarParameter() {group=Opacity, parameterName=Multiply by Time of Day}
  )
  n29 = ScalarParameter() {defaultValue=-1.5, group=Color, parameterName=Desaturation Factor, sortPriority=1}
  n30 = Add(
   A=VectorParameter() {group=Color, parameterName=DarkColor, sortPriority=2, defaultValue=<rgba 0.0501 0.0332 0.1302 1>}.RGB
   B=Multiply(A=Desaturation(None=@n23, Fraction=@n29)) {constB=1.5}
  )
  n31 = Multiply(
   A=MaterialFunctionCall!(
    Media Player TexCoord
    Media Color Code Coordinate
    Media Texture Resolution
    Base Color Parameter
    Base Color Mix
    Media Color Code Mix
    Intermission Color Parameter
    Intermission Color Mix
    Desaturation
   ) {materialFunction=/Game/Environments/Papaya/Materials/MaterialFunctions/MF_Papaya_GadgetColor}.Result
   B=@n8.RGB
  )
  n32 = LinearInterpolate(A=@n31, Alpha=ScalarParameter() {defaultValue=0.6, group=Color, parameterName=Lack of Particle Influence}) {constAlpha=0.6}
  n33 = Multiply(A=@n31, B=@n11)
  n34 = ScalarParameter() {defaultValue=5, group=World Normal, parameterName=WorldNormalIntensity}
  MaterialFunctionCall(Emissive=Add(A=Multiply(A=@n30, B=@n32), B=Multiply(A=@n33, B=@n34) {constB=5, desc=Fresnel edge color})) {materialFunction=&MF_FX_EmissiveAdjustments}
 ]
}
