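# Translucent unlit particle sprite material for hologram light shafts on Papaya gadgets.
# Color comes from the shared gadget color function, multiplied by particle color and passed through an FX emissive adjustment function.
# Particle alpha times the green channel of a flare light-shaft mask forms the opacity signal.
# The default shading path applies scene depth fade and camera depth fade to it, while mobile uses the plain masked alpha.
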Material /PapayaGameplay/Blueprints/Gadgets/FX/Materials/M_Papaya_HologramLightShafts
graph {
 properties {
  bUsedWithParticleSprites=true
  blendMode=BLEND_Translucent
  pixelDepthOffsetMode=PDOM_Legacy
  shadingModel=MSM_Unlit
  usedShadingModels=Unlit
 }
 nodes [
  n1 = ParticleColor()
  MaterialFunctionCall(Emissive=Multiply(
    A=MaterialFunctionCall!(
     Media Player TexCoord
     Media Color Code Coordinate
     Media Texture Resolution
     Base Color Parameter
     Base Color Mix
     Media Color Code Mix
     Intermission Color Parameter
     Intermission Color Mix
     Desaturation
    ) {materialFunction=/Game/Environments/Papaya/Materials/MaterialFunctions/MF_Papaya_GadgetColor}.Result
    B=@n1.RGB
   )) {materialFunction=&MF_FX_EmissiveAdjustments}
  n2 = Multiply(A=@n1.A, B=TextureSample(UVs=TextureCoordinate()) {texture=/Game/Effects/Fort_Effects/Textures/Mask/T_Flare_LightShafts}.G)
  n3 = ScalarParameter() {defaultValue=50, parameterName=DepthFade}
  n4 = ScalarParameter() {defaultValue=100, parameterName=CameraFadeLength}
  n5 = ScalarParameter() {parameterName=CameraFadeOffset}
  ShadingPathSwitch(
   Default=Multiply(
    A=DepthFade(Opacity=@n2, FadeDistance=@n3)
    B=MaterialFunctionCall(Fade Length=@n4, Fade Offset=@n5) {materialFunction=&CameraDepthFade}.Result
   )
   Mobile=@n2
  )
 ]
}
