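# Additive, unlit, two-sided mesh-particle material for a gadget light pillar that grows with camera distance to stay visible from afar.
# World position offset combines SplineThicken widening with scaling about the object position, both driven by clamped camera distance.
# Dynamic parameters set width, hue shift, desaturation and a sine pulse affecting base width and opacity.
# UV gradients shape opacity, with particle alpha, near-camera fade and depth fade, which mobile skips.
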Material /PapayaGameplay/Blueprints/Gadgets/FX/Materials/M_Papaya_Light_Pillar_Distance_Scale
graph {
 properties {
  bUsedWithMeshParticles=true
  blendMode=BLEND_Additive
  pixelDepthOffsetMode=PDOM_Legacy
  shadingModel=MSM_Unlit
  twoSided=true
  usedShadingModels=Unlit
 }
 nodes [
  n1 = DynamicParameter() {defaultValue=<rgba 100 1 1 1>, paramNames [SplineWidth,HueShift,Desat,PulseSpeed]}
  n2 = LinearInterpolate(
   A=Constant() {r=0.25}
   B=Constant() {r=1}
   Alpha=MaterialFunctionCall(Speed=DynamicParameter() {defaultValue=<rgba 100 1 1 1>, paramNames [SplineWidth,HueShift,Desat,PulseSpeed]}.PulseSpeed) {materialFunction=&DebugTimeSine}.Result
  )
  n3 = ObjectPositionWS()
  n4 = Distance(A=@n3, B=CameraPositionWS())
  n5 = ScalarParameter() {defaultValue=50, parameterName=DistanceOffset}
  n6 = Clamp(None=Divide(A=Subtract(A=@n4, B=@n5), B=ScalarParameter() {defaultValue=30000, parameterName=ScaleUpDistance}))
  n7 = LinearInterpolate(B=Multiply(A=ScalarParameter() {defaultValue=500, parameterName=MaxWidthMult}, B=@n2), Alpha=@n6) {constA=1, constB=3}
  n8 = LinearInterpolate(B=ScalarParameter() {defaultValue=5, parameterName=MaxWidthMult_Tip}, Alpha=@n6) {constA=1, constB=3}
  n9 = Multiply(A=ScalarParameter() {defaultValue=60, parameterName=MaxScaleIncrease}) {constB=-1}
  n10 = Subtract(A=@n3, B=WorldPosition().XYZ)
  Add(
   A=MaterialFunctionCall(WidthBase=Multiply(A=@n1.SplineWidth, B=@n7), WidthTip=Multiply(A=@n1.SplineWidth, B=@n8)) {materialFunction=&SplineThicken}.WorldPositionOffset
   B=LinearInterpolate(B=Multiply(A=@n9, B=@n10) {constB=-1}, Alpha=@n6) {constB=0}
  )
  n11 = DynamicParameter() {defaultValue=<rgba 100 1 1 1>, paramNames [SplineWidth,HueShift,Desat,PulseSpeed]}
  n12 = ParticleColor()
  MaterialFunctionCall(Emissive=Desaturation(
    None=MaterialFunctionCall(Hue Shift Percentage=@n11.HueShift, Texture=@n12.RGB) {materialFunction=&HueShift}.Result
    Fraction=DynamicParameter() {defaultValue=<rgba 100 1 1 1>, paramNames [SplineWidth,HueShift,Desat,PulseSpeed]}.Desat
   )) {materialFunction=&MF_FX_EmissiveAdjustments}
  n13 = ComponentMask(None=TextureCoordinate()) {r=true}
  n14 = ComponentMask(None=TextureCoordinate()) {g=true}
  n15 = OneMinus(None=@n14)
  n16 = Add(A=Multiply(A=@n15, B=@n14), B=@n15)
  n17 = Multiply(A=Multiply(A=OneMinus(None=@n13), B=@n13), B=Multiply(A=Power(Base=@n16) {constExponent=4}, B=@n14))
  n18 = DepthFade(Opacity=@n17, FadeDistance=ScalarParameter() {defaultValue=300, parameterName=DepthFadeDistance}) {fadeDistanceDefault=200}
  n19 = ShadingPathSwitch(Default=@n18, Deferred=@n18, Mobile=@n17) {desc=Migrated from FeatureLevelSwitch}
  Multiply(
   A=Multiply(A=Multiply(A=@n12.A, B=@n19), B=@n2)
   B=MaterialFunctionCall(Distance=ScalarParameter() {defaultValue=25000, parameterName=NearFade_Distance}) {materialFunction=&Near_Camera_Fade}.Result
  )
 ]
}
