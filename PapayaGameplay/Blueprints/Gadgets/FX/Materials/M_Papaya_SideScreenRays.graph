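# Additive, two-sided unlit material for light rays from the side screens of Papaya gadgets.
# Opacity comes from UV0 gradients that soften the sides and ends, scaled by an Alpha parameter and an inverted CameraDepthFade that fades rays out with distance.
# Color comes from MF_Papaya_GadgetColor, multiplied by noise sampled along only the U axis of UV1 to form streaks.
# A time-driven HueShift on that noise, read through its red channel, animates the streaks' brightness.
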Material /PapayaGameplay/Blueprints/Gadgets/FX/Materials/M_Papaya_SideScreenRays
graph {
 properties {
  bCastRayTracedShadows=false
  blendMode=BLEND_Additive
  pixelDepthOffsetMode=PDOM_Legacy
  shadingModel=MSM_Unlit
  twoSided=true
  usedShadingModels=Unlit
 }
 nodes [
  n1 = MaterialFunctionCall(Float2=TextureCoordinate()) {materialFunction=&BreakOutFloat2Components}
  n2 = Multiply(A=@n1.R, B=OneMinus(None=@n1.R))
  n3 = MaterialFunctionCall(Float4=Constant4Vector() {constant=<rgba 0 0.1 0 1>}.RGBA) {materialFunction=&BreakOutFloat4Components}
  n4 = Multiply(
   A=Power(Base=OneMinus(None=@n1.G)) {constExponent=5}
   B=Saturate(None=MaterialFunctionCall(Input=@n1.G, Input Low=@n3.R, Input High=@n3.R, Target Low=@n3.R, Target High=@n3.R) {materialFunction=&RemapValueRange}.Result)
  )
  n5 = MaterialFunctionCall(
   Fade Length=ScalarParameter() {defaultValue=4096, parameterName=Fade Length}
   Fade Offset=ScalarParameter() {defaultValue=7500, parameterName=Fade Offset}
  ) {materialFunction=&CameraDepthFade}
  Multiply(A=Multiply(A=Multiply(A=@n2, B=@n4), B=OneMinus(None=@n5.Result)), B=ScalarParameter() {defaultValue=1, parameterName=Alpha})
  n6 = Time()
  n7 = Constant() {r=0.5}
  n8 = MaterialFunctionCall(Float2=Multiply(A=TextureCoordinate() {coordinateIndex=1}, B=Constant() {r=5})) {materialFunction=&BreakOutFloat2Components}
  n9 = Constant()
  n10 = MaterialFunctionCall(
   Hue Shift Percentage=Frac(None=Multiply(A=@n6, B=@n7))
   Texture=TextureSample(UVs=MaterialFunctionCall(X=@n8.R, Y=@n9) {materialFunction=&MakeFloat2}.Result) {samplerType=SAMPLERTYPE_LinearColor, texture=&HighContrastNoise}.RGB
  ) {materialFunction=&HueShift}
  Multiply(A=Multiply(
    A=MaterialFunctionCall!(
     Media Player TexCoord
     Media Color Code Coordinate
     Media Texture Resolution
     Base Color Parameter
     Base Color Mix
     Media Color Code Mix
     Intermission Color Parameter
     Intermission Color Mix
     Desaturation
    ) {materialFunction=/Game/Environments/Papaya/Materials/MaterialFunctions/MF_Papaya_GadgetColor}.Result
    B=MaterialFunctionCall(Float3=@n10.Result) {materialFunction=&BreakOutFloat3Components}.R
   )) {constB=5}
 ]
}
