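# Additive, unlit, two-sided Niagara sprite material for a dome effect in Papaya gadget FX, with color, alpha and offset from MF_Papaya_GlobeColor.
# World position offset shrinks sprites toward their particle center, using a dynamic-parameter Dot and a camera-alignment term built from NX alone.
# A camera-versus-normal test hue-shifts color on away-facing sides before emissive adjustment.
# Alpha is multiplied by a lens mask and opacity scale, then depth-faded.
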Material /PapayaGameplay/Blueprints/Gadgets/FX/Materials/M_Papaya_Sprite_Dome
graph {
 properties {
  bUsedWithNiagaraSprites=true
  blendMode=BLEND_Additive
  pixelDepthOffsetMode=PDOM_Legacy
  shadingModel=MSM_Unlit
  twoSided=true
  usedShadingModels=Unlit
 }
 nodes [
  n1 = MaterialFunctionCall() {materialFunction=/PapayaGameplay/Blueprints/Gadgets/FX/Materials/MaterialFunctions/MF_Papaya_GlobeColor}
  n2 = ParticlePositionWS()
  n3 = WorldPosition()
  n4 = DynamicParameter() {parameterIndex=1, paramNames [NX,NY,NZ,Dot]}
  n5 = Power(Base=Saturate(None=Abs(None=@n4.Dot)))
  n6 = Constant() {r=1}
  n7 = DotProduct(
   A=Transform(None=MaterialFunctionCall(X=@n4.NX, Y=@n4.NX, Z=@n4.NX) {materialFunction=&MakeFloat3}.Result) {transformSourceType=TRANSFORMSOURCE_Local}
   B=Transform(None=Constant3Vector() {constant=<rgba 0 0 -1 0>}) {transformSourceType=TRANSFORMSOURCE_Camera}
  )
  n8 = Power(Base=Saturate(None=Abs(None=@n7))) {constExponent=0.8}
  Add(A=@n1.WPO, B=LinearInterpolate(A=Subtract(A=@n2, B=@n3.XYZ), B=Constant3Vector(), Alpha=LinearInterpolate(A=@n5, B=@n6, Alpha=@n8)))
  n9 = TextureSample() {samplerType=SAMPLERTYPE_LinearGrayscale, texture=/PapayaGameplay/Blueprints/Gadgets/FX/Textures/LensMask_512}
  LinearInterpolate(
   B=Multiply(A=Multiply(A=@n1.Alpha, B=@n9.R), B=ScalarParameter() {defaultValue=1, parameterName=Opacity Scale})
   Alpha=DepthFade() {fadeDistanceDefault=500}
  )
  n10 = If(
   A=DotProduct(A=CameraVectorWS(), B=VertexNormalWS())
   "A > B"=@n1.RGB
   "A == B"=@n1.RGB
   "A < B"=MaterialFunctionCall(Hue Shift Percentage=Constant() {r=0.1}, Texture=@n1.RGB) {materialFunction=&HueShift}.Result
  )
  MaterialFunctionCall(Emissive=@n10) {materialFunction=&MF_FX_EmissiveAdjustments}
  Multiply(A=@n1.RGB) {constB=0.25}
  Reroute(None=@n10)
  n11 = Constant() {r=0.5}
  SphereMask(A=TextureCoordinate(), B=@n11, Radius=@n11, Hardness=Constant() {r=0.8})
 ]
}
