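# Additive unlit material for Papaya gadget spline meshes, drawing glowing, audio-reactive swirly ribbons.
# A scrolling RGB mask multiplies wave and noise textures, tinted by MF_Papaya_GadgetColor with a height-based hue shift and HDR boost.
# End, edge and view-angle fades shape the glow, with angle and depth fades skipped on cheaper paths.
# SplineThicken widens ribbons by camera distance and EventAudio_Amplitude, which also drives scroll and offset, while sine waves add camera-perpendicular wobble.
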Material /PapayaGameplay/Blueprints/Gadgets/FX/Materials/M_Papaya_Swirly_SplineMesh
graph {
 properties {blendMode=BLEND_Additive, pixelDepthOffsetMode=PDOM_Legacy, shadingModel=MSM_Unlit, twoSided=true, usedShadingModels=Unlit}
 nodes [
  n1 = MaterialFunctionCall(Float2=VectorParameter() {defaultValue=<rgba -1 1 0 1>}.RGB) {materialFunction=&BreakOutFloat2Components}
  n2 = Time()
  n3 = ScalarParameter() {parameterName=Speed Multiplier}
  n4 = CollectionParameter() {collection=&FortniteMaterialParameters, parameterName=EventAudio_Amplitude}
  n5 = ScalarParameter() {defaultValue=0.05, parameterName=MPC Speed Multiplier}
  n6 = Frac(None=Add(A=Multiply(A=@n2, B=@n3) {constB=0.05}, B=Multiply(A=@n4, B=@n5) {constB=0.05}))
  n7 = Add(
   A=Multiply(A=TextureCoordinate()) {constB=0.5}
   B=MaterialFunctionCall(R=Constant(), G=LinearInterpolate(A=@n1.R, B=@n1.R, Alpha=@n6) {constA=-1}) {materialFunction=&AppendMany}.RG
  )
  n8 = TextureSample(UVs=Rotator(Coordinate=@n7, Time=Constant() {r=6.2832})) {texture=/Game/Global/Textures/Default/Masks_and_Gradients/3pxRGBImage}
  n9 = Multiply(A=@n7, B=Constant2Vector() {g=0.45, r=0.25})
  n10 = Add(A=Multiply(A=@n7, B=Constant2Vector() {g=0.7, r=0.5}), B=Constant2Vector() {g=0.05}) {constB=0.05}
  n11 = Multiply(A=@n7, B=Constant2Vector() {g=0.2, r=0.2})
  n12 = Constant2Vector() {g=0.05}
  n13 = LinearInterpolate(
   A=Add(
    A=TextureSample(UVs=@n9) {texture=/Game/Athena/Prototype/Blueprints/Sneezy/Textures/T_Sneezy_MagneticWaves}.RGB
    B=TextureSample(UVs=@n10) {texture=&T_Sneezy_MagneticWaves}.RGB
   )
   B=TextureSample(UVs=Add(A=@n11, B=@n12) {constB=0.05}) {samplerType=SAMPLERTYPE_LinearColor, texture=&HighContrastNoise}.RGB
   Alpha=ScalarParameter() {parameterName=Mix Glow}
  )
  n14 = Divide(A=Subtract(A=WorldPosition().XYZ, B=ObjectPositionWS()), B=ObjectRadius())
  n15 = LinearInterpolate(
   A=ScalarParameter() {parameterName=Gradient Hue Shift Amount}
   Alpha=Power(Base=ComponentMask(None=@n14) {g=true}, Exp=ScalarParameter() {defaultValue=1, parameterName=Gradient Hue Shift Power})
  ) {constB=0}
  n16 = MaterialFunctionCall!(
   Media Player TexCoord
   Media Color Code Coordinate
   Media Texture Resolution
   Base Color Parameter
   Base Color Mix
   Media Color Code Mix
   Intermission Color Parameter
   Intermission Color Mix
   Desaturation
  ) {materialFunction=/Game/Environments/Papaya/Materials/MaterialFunctions/MF_Papaya_GadgetColor}
  n17 = Multiply(
   A=Desaturation(None=Multiply(A=@n8.RGB, B=@n13), Fraction=ScalarParameter() {defaultValue=1, parameterName=Desaturation})
   B=Multiply(
    A=MaterialFunctionCall(Hue Shift Percentage=@n15, Texture=@n16.Result) {materialFunction=&HueShift}.Result
    B=ScalarParameter() {defaultValue=50, parameterName=Color Multiplier}
   )
  )
  n18 = ComponentMask(None=StaticSwitchParameter(True=TextureCoordinate() {coordinateIndex=1}, False=TextureCoordinate()) {parameterName=UseSecondaryUvSetForLengthFade}) {g=true}
  n19 = Power(Base=Multiply(A=OneMinus(None=@n18), B=@n18), Exp=ScalarParameter() {defaultValue=4, parameterName=EndsFadePower}) {constExponent=4.5}
  n20 = ScalarParameter() {defaultValue=30, parameterName=EndsFadeMultiplier}
  n21 = ComponentMask(None=TextureCoordinate()) {r=true}
  n22 = Power(Base=Multiply(A=OneMinus(None=@n21), B=@n21)) {constExponent=3.25}
  n23 = CollectionParameter() {collection=&FortniteMaterialParameters, parameterName=EventAudio_Amplitude}
  n24 = LinearInterpolate(Alpha=@n23) {constA=1, constB=8}
  n25 = Multiply(A=Clamp(None=Multiply(A=@n19, B=@n20) {constB=24}), B=Clamp(None=Multiply(A=@n22, B=@n24) {constB=15}))
  n26 = CollectionParameter() {collection=&FortniteMaterialParameters, parameterName=EventAudio_Amplitude}
  n27 = Multiply(A=@n26, B=ScalarParameter() {defaultValue=5, parameterName=Audio Amplitude})
  n28 = LinearInterpolate(A=ScalarParameter() {parameterName=SplineThicknessBeatFar}, B=ScalarParameter() {parameterName=SplineThicknessBeatClose}, Alpha=@n27)
  n29 = CameraPositionWS()
  n30 = WorldPosition()
  n31 = Divide(
   A=Subtract(A=Distance(A=@n29, B=@n30.XYZ), B=ScalarParameter() {defaultValue=512, parameterName=Distance Add})
   B=ScalarParameter() {parameterName=Scale Distance Divisor}
  )
  n32 = LinearInterpolate(
   A=Multiply(A=ScalarParameter() {defaultValue=350, parameterName=SplineWidthLow}, B=@n28)
   B=Multiply(A=ScalarParameter() {defaultValue=350, parameterName=SplineWidthHigh}, B=@n28)
   Alpha=OneMinus(None=Clamp(None=@n31))
  )
  n33 = MaterialFunctionCall(WidthBase=@n32, WidthTip=@n32) {materialFunction=&SplineThicken}
  n34 = Power(Base=DotProduct(A=@n33.Normal, B=CameraVectorWS()), Exp=ScalarParameter() {defaultValue=1, parameterName=Angle Fade Power})
  n35 = Multiply(A=@n25, B=Saturate(None=Multiply(A=@n34) {constB=10}))
  n36 = Multiply(
   A=MaterialFunctionCall(LDR Input=Saturate(None=@n17), HDR Tint=@n17, HDR Intensity Multiplier=ScalarParameter() {defaultValue=150, parameterName=HDR Intensity}) {materialFunction=&DeriveHDRfromLDR}.RemappedRange
   B=QualitySwitch(Default=ShadingPathSwitch(Default=@n35, Forward=@n25, Mobile=@n25), Low=@n25)
  )
  n37 = LinearInterpolate(B=@n36, Alpha=DepthFade(FadeDistance=Constant() {r=500}))
  Multiply(
   A=QualitySwitch(Default=ShadingPathSwitch(Default=@n37, Deferred=@n37, Forward=@n36, Mobile=@n36), Low=@n36)
   B=ScalarParameter() {defaultValue=1, parameterName=Alpha}
  )
  n38 = MaterialFunctionCall(Float2=TextureCoordinate()) {materialFunction=&BreakOutFloat2Components}
  n39 = Frac(None=Multiply(A=Time()) {constB=0.5})
  n40 = Add(A=Multiply(A=@n38.G) {constB=5}, B=@n39)
  n41 = Add(A=Multiply(A=@n38.G) {constB=6}, B=@n39)
  n42 = Multiply(
   A=CrossProduct(
    A=MaterialFunctionCall() {materialFunction=&CameraDirectionVector}.Result
    B=Constant3Vector() {constant=<rgba 0 0 1 0>}
   )
   B=Add(A=Sine(None=@n40), B=Sine(None=@n41))
  )
  n43 = ScalarParameter() {defaultValue=50, parameterName=WPO Gain}
  n44 = ScalarParameter() {defaultValue=1, parameterName=ScaleSmall}
  n45 = ActorPositionWS()
  n46 = WorldPosition()
  n47 = LinearInterpolate(A=@n44, B=LinearInterpolate(A=@n44, B=Subtract(A=@n45, B=@n46.XYZ)) {constAlpha=-0.1}, Alpha=@n27) {constA=10}
  Reroute(None=Add(A=Add(A=@n33.WorldPositionOffset, B=@n47), B=Multiply(A=@n42, B=@n43)))
 ]
}
