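# Two-sided material for the front of the Papaya tetra display gadget, showing media tinted and deformed by audio, usable on Niagara mesh particles.
# UV1 picks one of four fixed screen sectors or, per a collection switch, seeds noise-varied rotating, panning crops.
# Equalizer amplitude drives tint in the shared gadget color function, whose result fades by view angle.
# The same amplitude scales masked world offset along directions decoded from vertex color.
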Material /PapayaGameplay/Blueprints/Gadgets/FX/Materials/M_Papaya_TetraDisplay_Front
graph {
 properties {bUsedWithNiagaraMeshParticles=true, pixelDepthOffsetMode=PDOM_Legacy, twoSided=true, usedShadingModels=Default Lit}
 nodes [
  Constant3Vector() {constant=<rgba 0.02 0.02 0.02 0>}
  Constant() {r=1}
  Constant() {r=0.3}
  n1 = TextureCoordinate()
  n2 = Multiply(A=@n1, B=Constant2Vector() {g=0.3889, r=0.2188})
  n3 = Add(A=@n2, B=Constant2Vector() {r=0.7812})
  n4 = MaterialFunctionCall(Float2=TextureCoordinate() {coordinateIndex=1}) {materialFunction=&BreakOutFloat2Components}
  n5 = If(A=@n4.R, "A > B"=Add(A=@n2, B=Constant2Vector() {g=0.6111}), "A < B"=Add(A=@n2, B=Constant2Vector() {g=0.6111, r=0.7812})) {constB=0.26}
  n6 = Time()
  n7 = TextureSample(UVs=MaterialFunctionCall(X=@n4.R, Y=Constant() {r=0.5}) {materialFunction=&MakeFloat2}.Result) {samplerType=SAMPLERTYPE_LinearGrayscale, texture=/Game/Global/Textures/Default/Noise/EqualizerComponents_v2_Noise}
  n8 = Multiply(A=ScalarParameter() {defaultValue=0.02, parameterName=Rotation Speed}, B=@n7.R)
  n9 = Multiply(
   A=MaterialFunctionCall(UVs=@n1, Rotation Center=Constant2Vector() {g=0.5, r=0.5}, "Rotation Angle (0-1)"=Multiply(A=@n6, B=@n8)) {materialFunction=&CustomRotator}.Rotated Values
   B=Constant2Vector() {g=1, r=0.5625}
  )
  n10 = Constant2Vector() {r=0.2187}
  n11 = Time()
  n12 = Multiply(A=ScalarParameter() {defaultValue=0.05, parameterName=Panner Speed}, B=@n7.R)
  n13 = Multiply(
   A=Sine(None=Multiply(A=@n11, B=@n12))
   B=MaterialFunctionCall(Float2=@n10) {materialFunction=&BreakOutFloat2Components}.R
  )
  n14 = Constant()
  n15 = Constant2Vector() {g=0.5, r=0.5}
  n16 = Subtract(
   A=Add(
    A=Add(A=@n9, B=@n10)
    B=MaterialFunctionCall(X=@n13, Y=@n14) {materialFunction=&MakeFloat2}.Result
   )
   B=@n15
  )
  n17 = If(
   A=CollectionParameter() {collection=/Game/Environments/Papaya/Materials/MPC_PapayaGadgetControls, parameterName=ScreensUseQuadrants}
   "A > B"=If(A=@n4.R, "A > B"=@n2, "A < B"=If(A=@n4.R, "A > B"=@n3, "A < B"=@n5) {constB=0.51}) {constB=0.76}
   "A < B"=Add(A=Multiply(A=@n16) {constB=0.7072}, B=@n15)
  ) {constB=0.5}
  n18 = CollectionParameter() {
   collection=&FortniteMaterialParameters
   parameterName=PapayaBaseColorParameter
  }
  n19 = CollectionParameter() {
   collection=&FortniteMaterialParameters
   parameterName=PapayaBaseColorParameterOffset
  }
  n20 = OneMinus(None=@n4.G)
  n21 = MaterialFunctionCall!(
   Audio Lookup Param=@n20
   Average Amplitude Mix=Constant() {r=0.1}
   Ramp Texture
   Ramp Exponent
   Noise Texture
   Noise Scale
   Noise Gain
   Strip Noise Gain
   Grid Tiles X/Y
   Grid Speed
   Output Solid Bars
   TexCoord Override
  ) {materialFunction=/Game/Environments/Papaya/Materials/MaterialFunctions/MF_Papaya_Equalizer_Data}
  n22 = Multiply(A=Multiply(A=@n19) {constB=2}, B=@n21.EQ Raw Amplitude)
  n23 = Power(Base=@n21.EQ Raw Amplitude) {constExponent=0.5}
  n24 = CollectionParameter() {collection=&MPC_PapayaGadgetControls, parameterName=TetraScreenTintMix}
  n25 = MaterialFunctionCall() {materialFunction=&CameraDirectionVector}
  Multiply(
   A=MaterialFunctionCall!(
    Media Player TexCoord=@n17
    Media Color Code Coordinate
    Media Texture Resolution
    Base Color Parameter=Add(A=@n18, B=@n22)
    Base Color Mix=LinearInterpolate(B=@n23, Alpha=@n24)
    Media Color Code Mix
    Intermission Color Parameter
    Intermission Color Mix
    Desaturation
   ) {materialFunction=/Game/Environments/Papaya/Materials/MaterialFunctions/MF_Papaya_GadgetColor}.Result
   B=DotProduct(A=PixelNormalWS(), B=Multiply(A=@n25.Result) {constB=-1})
  )
  n26 = Multiply(A=VertexColor()) {constB=2}
  n27 = Multiply(A=Transform(None=Subtract(A=@n26)) {transformSourceType=TRANSFORMSOURCE_Local}, B=@n21.EQ Raw Amplitude)
  n28 = Constant() {r=1}
  n29 = Constant()
  n30 = Saturate(None=Multiply(A=@n20, B=If(A=@n4.R, "A > B"=@n28, "A < B"=@n29) {constB=0.0001}))
  Reroute(None=Multiply(A=Multiply(A=@n27, B=@n30), B=ScalarParameter() {defaultValue=300, parameterName=WPO Gain}))
  TextureSample(UVs=@n17) {texture=/PapayaGameplay/Blueprints/Gadgets/ForTestingOnly/papaya_tetra_screen_sectors}
 ]
}
