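# Two-sided material for the sides of the Papaya TetraDisplay gadget effect, usable on Niagara mesh particles, that reacts to audio like an equalizer.
# Inverted UV1 V indexes the shared equalizer function, whose amplitude scales a masked world position offset along directions decoded from vertex color.
# Amplitude also shifts the collection-driven gadget color and, with a center-ramp texture, its HDR intensity.
# A normal-to-camera dot product dims that color at grazing angles.
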Material /PapayaGameplay/Blueprints/Gadgets/FX/Materials/M_Papaya_TetraDisplay_Sides
graph {
 properties {bUsedWithNiagaraMeshParticles=true, pixelDepthOffsetMode=PDOM_Legacy, twoSided=true, usedShadingModels=Default Lit}
 nodes [
  Constant() {r=1}
  Constant() {r=0.3}
  n1 = VertexColor()
  n2 = Transform(None=Subtract(A=Multiply(A=@n1) {constB=2})) {transformSourceType=TRANSFORMSOURCE_Local}
  n3 = MaterialFunctionCall(Float2=TextureCoordinate() {coordinateIndex=1}) {materialFunction=&BreakOutFloat2Components}
  n4 = OneMinus(None=@n3.G)
  n5 = MaterialFunctionCall!(
   Audio Lookup Param=@n4
   Average Amplitude Mix=Constant() {r=0.1}
   Ramp Texture
   Ramp Exponent=Constant() {r=0.5}
   Noise Texture
   Noise Scale
   Noise Gain
   Strip Noise Gain
   Grid Tiles X/Y
   Grid Speed
   Output Solid Bars
   TexCoord Override
  ) {materialFunction=/Game/Environments/Papaya/Materials/MaterialFunctions/MF_Papaya_Equalizer_Data}
  n6 = Multiply(A=@n4, B=If(A=@n3.R, "A > B"=Constant() {r=1}, "A < B"=Constant()) {constB=0.0001})
  Multiply(A=Multiply(A=Multiply(A=@n2, B=@n5.EQ Raw Amplitude), B=Saturate(None=@n6)), B=ScalarParameter() {defaultValue=300, parameterName=WPO Gain})
  n7 = CollectionParameter() {
   collection=&FortniteMaterialParameters
   parameterName=PapayaBaseColorParameter
  }
  n8 = CollectionParameter() {
   collection=&FortniteMaterialParameters
   parameterName=PapayaBaseColorParameterOffset
  }
  n9 = Multiply(A=@n8) {constB=2}
  n10 = MaterialFunctionCall!(
   Media Player TexCoord
   Media Color Code Coordinate
   Media Texture Resolution
   Base Color Parameter=Add(A=@n7, B=Multiply(A=@n9, B=@n5.EQ Raw Amplitude))
   Base Color Mix
   Media Color Code Mix
   Intermission Color Parameter
   Intermission Color Mix
   Desaturation
  ) {materialFunction=/Game/Environments/Papaya/Materials/MaterialFunctions/MF_Papaya_GadgetColor}
  n11 = Multiply(A=Power(Base=@n5.EQ Raw Amplitude) {constExponent=1}, B=ScalarParameter() {defaultValue=10, parameterName=HDR Gain})
  n12 = TextureSample() {samplerType=SAMPLERTYPE_LinearGrayscale, texture=/PapayaGameplay/Blueprints/Gadgets/FX/Textures/EqualizerComponents_v2_CenterRamp}
  n13 = MaterialFunctionCall() {materialFunction=&CameraDirectionVector}
  Multiply(
   A=MaterialFunctionCall(LDR Input=Saturate(None=@n10.Result), HDR Intensity Multiplier=Multiply(A=@n11, B=@n12.R)) {materialFunction=&DeriveHDRfromLDR}.RemappedRange
   B=DotProduct(A=PixelNormalWS(), B=Multiply(A=@n13.Result) {constB=-1})
  )
 ]
}
