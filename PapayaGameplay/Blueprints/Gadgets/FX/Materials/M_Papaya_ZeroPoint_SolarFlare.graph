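# Translucent, unlit, two-sided mesh-particle material for the Papaya Zero Point solar flare effect.
# Panned magnetic-wave and caustic textures, offset by per-particle Random, form a pattern that hue-shifts the gadget color for emissive.
# The pattern also drives opacity through Erode/CenterPower masking, with view-angle, near-camera and depth fades plus optional time-of-day scaling.
# Dynamic Squeeze alters wave tiling and pinches the mesh through world position offset.
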Material /PapayaGameplay/Blueprints/Gadgets/FX/Materials/M_Papaya_ZeroPoint_SolarFlare
graph {
 properties {
  bUsedWithMeshParticles=true
  blendMode=BLEND_Translucent
  pixelDepthOffsetMode=PDOM_Legacy
  shadingModel=MSM_Unlit
  twoSided=true
  usedShadingModels=Unlit
 }
 nodes [
  n1 = ScalarParameter() {defaultValue=-256, parameterName=SqueezeWPOMult}
  n2 = DynamicParameter() {defaultValue=<rgba 1 1 12 0>, paramNames [Random,Erode,CenterPower,Squeeze]}
  n3 = ComponentMask(None=TextureCoordinate()) {r=true}
  n4 = AppendVector(A=Multiply(A=Subtract(A=@n3) {constB=0.5}) {constB=2}, B=Constant2Vector())
  LinearInterpolate(
   A=Constant()
   B=Multiply(A=Multiply(A=@n2.Squeeze, B=@n1), B=Transform(None=@n4))
   Alpha=ScalarParameter() {defaultValue=1, parameterName=Use WPO}
  )
  n5 = ParticleColor()
  n6 = VectorParameter() {parameterName=Tile1, defaultValue=<rgba 0.4 0.5 0 0>}
  n7 = AppendVector(A=@n6.R, B=@n6.R)
  n8 = VectorParameter() {parameterName=Tile1SqueezeMult, defaultValue=<rgba 0.3 2.5 0 1>}
  n9 = LinearInterpolate(A=@n7, B=Multiply(A=@n7, B=AppendVector(A=@n8.R, B=@n8.R)), Alpha=@n2.Squeeze) {constA=0.8, constB=2}
  n10 = Multiply(A=Add(A=TextureCoordinate() {coordinateIndex=1}, B=Multiply(B=@n9) {constA=-0.2, constB=0.5}), B=@n9)
  n11 = VectorParameter() {parameterName=PanDirection1, defaultValue=<rgba 0.2 0.02 0 0>}
  n12 = AppendVector(A=@n11.R, B=@n11.R)
  n13 = Add(A=Multiply(A=Time(), B=ScalarParameter() {defaultValue=0.5, parameterName=PanSpeed}), B=@n2.Random)
  n14 = Panner(Coordinate=Multiply(A=@n10) {constB=1.5}, Time=@n13, Speed=@n12) {speedX=-0.2, speedY=-0.02}
  n15 = VectorParameter() {parameterName=PanDirection1Mult, defaultValue=<rgba 1 -1 0 1>}
  n16 = Panner(Coordinate=@n10, Time=@n13, Speed=Multiply(A=@n12, B=AppendVector(A=@n15.R, B=@n15.R))) {speedX=-0.2, speedY=0.02}
  n17 = Constant3Vector() {constant=<rgba 1 0 0.4 0>}
  n18 = VectorParameter() {parameterName=Tile2, defaultValue=<rgba 0.3 0.5 0 0>}
  n19 = VectorParameter() {parameterName=PanDirection2, defaultValue=<rgba 0.1 0.04 0 0>}
  n20 = Panner(Coordinate=Multiply(A=TextureCoordinate(), B=AppendVector(A=@n18.R, B=@n18.R)), Time=@n13, Speed=AppendVector(A=@n19.R, B=@n19.R)) {speedX=0.1, speedY=0.04}
  n21 = TextureSample(UVs=Add(A=@n20) {constB=0.01}) {texture=/Game/Global/Textures/Default/Caustics/T_Caustics_Softer}
  n22 = Constant3Vector() {constant=<rgba 0 1 0.6 0>}
  n23 = TextureSample(UVs=@n20) {texture=&T_Caustics_Softer}
  n24 = Add(
   A=Add(
    A=TextureSample(UVs=@n14) {texture=/Game/Athena/Prototype/Blueprints/Sneezy/Textures/T_Sneezy_MagneticWaves}.RGB
    B=TextureSample(UVs=@n16) {texture=&T_Sneezy_MagneticWaves}.RGB
   )
   B=LinearInterpolate(A=Multiply(A=@n17, B=@n21.G), B=Multiply(A=@n22, B=@n23.G))
  )
  n25 = Constant() {r=0.5}
  n26 = Power(Base=SphereMask(A=TextureCoordinate(), B=@n25, Radius=@n25) {hardnessPercent=0})
  n27 = Multiply(A=VectorParameter() {parameterName=Color, defaultValue=<rgba 0.0771 0.2812 0.4 0>}.RGB, B=@n26)
  n28 = Add(A=@n24, B=@n27)
  n29 = Multiply(A=@n5.RGB, B=@n28)
  n30 = Add(A=Multiply(A=@n5.RGB, B=@n24), B=@n27)
  n31 = Multiply(A=Desaturation(None=StaticSwitchParameter(True=@n29, False=@n30) {defaultValue=true, parameterName=ParticleColorAffectsBackground}, Fraction=Constant())) {constB=0.1}
  n32 = MaterialFunctionCall!(
   Media Player TexCoord
   Media Color Code Coordinate
   Media Texture Resolution
   Base Color Parameter
   Base Color Mix
   Media Color Code Mix
   Intermission Color Parameter
   Intermission Color Mix
   Desaturation
  ) {materialFunction=/Game/Environments/Papaya/Materials/MaterialFunctions/MF_Papaya_GadgetColor}
  MaterialFunctionCall(Emissive=Multiply(
    A=MaterialFunctionCall(Hue Shift Percentage=@n31, Texture=@n32.Result) {materialFunction=&HueShift}.Result
    B=ScalarParameter() {defaultValue=1.2, parameterName=Emission Color Gain}
   )) {materialFunction=&MF_FX_EmissiveAdjustments}
  n33 = SphereMask(
   A=ComponentMask(None=TextureCoordinate()) {g=true}
   B=ScalarParameter() {defaultValue=0.5, parameterName=ErosionCenter}
   Radius=LinearInterpolate(B=ScalarParameter() {defaultValue=0.25, parameterName=ErosionMaxRadius}, Alpha=Saturate(None=@n2.Erode)) {constB=0.25}
  ) {attenuationRadius=0.25, hardnessPercent=0}
  n34 = Multiply(A=Power(Base=@n28, Exp=LinearInterpolate(B=@n2.CenterPower, Alpha=@n33) {constA=0.55, constB=5})) {constB=2}
  n35 = MaterialFunctionCall("FadeSensitivity (optional)"=ScalarParameter() {defaultValue=1, parameterName=AngleFade}) {materialFunction=&FadeBasedOnViewAngle}
  n36 = MaterialFunctionCall(Distance=ScalarParameter() {defaultValue=800, parameterName=NearFadeDIst}) {materialFunction=&Near_Camera_Fade}
  n37 = ScalarParameter() {defaultValue=200, parameterName=DepthFade}
  n38 = Multiply(A=Saturate(None=Multiply(A=@n35.Result) {constB=2}), B=Multiply(A=@n36.Result, B=DepthFade(FadeDistance=@n37)))
  n39 = Multiply(
   A=Multiply(A=@n5.A, B=Multiply(A=@n34, B=@n26))
   B=ShadingPathSwitch(Default=QualitySwitch(Default=@n38, Low=@n36.Result), Mobile=Constant() {r=0.7}) {desc=Migrated from FeatureLevelSwitch}
  )
  n40 = Multiply(A=@n39) {constB=0.3}
  StaticSwitchParameter(
   True=MaterialFunctionCall(Color_Morning=@n40, Color_Day=@n39, Color_Evening=@n40, Color_Night=@n40) {materialFunction=&Color_Time_of_Day}.Result
   False=@n39
  ) {parameterName=Use color time of day}
 ]
}
