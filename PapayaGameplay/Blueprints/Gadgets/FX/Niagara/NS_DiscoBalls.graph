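# NS_DiscoBalls is a Niagara gadget effect that burst-spawns clusters of orbiting, spinning disco balls, each drawn by sprite and mesh renderers.
# UniqueID divided by cluster size gives each particle a cluster index and orbit phase, with the index seeding a shared orbit rate.
# Cluster pivots circle the Z axis with emitter age, while particles orbit them at scale-based radii with random height offsets.
# User parameters set count, clusters, scale and pivot distance.
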NiagaraSystem /PapayaGameplay/Blueprints/Gadgets/FX/Niagara/NS_DiscoBalls
dependencies {usedDataInterfaces [], usedDynamicInputs [], usedModules [&SystemState]}
emitters {empty {
  emitterUpdateScript [
   &EmitterState(
    Inactive Response=::NewEnumerator1
    Life Cycle Mode=::NewEnumerator1
    Loop Behavior=::NewEnumerator1
    Loop Duration Mode=::NewEnumerator1
    UseLoopDelay=false
   )
   &SpawnBurst_Instantaneous(Spawn Count=&MakeIntFromFloat(Float=$User.BurstAmount), Use Spawn Probability=false)
  ]
  particleSpawnScript [
   /Niagara/Modules/Spawn/Initialization/InitializeParticle(
    Mesh Scale=&VectorFromFloat(Value=&UniformRangedFloat(Minimum=$User.ScaleMin, Maximum=$User.ScaleMax))
    Module.Write Color=false
    Module.Write Lifetime=false
    Module.Write Mass=true
    Module.Write Material Random=true
    Module.Write Position=false
    Module.Write Scale=true
    Module.Write SpriteRotation=true
    Module.Write SpriteSize=false
    Position=&SimulationPosition()
   )
   /SetVariables(Particles.SpriteSize=&Vector2DFromFloat(Value=&Multiply_Float(A=&MakeFloatFromVector(Vector=$Particles.Scale, Channel=::NewEnumerator0), B=250))) {isSetParameters=true}
   /SetVariables(
    Particles.U:NiagaraFloat=&Frac_Float(Float=&Divide_Float(
      A=HlslExpression() {hlslExpression=Particles.UniqueID}
      B=&Divide_Float(A=$User.BurstAmount, B=$User.ClusterCount)
     ))
    Particles.V:NiagaraFloat=&Ceil_Float(Float=&Divide_Float(
      A=HlslExpression() {hlslExpression=Particles.UniqueID}
      B=&Divide_Float(A=$User.BurstAmount, B=$User.ClusterCount)
     ))
   ) {isSetParameters=true}
   /SetVariables(Particles.OrbitRate:NiagaraFloat=&UniformRangedFloat(
     Minimum=-0.2
     Maximum=0.01
     Randomness Mode=::NewEnumerator1
     Random Seed=&MakeIntFromFloat(Float=$Particles.V)
     Fixed Random Seed=true
    )) {isSetParameters=true}
   /SetVariables(Particles.PosOffset:Vector3f=&MakeVector(X=0, Y=0, Z=&UniformRangedFloat(Minimum=-100, Maximum=100))) {isSetParameters=true}
   &AddRotationalVelocity(Rotation Rate=&UniformRangedVector(Minimum=<0 0 -0.1>, Maximum=<0 0 0.1>))
   &ApplyInitialForces()
  ]
  particleUpdateScript [
   &ParticleState(Kill Particles When Lifetime Has Elapsed=false, Let Infinitely Lived Particles Die When Emitter Deactivates=true)
   /SetVariables(Particles.OrbitPivot:Vector3f=&self:RotatePositionAroundZ(
     Angle:NiagaraFloat=&Add_Float(
      A=&Multiply_Float(A=360, B=&Divide_Float(A=$Particles.Initial.V, B=$User.ClusterCount))
      B=&Multiply_Float(A=$Emitter.Age, B=-5)
     )
     Position:Vector3f=&MakeVector(X=&Multiply_Float(A=$User.PivotOffsetScale, B=400), Y=0, Z=0)
    )) {isSetParameters=true}
   &RotateAroundPoint(
    Radius:NiagaraFloat=&Multiply_Float(A=&MakeFloatFromVector(Vector=$Particles.Scale, Channel=::NewEnumerator0), B=45)
    Rotation Center:Vector3f=$Particles.OrbitPivot
    Rotation Phase:NiagaraFloat=$Particles.Initial.U
    Rotation Rate:NiagaraFloat=$Particles.OrbitRate
   )
   /SetVariables(Particles.Position=&Add_Vector(A=Linked() {linkedVariable {name=Particles.Position, type=NiagaraPosition}}, B=$Particles.PosOffset)) {isSetParameters=true}
   &SolveForcesAndVelocity(Clamp Velocity=false, Limit Acceleration=false)
   &SolveRotationalForcesAndVelocity()
   &DynamicMaterialParameters(
    Index 0 Param 1=$Particles.MaterialRandom
    Module.Param0WriteEnabled=true
    Module.Param1WriteEnabled=true
    Module.Param2WriteEnabled=true
    Module.Param3WriteEnabled=true
   )
  ]
  rendererClasses [&NiagaraSpriteRendererProperties, &NiagaraMeshRendererProperties]
  renderers [
   {rendererClass=&NiagaraSpriteRendererProperties}
   {rendererClass=&NiagaraMeshRendererProperties, rendererIndex=1}
  ]
 }}
summary {
 systemName=NS_DiscoBalls
 emitters [{emitterName=empty, rendererClasses [&NiagaraSpriteRendererProperties, &NiagaraMeshRendererProperties]}]
 userVariables [
  User.BurstAmount {type=NiagaraFloat, defaultValue=15, description=""}
  User.ClusterCount {type=NiagaraFloat, defaultValue=5, description=""}
  User.PivotOffsetScale {type=NiagaraFloat, defaultValue=1, description=""}
  User.ScaleMax {type=NiagaraFloat, defaultValue=1, description=""}
  User.ScaleMin {type=NiagaraFloat, defaultValue=3, description=""}
 ]
}
