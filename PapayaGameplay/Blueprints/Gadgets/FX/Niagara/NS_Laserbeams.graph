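# Niagara system for a music-reactive laser beam gadget effect, driven by percussion values from the Papaya_Jukebox_NPC parameter collection.
# FN_Center_Mesh holds one persistent downward-oriented mesh.
# FN_Center_Sprite probabilistically bursts short-lived, cone-oriented beam meshes with layered sprites and a light, colored randomly or by User.LaserColor via UseCustomColor.
# Percussion sampled at spawn sets beam scale, while live percussion drives rotation in a random direction.
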NiagaraSystem /PapayaGameplay/Blueprints/Gadgets/FX/Niagara/NS_Laserbeams
dependencies {
 usedDataInterfaces [&NiagaraDataInterfaceCurve]
 usedDynamicInputs []
 usedModules [
  &self:SystemSpawnScript.SystemScriptSource.SystemScriptGraph.NiagaraNodeAssignment_0.SetVariables_#1
  &SystemState
 ]
}
emitters {
 FN_Center_Mesh {
  emitterUpdateScript [
   &EmitterState(Life Cycle Mode=::NewEnumerator1, Loop Duration=0, UseLoopDelay=false)
   &SpawnBurst_Instantaneous(Spawn Count=1, Spawn Probability=HlslExpression() {hlslExpression="Emitter.LoopCount == 0 ? 1.f : 0.f"}, Use Spawn Probability=true)
  ]
  particleSpawnScript [
   /Niagara/Modules/Spawn/Initialization/InitializeParticle(
    Module.Write Color=true
    Module.Write Lifetime=true
    Module.Write Mass=true
    Module.Write Material Random=true
    Module.Write Position=true
    Module.Write Scale=true
    Module.Write SpriteRotation=false
    Module.Write SpriteSize=false
    Position=&SimulationPosition()
    Sprite Size=<25 25>
   )
   /Niagara/Modules/Spawn/Orientation/V2/InitialMeshOrientation(Orientation Axis:Vector3f=<0 0 -1>, Use Orientation Vector=true, Use Rotation Vector=false)
  ]
  particleUpdateScript [
   &ParticleState(Kill Particles When Lifetime Has Elapsed=false)
   &ScaleColor(Scale RGB=&MakeVectorFromLinearColorRGB(Color=<rgba 1 1 1 1>), ScaleA=true, ScaleRGB=true)
  ]
  rendererClasses [&NiagaraMeshRendererProperties]
  renderers [{rendererClass=&NiagaraMeshRendererProperties, rendererIndex=0}]
 }
 FN_Center_Sprite {
  emitterUpdateScript [
   &EmitterState(Inactive Response=::NewEnumerator1, Life Cycle Mode=::NewEnumerator1, Loop Duration=0.5, UseLoopDelay=false)
   &SpawnBurst_Instantaneous(Spawn Count=1, Spawn Probability=$System.SpawnProbability, Use Spawn Probability=true)
  ]
  particleSpawnScript [
   &InitializeParticle(
    Color=&LerpLinearColors(
     StartColor=&UniformRangedLinearColor(Minimum=<rgba 0 0 1 1>, Maximum=<rgba 1 0 0 1>)
     EndColor=$User.LaserColor
     LerpFactor=$User.UseCustomColor
    )
    Lifetime=0.5
    Mesh Scale=&UniformRangedVector(Minimum=<1 1 1>, Maximum=<2 2 2>)
    Module.Write Color=true
    Module.Write Lifetime=true
    Module.Write Mass=false
    Module.Write Material Random=true
    Module.Write Position=true
    Module.Write Scale=true
    Module.Write SpriteRotation=false
    Module.Write SpriteSize=true
    Position=&Add_Vector(A=/Niagara/DynamicInputs/Helpers/SimulationPosition(), B=<0 0 23>)
    Sprite Size=&Vector2DFromFloat(Value=768)
   )
   /SetVariables(Particles.RandomNegOneToOne:NiagaraFloat=&UniformAOrBFloat(A=-1)) {isSetParameters=true}
   /SetVariables(Particles.SpriteSize2=&Multiply_Vector2(A=$Particles.SpriteSize, B=&Vector2DFromFloat(Value=0.5))) {isSetParameters=true}
   /SetVariables(Particles.Color2=&MultiplyLinearColor(Linear Color=$Particles.Color, Scale Factor=<1 1 1 1>)) {isSetParameters=true}
   &InitialMeshOrientation(
    Orientation Axis:Vector3f=/Game/Effects/Niagara/DynamicInputs/Random/FN_RandomVectorInCone(
     Cone Angle:NiagaraFloat=45
     Cone Axis:Vector3f=<0 0 1>
     Cone Axis Coordinate Space:ENiagaraCoordinateSpace=::Local
     Velocity Distribution Along Cone Axis:NiagaraFloat=0.5
    )
    Rotation=&UniformRangedVector(Minimum=<0 0 0>, Maximum=<1 0 0>)
    Use Orientation Vector=true
    Use Rotation Vector=true
   )
   /SetVariables(Particles.PapayaDriver:NiagaraFloat=&MakeFloatFromVector4(Vector 4=$NPC.Papaya_Jukebox_NPC.Percussion, Channel=::NewEnumerator2)) {isSetParameters=true}
   /SetVariables(Particles.LightPosition:Vector3f=&Add_Vector(A=Linked() {linkedVariable {name=Particles.Position, type=NiagaraPosition}}, B=<0 0 25>)) {isSetParameters=true}
   /SetVariables(Particles.LightExponent=2.25) {isSetParameters=true}
   /SetVariables(Particles.LightColor=&MultiplyLinearColor(Linear Color=$Particles.Color, Scale Factor=<10 10 10 1>)) {isSetParameters=true}
   /SetVariables(Particles.FaceUpwardVector:Vector3f=<0 0 1>) {isSetParameters=true}
   /SetVariables(Particles.SpriteSize3:Vector2f=&Multiply_Vector2(A=$Particles.SpriteSize, B=&Vector2DFromFloat(Value=0.0695))) {isSetParameters=true}
   /SetVariables(Particles.Color3:LinearColor=&MultiplyLinearColor(Linear Color=$Particles.Color, Scale Factor=&MakeVector4_FromFloat(Float=100))) {isSetParameters=true}
   /SetVariables(Particles.CameraOffsetLargeglow=40) {isSetParameters=true}
  ]
  particleUpdateScript [
   &ParticleState()
   &ScaleColor(
    Scale Alpha=&FloatFromCurve(FloatCurve=DataInterface() {propertyValues {Curve {keys [<at 0 0>,<at 0.25 1>,<at 0.8 1>,<at 1.0011 0.002>]}}})
    Scale RGB=&MakeVectorFromLinearColorRGB(Color=<rgba 1 1 1 1>)
    ScaleA=true
    ScaleRGB=true
   )
   &ScaleMeshSize(Scale Factor=&VectorFromFloat(Value=&FloatFromCurve(FloatCurve=DataInterface() {propertyValues {Curve {keys [
          {interpMode=RCIM_Cubic, leaveTangent=4.3456, tangentMode=RCTM_Break, time=-0.0046, value=0.5}
          {interpMode=RCIM_Cubic, leaveTangent=-0.1124, tangentMode=RCTM_Break, time=0.4138, value=1}
          {arriveTangent=-1.4315, interpMode=RCIM_Cubic, leaveTangent=-1.4315, tangentMode=RCTM_User, time=1.0011, value=0.5}
         ]}}})))
   /Niagara/Modules/Update/Orientation/V2/UpdateMeshOrientation(
    Rotation Rate=0.05
    Rotation Vector:Vector3f=&Multiply_VectorByFloat(
     Vector=<1 1 1>
     Float=&Multiply_Float(
      A=$Particles.RandomNegOneToOne
      B=&MakeFloatFromVector4(Vector 4=$NPC.Papaya_Jukebox_NPC.Percussion, Channel=::NewEnumerator0)
     )
    )
   )
   &ScaleMeshSize(Scale Factor=&MakeVector(X=$Particles.PapayaDriver, Y=$Particles.PapayaDriver, Z=$Particles.PapayaDriver)) {module=ScaleMeshSize001}
   /SetVariables(Particles.DynamicMeshParameters:Vector4f=&MakeVector4(X=$User.UseCustomColor, Y=0, Z=0, W=0)) {isSetParameters=true}
  ]
  rendererClasses [
   &NiagaraSpriteRendererProperties
   &NiagaraMeshRendererProperties
   &NiagaraLightRendererProperties
  ]
  renderers [
   {rendererClass=&NiagaraSpriteRendererProperties}
   {rendererClass=&NiagaraSpriteRendererProperties, rendererIndex=1}
   {rendererClass=&NiagaraMeshRendererProperties, rendererIndex=2}
   {rendererClass=&NiagaraLightRendererProperties, rendererIndex=3}
   {rendererClass=&NiagaraSpriteRendererProperties, rendererIndex=4}
  ]
 }
}
summary {
 systemName=NS_Laserbeams
 emitters [
  {
   emitterName=FN_Center_Sprite
   rendererClasses [
    &NiagaraSpriteRendererProperties
    &NiagaraMeshRendererProperties
    &NiagaraLightRendererProperties
   ]
  }
  {emitterName=FN_Center_Mesh, rendererClasses [&NiagaraMeshRendererProperties]}
 ]
 userVariables [
  User.LaserColor {type=LinearColor, description="", defaultValue {struct=LinearColor, value=<rgba 1 1 1 1>}}
  User.LaserEnv {type=NiagaraFloat, defaultValue=2, description=""}
  User.UseCustomColor {type=NiagaraFloat, defaultValue=0, description=""}
 ]
}
