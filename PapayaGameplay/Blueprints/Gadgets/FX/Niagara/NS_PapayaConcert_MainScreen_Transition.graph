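# Niagara effect for the Papaya concert main screen transition, scattering mesh fragments, sparklies and rift clouds across a wide screen-sized box.
# One-shot bursts throw fragments and GPU sparklies outward, alongside emitters that keep spawning both, with separate Concert fragment variants at higher counts.
# Looping fragments emit location events, likely feeding the short-lived glint sprites.
# Sparklies, clouds and concert glints take tint from User.BaseColorCurve, indexed by the Papaya_Jukebox_NPC parameter collection.
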NiagaraSystem /PapayaGameplay/Blueprints/Gadgets/FX/Niagara/NS_PapayaConcert_MainScreen_Transition
dependencies {
 usedDataInterfaces [&NiagaraDataInterfaceCurve, &NiagaraDataInterfaceVectorCurve]
 usedDynamicInputs []
 usedModules [&SystemState]
}
emitters {
 Fragment_Burst {
  emitterUpdateScript [
   &EmitterState(Life Cycle Mode=::NewEnumerator1, Loop Behavior=::NewEnumerator1, Loop Duration=1, UseLoopDelay=false)
   &SpawnBurst_Instantaneous(Spawn Count=6, Use Spawn Probability=false)
  ]
  particleSpawnScript [
   /Niagara/Modules/Spawn/Initialization/InitializeParticle(
    Lifetime=&UniformRangedFloat(Minimum=1.5, Maximum=3.5)
    Mesh Scale=&UniformRangedVector(Minimum=<2 2 2>, Maximum=<15 15 15>)
    Module.Write Color=true
    Module.Write Lifetime=true
    Module.Write Mass=true
    Module.Write Material Random=true
    Module.Write Position=false
    Module.Write Scale=true
    Module.Write SpriteRotation=false
    Module.Write SpriteSize=true
    Position=&SimulationPosition()
    Sprite Size=<5 5>
   )
   &BoxLocation(Box Offset=<0 -10 0>, Box Size=<2750 1 1500>, MaskSpawnGroup=false, Module.Override Seed=false, Rotate Box=false, Surface Only=false)
   /Niagara/Modules/Spawn/Location/SphereLocation(
    Hemisphere Y=true
    MaskSpawnGroup=false
    Module.Override Randomness=false
    Non Uniform Scale=<1 -1 1>
    Sphere Radius=&UniformRangedFloat(Minimum=10, Maximum=15)
    Surface Only=false
   )
   &AddVelocity(Velocity=&Add_Vector(
     A=&Multiply_Vector(A=$Output.BoxLocation.BoxVector, B=<0 100 0>)
     B=&Multiply_Vector(
      A=$Output.SphereLocation.SphereNormal
      B=&VectorFromFloat(Value=&UniformRangedFloat(Minimum=500, Maximum=500))
     )
    ))
   /Niagara/Modules/Spawn/Orientation/V2/InitialMeshOrientation(
    Rotation=&UniformRangedVector(Minimum=<0 0 0>, Maximum=<1 1 1>)
    Use Orientation Vector=true
    Use Rotation Vector=true
   )
  ]
  particleUpdateScript [
   &ParticleState()
   &AccelerationForce(Acceleration=<0 0 -100>)
   &Drag(Drag=&UniformRangedFloat(Minimum=0.5, Maximum=3), Use Linear Drag=true, Use Rotational Drag=false)
   &SolveForcesAndVelocity(Clamp Velocity=false, Limit Acceleration=false)
   /Niagara/Modules/Update/Orientation/V2/UpdateMeshOrientation(Rotation Vector:Vector3f=&UniformRangedVector(Minimum=<0 0 0>, Maximum=<1 1 1>))
   &ScaleMeshSize(Scale Factor=&VectorFromCurve(VectorCurve=DataInterface() {propertyValues {
       XCurve {keys [<at 0.2 1 RCIM_Cubic>, {arriveTangent=-1.25, interpMode=RCIM_Cubic, leaveTangent=-1.25, time=0.8, value=0.75}, <at 1 0 RCIM_Cubic>]}
       YCurve {keys [<at 0.2 1 RCIM_Cubic>, {arriveTangent=-0.25, interpMode=RCIM_Cubic, leaveTangent=-0.25, time=0.8, value=0.85}, <at 1 0.8 RCIM_Cubic>]}
       ZCurve {keys [<at 0.2 1 RCIM_Cubic>,<at 1 1 RCIM_Cubic>]}
      }}))
   &ScaleColor(
    Scale Alpha=&FloatFromCurve(FloatCurve=DataInterface() {propertyValues {Curve {keys [<at 0.7 1 RCIM_Cubic>,<at 1.0011 0.002>]}}})
    ScaleA=true
    ScaleRGB=true
   )
  ]
  rendererClasses [&NiagaraMeshRendererProperties]
  renderers [{rendererClass=&NiagaraMeshRendererProperties, rendererIndex=0}]
 }
 Fragment_Burst_Concert {
  emitterUpdateScript [
   &EmitterState(Life Cycle Mode=::NewEnumerator1, Loop Behavior=::NewEnumerator1, Loop Duration=1, UseLoopDelay=false)
   &SpawnBurst_Instantaneous(Spawn Count=24, Use Spawn Probability=false)
  ]
  particleSpawnScript [
   &InitializeParticle(
    Lifetime=&UniformRangedFloat(Minimum=1.5, Maximum=3.5)
    Mesh Scale=&UniformRangedVector(Minimum=<2 2 2>, Maximum=<15 15 15>)
    Module.Write Color=true
    Module.Write Lifetime=true
    Module.Write Mass=true
    Module.Write Material Random=true
    Module.Write Position=false
    Module.Write Scale=true
    Module.Write SpriteRotation=false
    Module.Write SpriteSize=true
    Position=&SimulationPosition()
    Sprite Size=<5 5>
   )
   &BoxLocation(Box Offset=<0 -10 0>, Box Size=<2750 1 1500>, MaskSpawnGroup=false, Module.Override Seed=false, Rotate Box=false, Surface Only=false)
   &SphereLocation(
    Hemisphere Y=true
    MaskSpawnGroup=false
    Module.Override Randomness=false
    Non Uniform Scale=<1 -1 1>
    Sphere Radius=&UniformRangedFloat(Minimum=10, Maximum=15)
    Surface Only=false
   )
   &AddVelocity(Velocity=&Add_Vector(
     A=&Multiply_Vector(A=$Output.BoxLocation.BoxVector, B=<0 100 0>)
     B=&Multiply_Vector(
      A=$Output.SphereLocation.SphereNormal
      B=&VectorFromFloat(Value=&UniformRangedFloat(Minimum=500, Maximum=500))
     )
    ))
   &InitialMeshOrientation(
    Rotation=&UniformRangedVector(Minimum=<0 0 0>, Maximum=<1 1 1>)
    Use Orientation Vector=true
    Use Rotation Vector=true
   )
  ]
  particleUpdateScript [
   &ParticleState()
   &AccelerationForce(Acceleration=<0 0 -100>)
   &Drag(Drag=&UniformRangedFloat(Minimum=0.5, Maximum=3), Use Linear Drag=true, Use Rotational Drag=false)
   &SolveForcesAndVelocity(Clamp Velocity=false, Limit Acceleration=false)
   &UpdateMeshOrientation(Rotation Vector:Vector3f=&UniformRangedVector(Minimum=<0 0 0>, Maximum=<1 1 1>))
   &ScaleMeshSize(Scale Factor=&VectorFromCurve(VectorCurve=DataInterface() {propertyValues {
       XCurve {keys [<at 0.2 1 RCIM_Cubic>, {arriveTangent=-1.25, interpMode=RCIM_Cubic, leaveTangent=-1.25, time=0.8, value=0.75}, <at 1 0 RCIM_Cubic>]}
       YCurve {keys [<at 0.2 1 RCIM_Cubic>, {arriveTangent=-0.25, interpMode=RCIM_Cubic, leaveTangent=-0.25, time=0.8, value=0.85}, <at 1 0.8 RCIM_Cubic>]}
       ZCurve {keys [<at 0.2 1 RCIM_Cubic>,<at 1 1 RCIM_Cubic>]}
      }}))
   &ScaleColor(
    Scale Alpha=&FloatFromCurve(FloatCurve=DataInterface() {propertyValues {Curve {keys [<at 0.7 1 RCIM_Cubic>,<at 1.0011 0.002>]}}})
    ScaleA=true
    ScaleRGB=true
   )
  ]
  rendererClasses [&NiagaraMeshRendererProperties]
  renderers [{rendererClass=&NiagaraMeshRendererProperties, rendererIndex=0}]
 }
 Fragment_Glints {
  emitterUpdateScript [&EmitterState(Life Cycle Mode=::NewEnumerator1, Loop Duration=1, UseLoopDelay=false)]
  particleSpawnScript [&InitializeParticle(
    Lifetime=&UniformRangedFloat(Minimum=0.05, Maximum=0.1)
    Module.Write Color=true
    Module.Write Lifetime=true
    Module.Write Mass=true
    Module.Write Material Random=true
    Module.Write Position=false
    Module.Write Scale=false
    Module.Write SpriteRotation=false
    Module.Write SpriteSize=true
    Position=&SimulationPosition()
    Sprite Size=&Vector2DFromFloat(Value=&UniformRangedFloat(Minimum=50, Maximum=300))
   )]
  particleUpdateScript [
   &ParticleState()
   &SolveForcesAndVelocity(Clamp Velocity=false, Limit Acceleration=false)
  ]
  rendererClasses [&NiagaraSpriteRendererProperties]
  renderers [{rendererClass=&NiagaraSpriteRendererProperties}]
 }
 Fragment_Glints_concert {
  emitterUpdateScript [&EmitterState(Life Cycle Mode=::NewEnumerator1, Loop Duration=1, UseLoopDelay=false)]
  particleSpawnScript [&InitializeParticle(
    Color=<rgba 10 1 10 1>
    Lifetime=&UniformRangedFloat(Minimum=0.05, Maximum=0.1)
    Module.Write Color=true
    Module.Write Lifetime=true
    Module.Write Mass=true
    Module.Write Material Random=true
    Module.Write Position=false
    Module.Write Scale=false
    Module.Write SpriteRotation=false
    Module.Write SpriteSize=true
    Position=&SimulationPosition()
    Sprite Size=&Vector2DFromFloat(Value=&UniformRangedFloat(Minimum=50, Maximum=300))
   )]
  particleUpdateScript [
   &ParticleState()
   &SolveForcesAndVelocity(Clamp Velocity=false, Limit Acceleration=false)
   &ScaleColor(
    Scale RGB=&MakeVectorFromLinearColorRGB(Color=&ColorFromCurve(ColorCurve=$User.BaseColorCurve, CurveIndex=&Frac_Float(Float=$NPC.Papaya_Jukebox_NPC.Base Color Parameter)))
    ScaleA=true
    ScaleRGB=true
   )
  ]
  rendererClasses [&NiagaraSpriteRendererProperties]
  renderers [{rendererClass=&NiagaraSpriteRendererProperties}]
 }
 Fragment_Looping {
  emitterUpdateScript [
   &EmitterState(Life Cycle Mode=::NewEnumerator1, Loop Duration=1, UseLoopDelay=false)
   &SpawnRate(SpawnRate=5, Use Spawn Probability=false)
  ]
  particleSpawnScript [
   &InitializeParticle(
    Lifetime=&UniformRangedFloat(Minimum=1.5, Maximum=3.5)
    Mesh Scale=&UniformRangedVector(Minimum=<2 2 2>, Maximum=<15 15 15>)
    Module.Write Color=true
    Module.Write Lifetime=true
    Module.Write Mass=true
    Module.Write Material Random=true
    Module.Write Position=false
    Module.Write Scale=true
    Module.Write SpriteRotation=false
    Module.Write SpriteSize=true
    Position=&SimulationPosition()
    Sprite Size=<5 5>
   )
   &BoxLocation(Box Offset=<0 -10 0>, Box Size=<2750 1 1500>, MaskSpawnGroup=false, Module.Override Seed=false, Rotate Box=false, Surface Only=false)
   &AddVelocity(Velocity=&Multiply_Vector(A=$Output.BoxLocation.BoxVector, B=<0 1 0>))
   &InitialMeshOrientation(
    Rotation=&UniformRangedVector(Minimum=<0 0 0>, Maximum=<1 1 1>)
    Use Orientation Vector=true
    Use Rotation Vector=true
   )
  ]
  particleUpdateScript [
   &ParticleState()
   &AccelerationForce(Acceleration=<0 0 -250>)
   &Drag(Drag=&UniformRangedFloat(Minimum=0, Maximum=3), Use Linear Drag=true, Use Rotational Drag=false)
   &SolveForcesAndVelocity(Clamp Velocity=false, Limit Acceleration=false)
   &UpdateMeshOrientation(Rotation Rate=0.5, Rotation Vector:Vector3f=&UniformRangedVector(Minimum=<0 0 0>, Maximum=<1 1 1>))
   &ScaleMeshSize(Scale Factor=&VectorFromCurve(VectorCurve=DataInterface() {propertyValues {
       XCurve {keys [
         <at 0 0>
         {arriveTangent=0.9375, interpMode=RCIM_Cubic, leaveTangent=0.9375, time=0.2, value=1}
         {arriveTangent=-1.25, interpMode=RCIM_Cubic, leaveTangent=-1.25, time=0.8, value=0.75}
         <at 1 0 RCIM_Cubic>
        ]}
       YCurve {keys [
         <at 0 0>
         {arriveTangent=1.0625, interpMode=RCIM_Cubic, leaveTangent=1.0625, time=0.2, value=1}
         {arriveTangent=-0.25, interpMode=RCIM_Cubic, leaveTangent=-0.25, time=0.8, value=0.85}
         <at 1 0.8 RCIM_Cubic>
        ]}
       ZCurve {keys [<at 0.1734 1>,<at 0.2 1 RCIM_Cubic>,<at 1 1 RCIM_Cubic>]}
      }}))
   &ScaleColor(
    Scale Alpha=&FloatFromCurve(FloatCurve=DataInterface() {propertyValues {Curve {keys [<at 0.7 1 RCIM_Cubic>,<at 1.0011 0.002>]}}})
    ScaleA=true
    ScaleRGB=true
   )
   &GenerateLocationEvent(Event Send Rate=2, Module.Use Event Send Rate=true)
  ]
  rendererClasses [&NiagaraMeshRendererProperties]
  renderers [{rendererClass=&NiagaraMeshRendererProperties, rendererIndex=0}]
 }
 Fragment_Looping_Concert {
  emitterUpdateScript [
   &EmitterState(Life Cycle Mode=::NewEnumerator1, Loop Duration=1, UseLoopDelay=false)
   &SpawnRate(SpawnRate=20, Use Spawn Probability=false)
  ]
  particleSpawnScript [
   &InitializeParticle(
    Lifetime=&UniformRangedFloat(Minimum=1.5, Maximum=3.5)
    Mesh Scale=&UniformRangedVector(Minimum=<2 2 2>, Maximum=<15 15 15>)
    Module.Write Color=true
    Module.Write Lifetime=true
    Module.Write Mass=true
    Module.Write Material Random=true
    Module.Write Position=false
    Module.Write Scale=true
    Module.Write SpriteRotation=false
    Module.Write SpriteSize=true
    Position=&SimulationPosition()
    Sprite Size=<5 5>
   )
   &BoxLocation(Box Offset=<0 -10 0>, Box Size=<2750 1 1500>, MaskSpawnGroup=false, Module.Override Seed=false, Rotate Box=false, Surface Only=false)
   &AddVelocity(Velocity=&Multiply_Vector(A=$Output.BoxLocation.BoxVector, B=<0 1 0>))
   &InitialMeshOrientation(
    Rotation=&UniformRangedVector(Minimum=<0 0 0>, Maximum=<1 1 1>)
    Use Orientation Vector=true
    Use Rotation Vector=true
   )
  ]
  particleUpdateScript [
   &ParticleState()
   &AccelerationForce(Acceleration=<0 0 -250>)
   &Drag(Drag=&UniformRangedFloat(Minimum=0, Maximum=3), Use Linear Drag=true, Use Rotational Drag=false)
   &SolveForcesAndVelocity(Clamp Velocity=false, Limit Acceleration=false)
   &UpdateMeshOrientation(Rotation Rate=0.5, Rotation Vector:Vector3f=&UniformRangedVector(Minimum=<0 0 0>, Maximum=<1 1 1>))
   &ScaleMeshSize(Scale Factor=&VectorFromCurve(VectorCurve=DataInterface() {propertyValues {
       XCurve {keys [
         <at 0 0>
         {arriveTangent=0.9375, interpMode=RCIM_Cubic, leaveTangent=0.9375, time=0.2, value=1}
         {arriveTangent=-1.25, interpMode=RCIM_Cubic, leaveTangent=-1.25, time=0.8, value=0.75}
         <at 1 0 RCIM_Cubic>
        ]}
       YCurve {keys [
         <at 0 0>
         {arriveTangent=1.0625, interpMode=RCIM_Cubic, leaveTangent=1.0625, time=0.2, value=1}
         {arriveTangent=-0.25, interpMode=RCIM_Cubic, leaveTangent=-0.25, time=0.8, value=0.85}
         <at 1 0.8 RCIM_Cubic>
        ]}
       ZCurve {keys [<at 0.1734 1>,<at 0.2 1 RCIM_Cubic>,<at 1 1 RCIM_Cubic>]}
      }}))
   &ScaleColor(
    Scale Alpha=&FloatFromCurve(FloatCurve=DataInterface() {propertyValues {Curve {keys [<at 0.7 1 RCIM_Cubic>,<at 1.0011 0.002>]}}})
    ScaleA=true
    ScaleRGB=true
   )
   &GenerateLocationEvent(Event Send Rate=2, Module.Use Event Send Rate=true)
  ]
  rendererClasses [&NiagaraMeshRendererProperties]
  renderers [{rendererClass=&NiagaraMeshRendererProperties, rendererIndex=0}]
 }
 Rift_Clouds {
  emitterUpdateScript [
   &EmitterState(Life Cycle Mode=::NewEnumerator1, Loop Behavior=::NewEnumerator1, Loop Duration=1, UseLoopDelay=false)
   &SpawnBurst_Instantaneous(Spawn Count=1, Use Spawn Probability=false)
   &SpawnBurst_Instantaneous(Spawn Count=1, Spawn Group=1, Use Spawn Probability=false) {module=SpawnBurst_Instantaneous001}
  ]
  particleSpawnScript [
   &InitializeParticle(
    Color=&LerpLinearColors(
     StartColor=<rgba 0.2188 0.7883 1 0.5>
     EndColor=<rgba 1.5 1.5 1.5 0.5>
     LerpFactor=&MakeFloatFromInt(Int=&SpawnGroup())
    )
    Mesh Scale=<2 12 8>
    Module.Write Color=true
    Module.Write Lifetime=true
    Module.Write Mass=true
    Module.Write Material Random=true
    Module.Write Position=false
    Module.Write Scale=true
    Module.Write SpriteRotation=false
    Module.Write SpriteSize=true
    Position=&SimulationPosition()
    Sprite Size=<5 5>
   )
   &InitialMeshOrientation(Orientation Vector=<0 1 0>, Use Orientation Vector=true, Use Rotation Vector=false)
   &SystemLocation(MaskSpawnGroup=false, Offset=<0 1 0>)
  ]
  particleUpdateScript [
   &ParticleState()
   &SolveForcesAndVelocity(Clamp Velocity=false, Limit Acceleration=false)
   &DynamicMaterialParameters(
    Index 0 Param 1=&Lerp_Float(
     A=&FloatFromCurve(FloatCurve=DataInterface() {propertyValues {Curve {keys [<at 0 0.8>,<at 1 1.8>]}}})
     B=1
     Alpha=&MakeFloatFromInt(Int=&SpawnGroup())
    )
    Index 0 Param 2=0.25
    Module.Param0WriteEnabled=true
    Module.Param1WriteEnabled=true
    Module.Param2WriteEnabled=false
    Module.Param3WriteEnabled=false
   )
   &ScaleMeshSize(Scale Factor=&VectorFromFloat(Value=&FloatFromCurve(FloatCurve=DataInterface() {propertyValues {Curve {keys [<at 0 0>,<at 0.25 1>]}}})))
   &ScaleColor(
    Scale Alpha=&FloatFromCurve(FloatCurve=DataInterface() {propertyValues {Curve {keys [<at 0 1 RCIM_Cubic>,<at 1.0011 0.002 RCIM_Cubic>]}}})
    Scale RGB=/Niagara/DynamicInputs/Math/SlerpVector(
     Vector A:Vector3f=<1 1 1>
     Vector B:Vector3f=&MakeVectorFromLinearColorRGB(Color=&ColorFromCurve(ColorCurve=$User.BaseColorCurve, CurveIndex=&Frac_Float(Float=$NPC.Papaya_Jukebox_NPC.Base Color Parameter)))
     Alpha:NiagaraFloat=&MakeFloatFromInt(Int=&SpawnGroup())
     Scale:NiagaraFloat=1
    )
    ScaleA=true
    ScaleRGB=true
   )
  ]
  rendererClasses [&NiagaraMeshRendererProperties]
  renderers [{rendererClass=&NiagaraMeshRendererProperties, rendererIndex=0}]
 }
 Sparklies_Burst {
  simTarget=GPUComputeSim
  emitterUpdateScript [
   &EmitterState(Life Cycle Mode=::NewEnumerator1, Loop Behavior=::NewEnumerator1, Loop Duration=1, UseLoopDelay=false)
   &SpawnBurst_Instantaneous(Spawn Count=300, Use Spawn Probability=false)
   &SpawnBurst_Instantaneous(Spawn Count=50, Spawn Group=1, Use Spawn Probability=false) {module=SpawnBurst_Instantaneous001}
  ]
  particleSpawnScript [
   &InitializeParticle(
    Color=&LerpLinearColors(
     StartColor=<rgba 20 20 20 1>
     EndColor=<rgba 1 1 1 1>
     LerpFactor=&MakeFloatFromInt(Int=&SpawnGroup())
    )
    Lifetime=&UniformRangedFloat(Minimum=1.5, Maximum=3.5)
    Module.Write Color=true
    Module.Write Lifetime=true
    Module.Write Mass=true
    Module.Write Material Random=true
    Module.Write Position=false
    Module.Write Scale=false
    Module.Write SpriteRotation=true
    Module.Write SpriteSize=true
    Position=&SimulationPosition()
    Sprite Rotation=/Niagara/DynamicInputs/Angles/DegreesToNormalizedAngle(AngleInDegrees:NiagaraFloat=&UniformRangedFloat(Minimum=0, Maximum=1))
    Sprite Size=&Vector2DFromFloat(Value=&UniformRangedFloat(Minimum=30, Maximum=50))
   )
   &BoxLocation(Box Offset=<0 -10 0>, Box Size=<2750 1 1500>, MaskSpawnGroup=false, Module.Override Seed=false, Rotate Box=false, Surface Only=false)
   &SphereLocation(
    Hemisphere Y=true
    MaskSpawnGroup=false
    Module.Override Randomness=false
    Non Uniform Scale=<1 -1 1>
    Sphere Radius=&UniformRangedFloat(Minimum=10, Maximum=15)
    Surface Only=false
   )
   &AddVelocity(Velocity=&Add_Vector(
     A=&Multiply_Vector(A=$Output.BoxLocation.BoxVector, B=<0 100 0>)
     B=&Multiply_Vector(
      A=$Output.SphereLocation.SphereNormal
      B=&VectorFromFloat(Value=&UniformRangedFloat(Minimum=500, Maximum=500))
     )
    ))
  ]
  particleUpdateScript [
   &ParticleState()
   &AccelerationForce(Acceleration=<0 0 -100>)
   /Niagara/Modules/Update/Forces/CurlNoiseForce(Noise Frequency=100, Noise Strength=1000, Pan Noise=true, Pan Noise Field=<0 0 500>)
   &Drag(Drag=&UniformRangedFloat(Minimum=2, Maximum=4), Use Linear Drag=true, Use Rotational Drag=false)
   &SolveForcesAndVelocity(Clamp Velocity=false, Limit Acceleration=false)
   &ScaleColor(
    Scale Alpha=&FloatFromCurve(FloatCurve=DataInterface() {propertyValues {Curve {keys [
         {arriveTangent=3.8421, interpMode=RCIM_Cubic, leaveTangent=3.8421, tangentMode=RCTM_User, time=0, value=0}
         {arriveTangent=0.002, interpMode=RCIM_Cubic, leaveTangent=0.002, time=0.5, value=1}
         {arriveTangent=-3.4849, interpMode=RCIM_Cubic, leaveTangent=-3.4849, tangentMode=RCTM_User, time=1.0011, value=0.002}
        ]}}})
    Scale RGB=&Lerp_Vector(
     A=&MakeVectorFromLinearColorRGB(Color=&ColorFromCurve(ColorCurve=$User.BaseColorCurve, CurveIndex=&Frac_Float(Float=$NPC.Papaya_Jukebox_NPC.Base Color Parameter)))
     B=&MakeVectorFromLinearColorRGB(Color=<rgba 7.4167 15.8975 16 1>)
     Alpha=&MakeFloatFromInt(Int=&SpawnGroup())
    )
    ScaleA=true
    ScaleRGB=true
   )
  ]
  rendererClasses [&NiagaraSpriteRendererProperties]
  renderers [{rendererClass=&NiagaraSpriteRendererProperties}]
 }
 Sparklies_Loop {
  simTarget=GPUComputeSim
  emitterUpdateScript [
   &EmitterState(Life Cycle Mode=::NewEnumerator1, Loop Duration=1, UseLoopDelay=false)
   &SpawnRate(SpawnRate=100, Use Spawn Probability=false)
   &SpawnRate(SpawnGroup=1, SpawnRate=10, Use Spawn Probability=false) {module=SpawnRate001}
  ]
  particleSpawnScript [
   &InitializeParticle(
    Color=&LerpLinearColors(
     StartColor=<rgba 20 20 20 1>
     EndColor=<rgba 1 1 1 1>
     LerpFactor=&MakeFloatFromInt(Int=&SpawnGroup())
    )
    Lifetime=&UniformRangedFloat(Minimum=1.5, Maximum=3.5)
    Module.Write Color=true
    Module.Write Lifetime=true
    Module.Write Mass=true
    Module.Write Material Random=true
    Module.Write Position=false
    Module.Write Scale=false
    Module.Write SpriteRotation=true
    Module.Write SpriteSize=true
    Position=&SimulationPosition()
    Sprite Rotation=&DegreesToNormalizedAngle(AngleInDegrees:NiagaraFloat=&UniformRangedFloat(Minimum=0, Maximum=1))
    Sprite Size=&Vector2DFromFloat(Value=&UniformRangedFloat(Minimum=30, Maximum=50))
   )
   &BoxLocation(Box Offset=<0 -10 0>, Box Size=<2750 1 1500>, MaskSpawnGroup=false, Module.Override Seed=false, Rotate Box=false, Surface Only=false)
   &AddVelocity(Velocity=&Multiply_Vector(A=$Output.BoxLocation.BoxVector, B=<0 1 0>))
  ]
  particleUpdateScript [
   &ParticleState()
   &CurlNoiseForce(Noise Frequency=100, Noise Strength=1000, Pan Noise=true, Pan Noise Field=<0 0 500>)
   &SolveForcesAndVelocity(Clamp Velocity=false, Limit Acceleration=false)
   &ScaleColor(
    Scale Alpha=&FloatFromCurve(FloatCurve=DataInterface() {propertyValues {Curve {keys [
         {arriveTangent=3.8421, interpMode=RCIM_Cubic, leaveTangent=3.8421, tangentMode=RCTM_User, time=0, value=0}
         {arriveTangent=0.002, interpMode=RCIM_Cubic, leaveTangent=0.002, time=0.5, value=1}
         {arriveTangent=-3.4849, interpMode=RCIM_Cubic, leaveTangent=-3.4849, tangentMode=RCTM_User, time=1.0011, value=0.002}
        ]}}})
    Scale RGB=&Lerp_Vector(
     A=&MakeVectorFromLinearColorRGB(Color=&ColorFromCurve(ColorCurve=$User.BaseColorCurve, CurveIndex=&Frac_Float(Float=$NPC.Papaya_Jukebox_NPC.Base Color Parameter)))
     B=&MakeVectorFromLinearColorRGB(Color=<rgba 7.4167 15.8975 16 1>)
     Alpha=&MakeFloatFromInt(Int=&SpawnGroup())
    )
    ScaleA=true
    ScaleRGB=true
   )
  ]
  rendererClasses [&NiagaraSpriteRendererProperties]
  renderers [{rendererClass=&NiagaraSpriteRendererProperties}]
 }
}
summary {
 systemName=NS_PapayaConcert_MainScreen_Transition
 emitters [
  {emitterName=Sparklies_Burst, simTarget=GPUComputeSim, rendererClasses [&NiagaraSpriteRendererProperties]}
  {emitterName=Sparklies_Loop, simTarget=GPUComputeSim, rendererClasses [&NiagaraSpriteRendererProperties]}
  {emitterName=Fragment_Burst, rendererClasses [&NiagaraMeshRendererProperties]}
  {emitterName=Fragment_Looping, rendererClasses [&NiagaraMeshRendererProperties]}
  {emitterName=Fragment_Burst_Concert, rendererClasses [&NiagaraMeshRendererProperties]}
  {emitterName=Fragment_Looping_Concert, rendererClasses [&NiagaraMeshRendererProperties]}
  {emitterName=Fragment_Glints, rendererClasses [&NiagaraSpriteRendererProperties]}
  {emitterName=Fragment_Glints_concert, rendererClasses [&NiagaraSpriteRendererProperties]}
  {emitterName=Rift_Clouds, rendererClasses [&NiagaraMeshRendererProperties]}
 ]
 userVariables [User.BaseColorCurve {
   type=NiagaraDataInterfaceColorCurve
   description=""
   defaultValue {
    struct=NiagaraExt_VariableValue_DataInterface
    value {
     dataInterface=&self:NiagaraDataInterfaceColorCurve_0
     dataInterfaceClass=&NiagaraDataInterfaceColorCurve
    }
   }
  }]
}
