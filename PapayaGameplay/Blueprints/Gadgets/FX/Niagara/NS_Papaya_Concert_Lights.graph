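# Niagara system for Papaya concert lights that pulse and change color with the jukebox's parameter collection.
# A first-loop burst spawns eight long-lived particles on filtered skeletal-mesh bones, each with a sprite renderer, initialized at zero size, and a light renderer.
# Color samples a user curve at the jukebox base color parameter plus a distance-weighted offset, then scales by average amplitude.
# The intermission curve is exposed but unused in the shown stack.
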NiagaraSystem /PapayaGameplay/Blueprints/Gadgets/FX/Niagara/NS_Papaya_Concert_Lights
dependencies {
 usedDataInterfaces [&NiagaraDataInterfaceSkeletalMesh]
 usedDynamicInputs []
 usedModules [&SystemState]
}
emitters {FN_Center_Sprite {
  emitterUpdateScript [
   &EmitterState()
   &SpawnBurst_Instantaneous(Spawn Count=8, Spawn Probability=HlslExpression() {hlslExpression="Emitter.LoopCount == 0 ? 1.f : 0.f"}, Use Spawn Probability=true)
  ]
  particleSpawnScript [
   /Niagara/Modules/Spawn/Initialization/InitializeParticle(
    Lifetime=50
    Module.Write Color=true
    Module.Write Lifetime=true
    Module.Write Mass=true
    Module.Write Material Random=true
    Module.Write Position=true
    Module.Write Scale=false
    Module.Write SpriteRotation=false
    Module.Write SpriteSize=true
    Position=&SimulationPosition()
    Sprite Size=<0 0>
   )
   /SetVariables(Particles.LightExponent=2) {isSetParameters=true}
   /Game/Effects/Niagara/Modules/Groups/FN_SetSpawnGroup_v1/FN_SetSpawnGroup(Emitter Bone Count=8)
   &SampleSkeletalMeshSkeleton(
    Mesh=DataInterface() {propertyValues {
      DefaultMesh=null
      ExcludeBoneName=root
      OptionalComponentClass=null
      PreviewMesh=/PapayaGameplay/Blueprints/Gadgets/FX/StaticMesh/sm_papaya_bone_location
      SkinningMode=SkinOnTheFly
      SoftSourceActor=null
      SourceMode=Default
      UvSetIndex=0
      WholeMeshLOD=-1
      bExcludeBone=false
      bReadDeformedGeometry=true
      bRequireCurrentFrameData=true
      ComponentTags []
      FilteredBones [Bone_012,Bone_018,Bone_020,Bone_023,Bone_030,Bone_037,Bone_013,Bone_019]
      FilteredSockets []
      MeshUserParameter {parameter {name=None, typeDefHandle {registeredTypeIndex=1}, varData []}}
      SamplingRegions []
     }}
    Specific Bone Index:NiagaraInt32=$Particles.SpawnGroup
   )
   /Niagara/Modules/Spawn/Location/SkeletalMeshSkeletonLocation(Module.Inherit Velocity=false, Module.Use Sampled Bone as Ribbon ID=true)
  ]
  particleUpdateScript [
   &ParticleState(Kill Particles When Lifetime Has Elapsed=false, Let Infinitely Lived Particles Die When Emitter Deactivates=true)
   &ScaleColor(
    Scale RGB=&MakeVectorFromLinearColorRGB(Color=&ColorFromCurve(
      ColorCurve=$User.ColorCurveBaseColor
      CurveIndex=&Add_Float(
       A=&Frac_Float(Float=$NPC.Papaya_Jukebox_NPC.Base Color Parameter)
       B=&Lerp_Float(
        A=0
        B=$NPC.Papaya_Jukebox_NPC.Base Color Parameter Offset
        Alpha=&NormalizeDistanceRange(
         Start Position:NiagaraPosition=$Engine.Owner.Position
         End Position:NiagaraPosition=<0 0 0>
         Distance:NiagaraFloat=1
         Invert Normalized Range:NiagaraBool=false
        )
       )
      )
     ))
    ScaleA=true
    ScaleRGB=true
   )
   /SetVariables(Particles.Color=&MultiplyLinearColor(
     Linear Color=$Particles.Color
     Scale Factor=&MakeVector4_FromFloat(Float=&Multiply_Float(A=$NPC.Papaya_Jukebox_NPC.AverageAmplitude, B=8))
    )) {isSetParameters=true}
   &ScaleSpriteSize()
  ]
  rendererClasses [&NiagaraSpriteRendererProperties, &NiagaraLightRendererProperties]
  renderers [
   {rendererClass=&NiagaraSpriteRendererProperties}
   {rendererClass=&NiagaraLightRendererProperties, rendererIndex=1}
  ]
 }}
summary {
 systemName=NS_Papaya_Concert_Lights
 emitters [{
   emitterName=FN_Center_Sprite
   rendererClasses [&NiagaraSpriteRendererProperties, &NiagaraLightRendererProperties]
  }]
 userVariables [
  User.ColorCurveBaseColor {
   type=NiagaraDataInterfaceColorCurve
   description=""
   defaultValue {
    struct=NiagaraExt_VariableValue_DataInterface
    value {
     dataInterface=&self:NiagaraDataInterfaceColorCurve_1
     dataInterfaceClass=&NiagaraDataInterfaceColorCurve
    }
   }
  }
  User.ColorCurveIntermission {
   type=NiagaraDataInterfaceColorCurve
   description=""
   defaultValue {
    struct=NiagaraExt_VariableValue_DataInterface
    value {
     dataInterface=&self:NiagaraDataInterfaceColorCurve_2
     dataInterfaceClass=&NiagaraDataInterfaceColorCurve
    }
   }
  }
 ]
}
