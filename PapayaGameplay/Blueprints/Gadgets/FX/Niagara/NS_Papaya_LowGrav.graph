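# Niagara effect for the Papaya low-gravity gadget, dressing a skeletal mesh supplied as a user parameter.
# Camera-offset decal sprites burst, then loop on its bones, while movement-spawned ribbons trail from four bone/socket indices.
# GPU embers spawn on the mesh and drift through a tiled vector field, alongside one persistent light.
# Embers and light are music-reactive, taking color-curve position and brightness from the Papaya Jukebox parameter collection's color and amplitude values.
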NiagaraSystem /PapayaGameplay/Blueprints/Gadgets/FX/Niagara/NS_Papaya_LowGrav
dependencies {
 usedDataInterfaces [&NiagaraDataInterfaceCurve, /Script/Niagara.NiagaraDataInterfaceVectorField]
 usedDynamicInputs []
 usedModules [&SystemState]
}
emitters {
 Decals_Loop {
  emitterUpdateScript [
   &EmitterState(Life Cycle Mode=::NewEnumerator1, Loop Duration=1, UseLoopDelay=false)
   &SpawnRate(SpawnRate=15, Use Spawn Probability=false)
  ]
  particleSpawnScript [
   /Game/Effects/Niagara/Modules/Groups/FN_SetSpawnGroup_v1/FN_SetSpawnGroup(Emitter Bone Count=17)
   /Niagara/Modules/Spawn/Initialization/InitializeParticle(
    Color=<rgba 1.5 1.5 1.5 1>
    Lifetime=2
    Module.Write Color=true
    Module.Write Lifetime=true
    Module.Write Mass=true
    Module.Write Material Random=true
    Module.Write Position=false
    Module.Write Scale=false
    Module.Write SpriteRotation=false
    Module.Write SpriteSize=true
    Position=&SimulationPosition()
    Sprite Size=&Vector2DFromFloat(Value=30)
   )
   &DynamicMaterialParameters(
    Index 0 Param 2=1
    Index 0 Param 4=&UniformRangedFloat(Minimum=0, Maximum=1)
    Module.Param0WriteEnabled=false
    Module.Param1WriteEnabled=true
    Module.Param2WriteEnabled=false
    Module.Param3WriteEnabled=true
   )
   &CameraOffset(Camera Offset Amount=30)
  ]
  particleUpdateScript [
   &ParticleState()
   &SampleSkeletalMeshSkeleton(Mesh:NiagaraDataInterfaceSkeletalMesh=$User.SkeletalMesh, Specific Bone Index:NiagaraInt32=$Particles.SpawnGroup) {enabled=false}
   /Niagara/Modules/Spawn/Location/SkeletalMeshSkeletonLocation(Module.Inherit Velocity=false, Module.Use Sampled Bone as Ribbon ID=true) {enabled=false}
   &SkeletalMeshLocation(Bone / Socket Index=$Particles.SpawnGroup, Bone Sampling Mode=::NewEnumerator4, Inherit Velocity=false, MaskSpawnGroup=false)
   &SolveForcesAndVelocity(Clamp Velocity=false, Limit Acceleration=false)
   &ScaleColor(
    Scale Alpha=&FloatFromCurve(FloatCurve=DataInterface() {propertyValues {Curve {keys [<at 0 0 RCIM_Cubic>,<at 0.5 1 RCIM_Cubic>,<at 1 0 RCIM_Cubic>]}}})
    Scale RGB=&Lerp_Vector(A=<1 0.6851 0.8647>, B=<0.3728 1 0.6644>, Alpha=$Particles.NormalizedAge)
    ScaleA=true
    ScaleRGB=true
   )
   &DynamicMaterialParameters(
    Index 0 Param 1=&FloatFromCurve(FloatCurve=DataInterface() {propertyValues {Curve {keys [
         {arriveTangent=-11.7199, interpMode=RCIM_Cubic, leaveTangent=-11.7199, tangentMode=RCTM_User, time=0, value=5}
         <at 0.4 1.5 RCIM_Cubic>
         {arriveTangent=7.1904, interpMode=RCIM_Cubic, leaveTangent=7.1904, tangentMode=RCTM_User, time=1, value=5}
        ]}}})
    Index 0 Param 3=&Lerp_Float(A=0.35, B=0, Alpha=$Particles.NormalizedAge)
    Module.Param0WriteEnabled=true
    Module.Param1WriteEnabled=false
    Module.Param2WriteEnabled=true
    Module.Param3WriteEnabled=false
   ) {module=DynamicMaterialParameters001}
  ]
  rendererClasses [&NiagaraSpriteRendererProperties]
  renderers [{rendererClass=&NiagaraSpriteRendererProperties}]
 }
 Decals_Start {
  emitterUpdateScript [
   &EmitterState(Life Cycle Mode=::NewEnumerator1, Loop Behavior=::NewEnumerator1, Loop Duration=1, UseLoopDelay=false)
   &SpawnBurst_Instantaneous(Spawn Count=17, Use Spawn Probability=false)
  ]
  particleSpawnScript [
   &InitializeParticle(
    Color=<rgba 1.5 1.5 1.5 1>
    Lifetime=0.6
    Module.Write Color=true
    Module.Write Lifetime=true
    Module.Write Mass=true
    Module.Write Material Random=true
    Module.Write Position=false
    Module.Write Scale=false
    Module.Write SpriteRotation=false
    Module.Write SpriteSize=true
    Position=&SimulationPosition()
    Sprite Size=&Vector2DFromFloat(Value=30)
   )
   &DynamicMaterialParameters(
    Index 0 Param 2=1
    Index 0 Param 4=&UniformRangedFloat(Minimum=0, Maximum=1)
    Module.Param0WriteEnabled=false
    Module.Param1WriteEnabled=true
    Module.Param2WriteEnabled=false
    Module.Param3WriteEnabled=true
   )
   &CameraOffset(Camera Offset Amount=30)
  ]
  particleUpdateScript [
   &ParticleState()
   &SampleSkeletalMeshSkeleton(Mesh:NiagaraDataInterfaceSkeletalMesh=$User.SkeletalMesh) {enabled=false}
   &SkeletalMeshSkeletonLocation(Module.Inherit Velocity=false, Module.Use Sampled Bone as Ribbon ID=true) {enabled=false}
   &SkeletalMeshLocation(
    Bone Sampling Mode=::NewEnumerator0
    Inherit Velocity=false
    MaskSpawnGroup=false
    Module.Override Randomness=false
    Skeletal Mesh=$User.SkeletalMesh
   )
   &SolveForcesAndVelocity(Clamp Velocity=false, Limit Acceleration=false)
   &ScaleColor(
    Scale Alpha=&FloatFromCurve(FloatCurve=DataInterface() {propertyValues {Curve {keys [<at 0 0 RCIM_Cubic>,<at 0.5 1 RCIM_Cubic>,<at 1 0 RCIM_Cubic>]}}})
    Scale RGB=&Lerp_Vector(A=<1 0.6851 0.8647>, B=<0.3728 1 0.6644>, Alpha=$Particles.NormalizedAge)
    ScaleA=true
    ScaleRGB=true
   )
   &DynamicMaterialParameters(
    Index 0 Param 1=&FloatFromCurve(FloatCurve=DataInterface() {propertyValues {Curve {keys [<at 0 0.6 RCIM_Cubic>,<at 1 4 RCIM_Cubic>]}}})
    Index 0 Param 3=&FloatFromCurve(FloatCurve=DataInterface() {propertyValues {Curve {keys [
         {arriveTangent=-1.9207, interpMode=RCIM_Cubic, leaveTangent=-1.9207, tangentMode=RCTM_User, time=0, value=-0.25}
         {arriveTangent=1.4434, interpMode=RCIM_Cubic, leaveTangent=1.4434, tangentMode=RCTM_User, time=0.4, value=0}
         <at 1 0.5 RCIM_Cubic>
        ]}}})
    Module.Param0WriteEnabled=true
    Module.Param1WriteEnabled=false
    Module.Param2WriteEnabled=true
    Module.Param3WriteEnabled=false
   ) {module=DynamicMaterialParameters001}
   &ScaleSpriteSize(Scale Factor=&Vector2DFromFloat(Value=&Lerp_Float(A=1.25, B=1, Alpha=$Particles.NormalizedAge)))
  ]
  rendererClasses [&NiagaraSpriteRendererProperties]
  renderers [{rendererClass=&NiagaraSpriteRendererProperties}]
 }
 Embers {
  simTarget=GPUComputeSim
  emitterUpdateScript [
   &EmitterState(Life Cycle Mode=::NewEnumerator1, Loop Behavior=::NewEnumerator1, Loop Duration=30, UseLoopDelay=false)
   &SpawnRate(SpawnRate=400, Use Spawn Probability=false)
   &SpawnBurst_Instantaneous(Spawn Count=20, Use Spawn Probability=false)
   &SpawnBurst_Instantaneous(Spawn Count=10, Spawn Time=0.02, Use Spawn Probability=false) {module=SpawnBurst_Instantaneous001}
  ]
  particleSpawnScript [
   &InitializeParticle(
    Lifetime=&UniformRangedFloat(Minimum=0.2, Maximum=1)
    Module.Write Color=true
    Module.Write Lifetime=true
    Module.Write Mass=true
    Module.Write Material Random=true
    Module.Write Position=false
    Module.Write Scale=false
    Module.Write SpriteRotation=true
    Module.Write SpriteSize=true
    Position=&SimulationPosition()
    Sprite Rotation=&UniformRangedFloat(Minimum=0, Maximum=360)
    Sprite Size=&UniformRangedVector2D(Minimum=<1 1>, Maximum=<3 3>)
   )
   &SampleSkeletalMeshSkeleton(Mesh:NiagaraDataInterfaceSkeletalMesh=$User.SkeletalMesh) {enabled=false}
   &SkeletalMeshSkeletonLocation(Module.Inherit Velocity=false, Module.Use Sampled Bone as Ribbon ID=true) {enabled=false}
   &SkeletalMeshLocation(
    Bone Sampling Mode=::NewEnumerator0
    Inherit Velocity=false
    MaskSpawnGroup=false
    Module.Override Randomness=false
    Skeletal Mesh=$User.SkeletalMesh
   )
   &AddVelocity(Velocity=&UniformRangedVector(Minimum=<-50 -50 0>, Maximum=<50 50 0>))
  ]
  particleUpdateScript [
   &ParticleState()
   /Niagara/VectorFields/SampleVectorField(
    FieldRotate:Quat4f=<0 0 -0.5 0>
    Sampled Vector Scale=0.2
    VectorField=DataInterface() {propertyValues {Field=/Game/Effects/VectorFields/Vel_Grid_01_145, bTileX=true, bTileY=true, bTileZ=true}}
   )
   /Niagara/VectorFields/ApplyVectorField(Vector Field Intensity as Force=0, Vector Field Intensity as Velocity=1, Vector Field Velocity Tightness=0.01)
   &InheritVelocity(Module.Limit Speed=true, Velocity Scale=<0.7 0.7 0.7>)
   &Drag(Drag=&UniformRangedFloat(Minimum=3, Maximum=5), Use Linear Drag=true, Use Rotational Drag=false)
   &AccelerationForce(Acceleration=<0 0 -200>)
   &SolveForcesAndVelocity(Clamp Velocity=false, Limit Acceleration=false)
   &ScaleColor(
    Scale RGB=&MakeVectorFromLinearColorRGB(Color=&ColorFromCurve(ColorCurve=$User.BaseColor_Curve, CurveIndex=&Frac_Float(Float=$NPC.Papaya_Jukebox_NPC.Base Color Parameter)))
    ScaleA=true
    ScaleRGB=true
   )
   &ScaleColor(
    Scale Alpha=&FloatFromCurve(FloatCurve=DataInterface() {propertyValues {Curve {keys [<at 0 0 RCIM_Cubic>,<at 0.2 1>]}}})
    Scale RGB=&VectorFromFloat(Value=&FloatFromCurve(FloatCurve=DataInterface() {propertyValues {Curve {keys [<at 0.2 6000 RCIM_Cubic>,<at 1 4000>]}}}))
    ScaleA=true
    ScaleRGB=true
   ) {module=ScaleColor001}
   /SetVariables(Particles.Color=&MultiplyLinearColor(
     Linear Color=$Particles.Color
     Scale Factor=&MakeVector4_FromFloat(Float=&Multiply_Float(A=$NPC.Papaya_Jukebox_NPC.AverageAmplitude, B=8))
    )) {isSetParameters=true}
   &ScaleSpriteSize(Scale Factor=&Vector2DFromFloat(Value=&FloatFromCurve(FloatCurve=DataInterface() {propertyValues {Curve {keys [
          <at 0 0.25>
          {arriveTangent=0.7857, interpMode=RCIM_Cubic, leaveTangent=0.7857, time=0.15, value=0.5}
          {arriveTangent=-0.5851, interpMode=RCIM_Cubic, leaveTangent=-0.5851, time=0.7, value=0.8}
          <at 1.0011 0.002>
         ]}}})))
  ]
  rendererClasses [&NiagaraSpriteRendererProperties]
  renderers [{rendererClass=&NiagaraSpriteRendererProperties}]
 }
 Light {
  emitterUpdateScript [
   &EmitterState(
    Inactive Response=::NewEnumerator1
    Life Cycle Mode=::NewEnumerator1
    Loop Behavior=::NewEnumerator1
    Loop Duration Mode=::NewEnumerator1
    UseLoopDelay=false
   )
   &SpawnBurst_Instantaneous(Spawn Count=1, Use Spawn Probability=false)
  ]
  particleSpawnScript [
   &InitializeParticle(
    Lifetime=0
    Module.Write Color=true
    Module.Write Lifetime=true
    Module.Write Mass=true
    Module.Write Material Random=true
    Module.Write Position=false
    Module.Write Scale=false
    Module.Write SpriteRotation=false
    Module.Write SpriteSize=true
    Position=&SimulationPosition()
    Sprite Size=<0 0>
   )
   &SystemLocation(MaskSpawnGroup=false, Offset=<0 20 50>)
  ]
  particleUpdateScript [
   &ParticleState(Kill Particles When Lifetime Has Elapsed=false, Let Infinitely Lived Particles Die When Emitter Deactivates=true)
   &SolveForcesAndVelocity(Clamp Velocity=false, Limit Acceleration=false)
   &ScaleColor(
    Scale RGB=&MakeVectorFromLinearColorRGB(Color=&ColorFromCurve(ColorCurve=$User.BaseColor_Curve, CurveIndex=&Frac_Float(Float=$NPC.Papaya_Jukebox_NPC.Base Color Parameter)))
    ScaleA=true
    ScaleRGB=true
   )
   /SetVariables(Particles.Color=&MultiplyLinearColor(
     Linear Color=$Particles.Color
     Scale Factor=&MakeVector4_FromFloat(Float=&Multiply_Float(A=$NPC.Papaya_Jukebox_NPC.AverageAmplitude, B=8))
    )) {isSetParameters=true}
   &FN_Light_v2(Light Exponent=12, Light Radius=250, LightEnabled=true)
  ]
  rendererClasses [&NiagaraSpriteRendererProperties, &NiagaraLightRendererProperties]
  renderers [
   {rendererClass=&NiagaraSpriteRendererProperties}
   {rendererClass=&NiagaraLightRendererProperties, rendererIndex=1}
  ]
 }
 Ribbons {
  emitterUpdateScript [
   &EmitterState(Life Cycle Mode=::NewEnumerator1, Loop Duration=1, UseLoopDelay=false)
   &SpawnPerUnit(
    Module.Use Max Movement Threshold=true
    Module.Use Movement Tolerance=true
    Movement Tolerance=0.3
    Spawn Per Unit=20
    Use Spawn Probability=false
   )
   &SpawnPerUnit(
    Module.Use Max Movement Threshold=true
    Module.Use Movement Tolerance=true
    Movement Tolerance=0.3
    Spawn Group=1
    Spawn Per Unit=20
    Use Spawn Probability=false
   ) {module=SpawnPerUnit001}
   &SpawnPerUnit(
    Module.Use Max Movement Threshold=true
    Module.Use Movement Tolerance=true
    Movement Tolerance=0.3
    Spawn Group=2
    Spawn Per Unit=20
    Use Spawn Probability=false
   ) {module=SpawnPerUnit002}
   &SpawnPerUnit(
    Module.Use Max Movement Threshold=true
    Module.Use Movement Tolerance=true
    Movement Tolerance=0.3
    Spawn Group=3
    Spawn Per Unit=20
    Use Spawn Probability=false
   ) {module=SpawnPerUnit003}
  ]
  particleSpawnScript [
   &InitializeRibbon(
    Color=<rgba 1 1 1 0.25>
    Lifetime=0.35
    Module.Write Color=true
    Module.Write Lifetime=true
    Module.Write Mass=true
    Module.Write Material Random=true
    Module.Write Position=false
    Module.Write RibbonTwist=false
    Module.Write RibbonWidth=true
    Ribbon Width=14
   )
   &SampleSkeletalMeshSkeleton(
    Mesh:NiagaraDataInterfaceSkeletalMesh=$User.SkeletalMesh
    Specific Bone Index:NiagaraInt32=&self:SkeletalSpawnInex(SpawnGroup:NiagaraInt32=&SpawnGroup())
   ) {enabled=false}
   &SkeletalMeshSkeletonLocation(Module.Inherit Velocity=false, Module.Use Sampled Bone as Ribbon ID=true) {enabled=false}
   &SkeletalMeshLocation(
    Bone / Socket Index=&SpawnGroup()
    Bone Sampling Mode=::NewEnumerator4
    Inherit Velocity=false
    MaskSpawnGroup=false
    Skeletal Mesh=$User.SkeletalMesh
   )
   &DynamicMaterialParameters(
    Index 0 Param 3=1
    Index 0 Param 4=1
    Module.Param0WriteEnabled=false
    Module.Param1WriteEnabled=false
    Module.Param2WriteEnabled=true
    Module.Param3WriteEnabled=true
   )
  ]
  particleUpdateScript [
   &ParticleState()
   &SolveForcesAndVelocity(Clamp Velocity=false, Limit Acceleration=false)
   &DynamicMaterialParameters(
    Index 0 Param 1=&FloatFromCurve(FloatCurve=DataInterface() {propertyValues {Curve {keys [
         {arriveTangent=4.9001, interpMode=RCIM_Cubic, leaveTangent=4.9001, tangentMode=RCTM_User, time=0, value=0.5}
         {arriveTangent=-0.1667, interpMode=RCIM_Cubic, leaveTangent=-0.1667, time=0.1, value=1}
         {arriveTangent=-1.8087, interpMode=RCIM_Cubic, leaveTangent=-1.8087, tangentMode=RCTM_User, time=0.3, value=0.45}
         <at 1 0 RCIM_Cubic>
        ]}}})
    Index 0 Param 2=&FloatFromCurve(FloatCurve=DataInterface() {propertyValues {Curve {keys [
         {arriveTangent=12.7041, interpMode=RCIM_Cubic, leaveTangent=12.7041, tangentMode=RCTM_User, time=0, value=0}
         <at 0.1 1 RCIM_Cubic>
         <at 0.3 0 RCIM_Cubic>
        ]}}})
    Module.Param0WriteEnabled=true
    Module.Param1WriteEnabled=true
    Module.Param2WriteEnabled=false
    Module.Param3WriteEnabled=false
   ) {module=DynamicMaterialParameters001}
   &ScaleRibbonWidth(Ribbon Width Scale=&FloatFromCurve(FloatCurve=DataInterface() {propertyValues {Curve {keys [<at 0 1 RCIM_Cubic>, {arriveTangent=-0.6, interpMode=RCIM_Cubic, leaveTangent=-0.6, time=0.15, value=1.25}, <at 1 0.4 RCIM_Cubic>]}}}))
   &ScaleColor(
    Scale Alpha=&FloatFromCurve(FloatCurve=DataInterface() {propertyValues {Curve {keys [<at 0 0 RCIM_Cubic>,<at 0.15 1 RCIM_Cubic>,<at 1 0 RCIM_Cubic>]}}})
    Scale RGB=&VectorFromFloat(Value=&FloatFromCurve(FloatCurve=DataInterface() {propertyValues {Curve {keys [{arriveTangent=-4.1251, interpMode=RCIM_Cubic, leaveTangent=-4.1251, tangentMode=RCTM_User, time=0, value=2}, <at 0.45 0.7 RCIM_Cubic>]}}}))
    ScaleA=true
    ScaleRGB=true
   )
  ]
  rendererClasses [&NiagaraRibbonRendererProperties]
  renderers [{rendererClass=&NiagaraRibbonRendererProperties}]
 }
}
summary {
 systemName=NS_Papaya_LowGrav
 emitters [
  {emitterName=Decals_Loop, rendererClasses [&NiagaraSpriteRendererProperties]}
  {emitterName=Decals_Start, rendererClasses [&NiagaraSpriteRendererProperties]}
  {emitterName=Light, rendererClasses [&NiagaraSpriteRendererProperties, &NiagaraLightRendererProperties]}
  {emitterName=Ribbons, rendererClasses [&NiagaraRibbonRendererProperties]}
  {emitterName=Embers, simTarget=GPUComputeSim, rendererClasses [&NiagaraSpriteRendererProperties]}
 ]
 userVariables [
  User.BaseColor_Curve {
   type=NiagaraDataInterfaceColorCurve
   description=""
   defaultValue {
    struct=NiagaraExt_VariableValue_DataInterface
    value {
     dataInterface=&self:NiagaraDataInterfaceColorCurve_0
     dataInterfaceClass=&NiagaraDataInterfaceColorCurve
    }
   }
  }
  User.Intermission_Curve {
   type=NiagaraDataInterfaceColorCurve
   description=""
   defaultValue {
    struct=NiagaraExt_VariableValue_DataInterface
    value {
     dataInterface=&self:NiagaraDataInterfaceColorCurve_1
     dataInterfaceClass=&NiagaraDataInterfaceColorCurve
    }
   }
  }
  User.SkeletalMesh {
   type=NiagaraDataInterfaceSkeletalMesh
   defaultValue {
    struct=NiagaraExt_VariableValue_DataInterface
    value {dataInterface=&self:NiagaraDataInterfaceSkeletalMesh_0}
   }
  }
 ]
}
