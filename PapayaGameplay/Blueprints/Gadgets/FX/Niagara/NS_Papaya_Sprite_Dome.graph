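# Niagara gadget effect that builds a dome of sprites on a Fibonacci sphere, sized by User.Radius.
# GPU Dots tile the surface outward-facing, take noise-texture colour and flip in a UV-offset wave at FlipSpeed, which also sets lifetime.
# Dots_CPU spawns a static quarter-count variant in random greys; GPU Rays add sparse, radial, camera-facing streaks.
# Dynamic material parameters pass random, UV and direction data to the sprite materials.
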NiagaraSystem /PapayaGameplay/Blueprints/Gadgets/FX/Niagara/NS_Papaya_Sprite_Dome
dependencies {
 usedDataInterfaces [/Script/Niagara.NiagaraDataInterfaceTexture, &NiagaraDataInterfaceCamera]
 usedDynamicInputs []
 usedModules [&SystemState]
}
emitters {
 Dots {
  simTarget=GPUComputeSim
  emitterUpdateScript [
   &EmitterState(
    Life Cycle Mode=::NewEnumerator1
    Loop Duration=&MakeCustomFloatFromBool(
     Bool=&SetBoolByFloatComparison(A=$User.FlipSpeed, B=0)
     True Float=&Divide_Float(A=1, B=$User.FlipSpeed)
     False Float=1
    )
    UseLoopDelay=false
   )
   &SpawnBurst_Instantaneous(Spawn Count=$User.SpawnCount, Use Spawn Probability=false)
  ]
  particleSpawnScript [
   /Niagara/Modules/Spawn/Initialization/InitializeParticle(
    Lifetime=&MakeCustomFloatFromBool(
     Bool=&SetBoolByFloatComparison(A=$User.FlipSpeed, B=0)
     True Float=&Divide_Float(A=1, B=$User.FlipSpeed)
     False Float=1
    )
    Module.Write Color=false
    Module.Write Lifetime=true
    Module.Write Mass=true
    Module.Write Material Random=true
    Module.Write Position=false
    Module.Write Scale=false
    Module.Write SpriteRotation=false
    Module.Write SpriteSize=true
    Position=&SimulationPosition()
    Sprite Size=<5 5>
   )
   /SetVariables(Particles.RandomValue:NiagaraFloat=&UniformRangedFloat(
     Minimum=0
     Maximum=10
     Randomness Mode=::NewEnumerator1
     Random Seed=&self:ExecutionIndex()
     Fixed Random Seed=true
    )) {isSetParameters=true, module=SetVariables_#1}
   &self:FibonacciSphere(Scale:Vector3f=&VectorFromFloat(Value=$User.Radius)) {module=FibonacciSphere}
   /Niagara/Modules/Update/Textures/SampleTexture(
    Texture=DataInterface() {propertyValues {
      Texture=&HighContrastNoise
      TextureUserParameter {parameter {name=None, typeDefHandle {registeredTypeIndex=2}, varData []}}
     }}
    UV:Vector2f=&MakeVector2D(
     X=&MakeFloatFromVector2d(Vector 2D=$Particles.UV)
     Y=&MakeFloatFromVector2d(Vector 2D=$Particles.UV, Channel=::NewEnumerator1)
    )
   )
   /SetVariables(Particles.SpriteSize=&Vector2DFromFloat(Value=&Multiply_Float(A=&self:SphereSurfaceArea(Radius:NiagaraFloat=$User.Radius), B=1))) {isSetParameters=true, module=SetVariables_#2}
   /SetVariables(Particles.SpriteAlignment=/Niagara/DynamicInputs/Vectors/CrossProduct(
     Vector A:Vector3f=<0 0 -1>
     Vector B:Vector3f=&NormalizeVector(Vector To Normalize=Linked() {linkedVariable {name=Particles.Position, type=NiagaraPosition}})
     Normalize Incoming Vectors:NiagaraBool=true
    )) {isSetParameters=true, module=SetVariables_#3}
   /SetVariables(Particles.SpriteFacing=&NormalizeVector(Vector To Normalize=Linked() {linkedVariable {name=Particles.Position, type=NiagaraPosition}})) {isSetParameters=true, module=SetVariables_#4}
   /SetVariables(Particles.UVInterval:NiagaraFloat=&self:UVInterval()) {isSetParameters=true, module=SetVariables_#5}
   /SetVariables(Particles.UVFraction:Vector2f=&self:UVFraction()) {isSetParameters=true, module=SetVariables_#6}
   &Color(Color=$Output.SampleTexture.SampledColor)
  ]
  particleUpdateScript [
   &ParticleState()
   /SetVariables(Particles.N:Vector3f=&NormalizeVector(Vector To Normalize=Linked() {linkedVariable {name=Particles.Position, type=NiagaraPosition}})) {isSetParameters=true}
   &self:RotateSprites(
    Axis:Vector3f=$Particles.SpriteAlignment
    Normalized Angle:NiagaraFloat=&MakeCustomFloatFromBool(
     Bool=&SetBoolByFloatComparison(A=$User.FlipSpeed, B=0)
     True Float=&ClampFloat(
      Float=&ModuloFloat(
       A=&Add_Float(
        A=&MakeFloatFromVector2d(Vector 2D=$Particles.UV)
        B=&Multiply_Float(A=$Particles.Age, B=$User.FlipSpeed)
       )
       B=1
      )
      Max=0.5
     )
     False Float=0
    )
    Vector To Rotate:Vector3f=$Particles.Initial.SpriteFacing
   ) {module=RotateSprites}
   /SetVariables(Particles.SpriteFacing=$Output.RotateSprites.New Vector) {isSetParameters=true}
   &DynamicMaterialParameters(
    Index 0 Param 1=$Particles.RandomValue
    Index 0 Param 2=&MakeFloatFromVector2d(Vector 2D=$Particles.UV)
    Index 0 Param 3=&MakeFloatFromVector2d(Vector 2D=$Particles.UV, Channel=::NewEnumerator1)
    Index 1 Param 1=&MakeFloatFromVector(Vector=$Particles.N, Channel=::NewEnumerator0)
    Index 1 Param 2=&MakeFloatFromVector(Vector=$Particles.N, Channel=::NewEnumerator1)
    Index 1 Param 3=&MakeFloatFromVector(Vector=$Particles.N, Channel=::NewEnumerator2)
    Index 1 Param 4=&DotProduct(Vector A=$Particles.N, Vector B=$Particles.SpriteFacing)
    Index 2 Param 1=&MakeFloatFromVector2d(Vector 2D=$Particles.UVFraction)
    Index 2 Param 2=&MakeFloatFromVector2d(Vector 2D=$Particles.UVFraction, Channel=::NewEnumerator1)
    Module.Param0WriteEnabled=true
    Module.Param0WriteEnabled1=true
    Module.Param0WriteEnabled2=true
    Module.Param1WriteEnabled=true
    Module.Param1WriteEnabled1=true
    Module.Param1WriteEnabled2=true
    Module.Param2WriteEnabled=true
    Module.Param2WriteEnabled1=true
    Module.Param2WriteEnabled2=true
    Module.Param3WriteEnabled=true
    Module.Param3WriteEnabled1=true
    Module.Param3WriteEnabled2=true
    Write Parameter Index 1=true
    Write Parameter Index 2=true
   )
  ]
  rendererClasses [&NiagaraSpriteRendererProperties]
  renderers [
   {rendererClass=&NiagaraSpriteRendererProperties}
   {rendererClass=&NiagaraSpriteRendererProperties, rendererIndex=1}
  ]
 }
 Dots_CPU {
  emitterUpdateScript [
   &EmitterState(
    Life Cycle Mode=::NewEnumerator1
    Loop Duration=&MakeCustomFloatFromBool(
     Bool=&SetBoolByFloatComparison(A=$User.FlipSpeed, B=0)
     True Float=&Divide_Float(A=1, B=$User.FlipSpeed)
     False Float=1
    )
    UseLoopDelay=false
   )
   &SpawnBurst_Instantaneous(
    Spawn Count=&MakeIntFromFloat(Float=&Multiply_Float(A=&MakeFloatFromInt(Int=$User.SpawnCount), B=0.25))
    Use Spawn Probability=false
   )
  ]
  particleSpawnScript [
   &InitializeParticle(
    Lifetime=&MakeCustomFloatFromBool(
     Bool=&SetBoolByFloatComparison(A=$User.FlipSpeed, B=0)
     True Float=&Divide_Float(A=1, B=$User.FlipSpeed)
     False Float=1
    )
    Module.Write Color=false
    Module.Write Lifetime=true
    Module.Write Mass=true
    Module.Write Material Random=true
    Module.Write Position=false
    Module.Write Scale=false
    Module.Write SpriteRotation=false
    Module.Write SpriteSize=true
    Position=&SimulationPosition()
    Sprite Size=<5 5>
   )
   /SetVariables(Particles.RandomValue:NiagaraFloat=&UniformRangedFloat(
     Minimum=0
     Maximum=10
     Randomness Mode=::NewEnumerator1
     Random Seed=&self:ExecutionIndex()
     Fixed Random Seed=true
    )) {isSetParameters=true, module=SetVariables_#1}
   &self:FibonacciSphere(Scale:Vector3f=&VectorFromFloat(Value=$User.Radius)) {module=FibonacciSphere}
   /SetVariables(Particles.SpriteSize=&Vector2DFromFloat(Value=&Multiply_Float(A=&self:SphereSurfaceArea(Radius:NiagaraFloat=$User.Radius), B=1))) {isSetParameters=true, module=SetVariables_#2}
   /SetVariables(Particles.SpriteAlignment=&CrossProduct(
     Vector A:Vector3f=<0 0 -1>
     Vector B:Vector3f=&NormalizeVector(Vector To Normalize=Linked() {linkedVariable {name=Particles.Position, type=NiagaraPosition}})
     Normalize Incoming Vectors:NiagaraBool=true
    )) {isSetParameters=true, module=SetVariables_#3}
   /SetVariables(Particles.SpriteFacing=&NormalizeVector(Vector To Normalize=Linked() {linkedVariable {name=Particles.Position, type=NiagaraPosition}})) {isSetParameters=true, module=SetVariables_#4}
   /SetVariables(Particles.UVInterval:NiagaraFloat=&self:UVInterval()) {isSetParameters=true, module=SetVariables_#5}
   /SetVariables(Particles.UVFraction:Vector2f=&self:UVFraction()) {isSetParameters=true, module=SetVariables_#6}
   &Color(Color=&UniformRangedLinearColor(Minimum=<rgba 1 1 1 1>, Maximum=<rgba 0 0 0 1>))
  ]
  particleUpdateScript [
   &ParticleState()
   &DynamicMaterialParameters(
    Index 0 Param 1=$Particles.RandomValue
    Index 0 Param 2=&MakeFloatFromVector2d(Vector 2D=$Particles.UV)
    Index 0 Param 3=&MakeFloatFromVector2d(Vector 2D=$Particles.UV, Channel=::NewEnumerator1)
    Index 1 Param 1=&MakeFloatFromVector(Vector=$Particles.SpriteFacing, Channel=::NewEnumerator0)
    Index 1 Param 2=&MakeFloatFromVector(Vector=$Particles.SpriteFacing, Channel=::NewEnumerator1)
    Index 1 Param 3=&MakeFloatFromVector(Vector=$Particles.SpriteFacing, Channel=::NewEnumerator2)
    Index 2 Param 1=&MakeFloatFromVector2d(Vector 2D=$Particles.UVFraction)
    Index 2 Param 2=&MakeFloatFromVector2d(Vector 2D=$Particles.UVFraction, Channel=::NewEnumerator1)
    Module.Param0WriteEnabled=true
    Module.Param0WriteEnabled1=true
    Module.Param0WriteEnabled2=true
    Module.Param1WriteEnabled=true
    Module.Param1WriteEnabled1=true
    Module.Param1WriteEnabled2=true
    Module.Param2WriteEnabled=true
    Module.Param2WriteEnabled1=true
    Module.Param2WriteEnabled2=true
    Module.Param3WriteEnabled=true
    Module.Param3WriteEnabled1=true
    Module.Param3WriteEnabled2=true
    Write Parameter Index 1=true
    Write Parameter Index 2=true
   )
  ]
  rendererClasses [&NiagaraSpriteRendererProperties]
  renderers [
   {rendererClass=&NiagaraSpriteRendererProperties}
   {rendererClass=&NiagaraSpriteRendererProperties, rendererIndex=1}
  ]
 }
 Rays {
  simTarget=GPUComputeSim
  emitterUpdateScript [
   &EmitterState(
    Life Cycle Mode=::NewEnumerator1
    Loop Duration=&MakeCustomFloatFromBool(
     Bool=&SetBoolByFloatComparison(A=$User.FlipSpeed, B=0)
     True Float=&Divide_Float(A=1, B=$User.FlipSpeed)
     False Float=1
    )
    UseLoopDelay=false
   )
   &SpawnBurst_Instantaneous(
    Spawn Count=&MakeIntFromFloat(Float=&Multiply_Float(A=&MakeFloatFromInt(Int=$User.SpawnCount), B=$User.SpawnCountRaysPct))
    Use Spawn Probability=false
   )
  ]
  particleSpawnScript [
   &InitializeParticle(
    Lifetime=&MakeCustomFloatFromBool(
     Bool=&SetBoolByFloatComparison(A=$User.FlipSpeed, B=0)
     True Float=&Divide_Float(A=1, B=$User.FlipSpeed)
     False Float=1
    )
    Module.Write Color=false
    Module.Write Lifetime=true
    Module.Write Mass=true
    Module.Write Material Random=true
    Module.Write Position=false
    Module.Write Scale=false
    Module.Write SpriteRotation=false
    Module.Write SpriteSize=true
    Position=&SimulationPosition()
    Sprite Size=<5 5>
   )
   /SetVariables(Particles.RandomValue:NiagaraFloat=&UniformRangedFloat(
     Minimum=0
     Maximum=10
     Randomness Mode=::NewEnumerator1
     Random Seed=&self:ExecutionIndex()
     Fixed Random Seed=true
    )) {isSetParameters=true}
   &self:FibonacciSphere(Scale:Vector3f=&VectorFromFloat(Value=$User.Radius)) {module=FibonacciSphere}
   &SampleTexture(
    Texture=DataInterface() {propertyValues {
      Texture=&HighContrastNoise
      TextureUserParameter {parameter {name=None, typeDefHandle {registeredTypeIndex=2}, varData []}}
     }}
    UV:Vector2f=&MakeVector2D(
     X=&MakeFloatFromVector2d(Vector 2D=$Particles.UV)
     Y=&MakeFloatFromVector2d(Vector 2D=$Particles.UV, Channel=::NewEnumerator1)
    )
   )
   /SetVariables(Particles.RaySpriteSize:Vector2f=&MakeVector2D(
     X=&Multiply_Float(
      A=&Multiply_Float(A=&self:SphereSurfaceArea(Radius:NiagaraFloat=$User.Radius), B=1)
      B=0.8
     )
     Y=&Multiply_Float(A=$User.Radius, B=0.2)
    )) {isSetParameters=true}
   /SetVariables(Particles.RaySpriteAlignment:Vector3f=&NormalizeVector(Vector To Normalize=Linked() {linkedVariable {name=Particles.Position, type=NiagaraPosition}})) {isSetParameters=true}
   &Color(Color=$Output.SampleTexture.SampledColor)
  ]
  particleUpdateScript [
   &ParticleState()
   /SetVariables(Particles.RaySpriteFacing:Vector3f=&NormalizeVector(Vector To Normalize=&ReturnCameraVector(Camera Property=::NewEnumerator4, Coordinate Space=::Simulation, Origin Position=$Particles.Position))) {isSetParameters=true}
   &DynamicMaterialParameters(
    Index 0 Param 1=$Particles.RandomValue
    Index 0 Param 2=&MakeFloatFromVector2d(Vector 2D=$Particles.UV)
    Index 0 Param 3=&MakeFloatFromVector2d(Vector 2D=$Particles.UV, Channel=::NewEnumerator1)
    Index 1 Param 1=&MakeFloatFromVector(Vector=$Particles.SpriteFacing, Channel=::NewEnumerator0)
    Index 1 Param 2=&MakeFloatFromVector(Vector=$Particles.SpriteFacing, Channel=::NewEnumerator1)
    Index 1 Param 3=&MakeFloatFromVector(Vector=$Particles.SpriteFacing, Channel=::NewEnumerator2)
    Module.Param0WriteEnabled=true
    Module.Param0WriteEnabled1=true
    Module.Param1WriteEnabled=true
    Module.Param1WriteEnabled1=true
    Module.Param2WriteEnabled=true
    Module.Param2WriteEnabled1=true
    Module.Param3WriteEnabled=true
    Module.Param3WriteEnabled1=true
    Write Parameter Index 1=true
   )
  ]
  rendererClasses [&NiagaraSpriteRendererProperties]
  renderers [
   {rendererClass=&NiagaraSpriteRendererProperties}
   {rendererClass=&NiagaraSpriteRendererProperties, rendererIndex=1}
  ]
 }
}
summary {
 systemName=NS_Papaya_Sprite_Dome
 emitters [
  {emitterName=Dots, simTarget=GPUComputeSim, rendererClasses [&NiagaraSpriteRendererProperties]}
  {emitterName=Rays, simTarget=GPUComputeSim, rendererClasses [&NiagaraSpriteRendererProperties]}
  {emitterName=Dots_CPU, rendererClasses [&NiagaraSpriteRendererProperties]}
 ]
 userVariables [
  User.FlipSpeed {type=NiagaraFloat, defaultValue=0.2, description=""}
  User.LoopDuration {type=NiagaraFloat, defaultValue=10, description=""}
  User.Radius {type=NiagaraFloat, defaultValue=1000, description=""}
  User.SpawnCount {type=NiagaraInt32, defaultValue=10000}
  User.SpawnCountRaysPct {type=NiagaraFloat, defaultValue=0.05, description=""}
 ]
}
