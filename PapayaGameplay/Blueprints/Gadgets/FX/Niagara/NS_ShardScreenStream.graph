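# Niagara gadget effect that streams drifting, spinning mesh shards on the GPU, reacting to jukebox audio.
# Shards spawn at a user-set rate in a Y-flattened sphere, storing normalized radial offset and an X-based UV.
# Linear and curl-noise forces directly set velocity, plus a radial push, while scale and alpha follow lifetime curves.
# Average amplitude from the Papaya_Jukebox parameter collection boosts spin and pushes a separate RenderPosition outward, with User.RenderPosFlip allowing mirroring.
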NiagaraSystem /PapayaGameplay/Blueprints/Gadgets/FX/Niagara/NS_ShardScreenStream
dependencies {usedDataInterfaces [&NiagaraDataInterfaceCurve], usedDynamicInputs [], usedModules [&SystemState]}
emitters {Shards {
  simTarget=GPUComputeSim
  emitterUpdateScript [
   &EmitterState(Life Cycle Mode=::NewEnumerator1, Loop Duration=1, UseLoopDelay=false)
   &SpawnRate(SpawnRate=$User.SpawnRate, Use Spawn Probability=false)
  ]
  particleSpawnScript [
   /Niagara/Modules/Spawn/Initialization/InitializeParticle(
    Lifetime=&UniformRangedFloat(Minimum=10, Maximum=11)
    Mesh Scale=&Multiply_Vector(
     A=&UniformRangedVector(Minimum=<0.6 0.6 0.6>, Maximum=<1.4 1.4 1.4>)
     B=&VectorFromFloat(Value=&UniformRangedFloat(Minimum=0.2, Maximum=1))
    )
    Module.Write Color=false
    Module.Write Lifetime=true
    Module.Write Mass=true
    Module.Write Material Random=true
    Module.Write Position=false
    Module.Write Scale=true
    Module.Write SpriteRotation=true
    Module.Write SpriteSize=true
    Position=&SimulationPosition()
    Sprite Rotation=&NormalizedAngleToDegrees(NormalizedAngle=&UniformRangedFloat(Minimum=0, Maximum=1))
    Sprite Size=&Vector2DFromFloat(Value=&UniformRangedFloat(Minimum=500, Maximum=700))
   )
   &InitialMeshOrientation(
    Rotation=&MakeVector(X=0, Y=&UniformRangedFloat(Minimum=0, Maximum=1), Z=0)
    Rotation Coordinate Space=::NewEnumerator0
    Use Orientation Vector=true
    Use Rotation Vector=true
   ) {module=InitialMeshOrientation001}
   /Niagara/Modules/Spawn/Location/SphereLocation(
    MaskSpawnGroup=false
    Module.Override Randomness=true
    Module.Override Seed=true
    Non Uniform Scale=<1 0.1 1>
    Randomness Mode:ENiagaraRandomnessMode=::NewEnumerator1
    Sphere Radius=$User.EmitRadius
    Surface Only=false
   )
   /SetVariables(Particles.NoiseOffset:Vector3f=<0 0 100>) {isSetParameters=true}
   /SetVariables(Particles.Spin:Vector3f=&UniformRangedVector(Minimum=<-0.3 -0.3 -0.3>, Maximum=<0.3 0.3 0.3>)) {isSetParameters=true}
   /SetVariables(Particles.UV:Vector2f=&MakeVector2D(
     X=&Add_Float(
      A=&Multiply_Float(
       A=&Divide_Float(
        A=&MakeFloatFromVector(Vector=Linked() {linkedVariable {name=Particles.Position, type=NiagaraPosition}}, Channel=::NewEnumerator0)
        B=$User.EmitRadius
       )
       B=0.5
      )
      B=0.5
     )
     Y=0
    )) {isSetParameters=true}
   /SetVariables(Particles.Radial:Vector3f=&Divide_Vector(
     A:Vector3f=Linked() {linkedVariable {name=Particles.Position, type=NiagaraPosition}}
     B:Vector3f=&VectorFromFloat(Value=$User.EmitRadius)
    )) {isSetParameters=true}
  ]
  particleUpdateScript [
   &ParticleState()
   &LinearForce(Force=&Multiply_VectorByFloat(Vector=<0 100 400>, Float=0.4))
   /Niagara/Modules/Update/Forces/CurlNoiseForce(Noise Frequency=0.001, Noise Strength=200, Pan Noise=true, Pan Noise Field=<0 0.1 0>)
   &CurlNoiseForce(
    Noise Frequency=0.0005
    Noise Strength=100
    Pan Noise=false
    Sample Position:Vector3f=&Add_Vector(A=Linked() {linkedVariable {name=Particles.Position, type=NiagaraPosition}}, B=$Particles.NoiseOffset)
   ) {module=CurlNoiseForce001}
   /SetVariables(Particles.Velocity=$Transient.PhysicsForce) {isSetParameters=true}
   /SetVariables() {isSetParameters=true}
   /SetVariables(Particles.Velocity=&Add_Vector(A=$Particles.Velocity, B=&Multiply_VectorByFloat(Vector=$Particles.Radial, Float=125))) {isSetParameters=true}
   &MeshRotationForce(
    Coordinate Space:ENiagaraMeshTransforms=::NewEnumerator3
    Rotation=&Add_Vector(
     A=$Particles.Spin
     B=&Multiply_VectorByFloat(Vector=$Particles.Spin, Float=&Multiply_Float(A=$NPC.Papaya_Jukebox_NPC.AverageAmplitude, B=2))
    )
   )
   &SolveForcesAndVelocity(Clamp Velocity=false, Limit Acceleration=false)
   &SolveRotationalForcesAndVelocity()
   &ScaleMeshSize(Scale Factor=&VectorFromFloat(Value=&FloatFromCurve(FloatCurve=DataInterface() {propertyValues {Curve {keys [
          {interpMode=RCIM_Cubic, tangentMode=RCTM_User, time=0, value=0}
          {interpMode=RCIM_Cubic, tangentMode=RCTM_User, time=0.3, value=1}
          {interpMode=RCIM_Cubic, tangentMode=RCTM_User, time=0.7, value=1}
          {interpMode=RCIM_Cubic, tangentMode=RCTM_User, time=1, value=0}
         ]}}})))
   /SetVariables(Particles.RenderPosition:Vector3f=&Add_Vector(
     A=&Multiply_Vector(A=Linked() {linkedVariable {name=Particles.Position, type=NiagaraPosition}}, B=$User.RenderPosFlip)
     B=&Add_Vector(
      A=&Multiply_VectorByFloat(Vector=$Particles.Radial, Float=&Multiply_Float(A=$NPC.Papaya_Jukebox_NPC.AverageAmplitude, B=250))
      B=&Multiply_VectorByFloat(
       Vector=&NormalizeVector(Vector To Normalize=Linked() {linkedVariable {name=Particles.Position, type=NiagaraPosition}})
       Float=&Multiply_Float(A=$NPC.Papaya_Jukebox_NPC.AverageAmplitude, B=150)
      )
     )
    )) {isSetParameters=true}
   &ScaleColor(
    Scale Alpha=&FloatFromCurve(FloatCurve=DataInterface() {propertyValues {Curve {keys [<at 0 0>, {interpMode=RCIM_Cubic, tangentMode=RCTM_User, time=0.2, value=1}, <at 1.0011 0.002>]}}})
    ScaleA=true
    ScaleRGB=true
   )
  ]
  rendererClasses [&NiagaraMeshRendererProperties]
  renderers [{rendererClass=&NiagaraMeshRendererProperties, rendererIndex=0}]
 }}
summary {
 systemName=NS_ShardScreenStream
 emitters [{emitterName=Shards, simTarget=GPUComputeSim, rendererClasses [&NiagaraMeshRendererProperties]}]
 userVariables [
  User.EmitRadius {type=NiagaraFloat, defaultValue=750, description=""}
  User.RenderPosFlip {type=Vector3f, description="", defaultValue {struct=Vector3f, value=<1 1 1>}}
  User.SpawnRate {type=NiagaraFloat, defaultValue=2, description=""}
 ]
}
