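; Master skeletal mesh actor for a dancing hologram mannequin with selectable dancer outfit, emote and scale.
; Construction fills character part slots from the chosen dancer's part array, enables custom depth, applies scale, sets the looping emote animation and writes a location ID label.
; Custom events replay the emote and toggle visibility with the hologram base.
; Appearing runs a timeline driving a PulsePosition material parameter across all parts, while disappearing spawns an effect at the base.
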(event Custom|EmotePlay
  (Components|Animation|PlayAnimation (Variables|SkeletalMeshActor|GetSkeletalMeshComponent) (Utilities|Array|Get(acopy) (Variables|Default|GetAnimation) (Variables|DancerSetup|GetEmoteSelect)) true))

(event Custom|ActorVisible
  (Rendering|SetVisibility (Variables|SkeletalMeshActor|GetSkeletalMeshComponent) true true)
  (Components|Activation|Activate (Variables|Default|GetHologram%20Base) true)
  (bind (_direction _pulseup) (Timeline|Pulse PlayFromStart)
    (:Update
      (bind _skeletalmeshcomponents (Customization|GetAllPartSkeletalMeshComponents))
      (for _array_element _skeletalmeshcomponents
        (Rendering|Material|SetScalarParameterValueonMaterials _array_element "PulsePosition" _pulseup)))))

(event Custom|ActorInvisible
  (bind (_returnvalue_location _returnvalue_rotation _returnvalue_scale) (Math|Transform|BreakTransform (Transformation|GetWorldTransform (Variables|Default|GetHologram%20Base))))
  (Components|Activation|Deactivate (Variables|Default|GetHologram%20Base))
  (Rendering|SetVisibility (Variables|SkeletalMeshActor|GetSkeletalMeshComponent) false true)
  (Effects|SpawnEmitteratLocation (Variables|Default|GetDisappearFX) _returnvalue_location _returnvalue_rotation _returnvalue_scale))

(fn CharacterPartScale
  (params (Scale :type Float%20%28double-precision%29))
  (bind _skeletalmeshcomponents (Customization|GetAllPartSkeletalMeshComponents))
  (for _array_element _skeletalmeshcomponents
    (Transformation|SetRelativeScale3D _array_element (Math|Conversions|ToVector(Float) Scale)))
  (return))

(fn CharacterPartSetup
  (bind _part (LocalVariables|GetPart))
  (for (_array_element _array_index) (Utilities|Select :Index (Variables|DancerSetup|GetDancer%20Select) :Option%200 (Variables|Dancers|GetBubblegum) :Option%201 (Variables|Dancers|GetDisco) :Option%202 (Variables|Dancers|GetJellyfish) :Option%203 (Variables|Dancers|GetNeonAnimal) :Option%204 (Variables|Dancers|GetReverb) :Option%205 (Variables|Dancers|GetSheath) :Option%206 (Variables|Dancers|GetStreetFashionRed) :Option%207 (Variables|Dancers|GetLasagna) :Option%208 (Variables|Dancers|GetRustler) :Option%209 (Variables|Dancers|GetFrontier) :Option%2010 (Variables|Dancers|GetDeadfire))
    (LocalVariables|SetPart _array_element)
    (switch int _array_index
      (:0
        (Customization|SetPart "Head" _part))
      (:1
        (Customization|SetPart "Body" _part))
      (:2
        (Customization|SetPart "Hat" _part))
      (:3
        (Customization|SetPart "Backpack" _part))
      (:4
        (Customization|SetPart "MiscOrTail" _part))
      (:5
        (Customization|SetPart "Face" _part))
      (:6
        (Customization|SetPart "Gameplay" _part))))
  (bind _skeletalmeshcomponents (Customization|GetAllPartSkeletalMeshComponents))
  (for _array_element_2 _skeletalmeshcomponents
    (Rendering|SetRenderCustomDepth _array_element_2 true)
    (Transformation|SetWorldScale3D _array_element_2 (Math|Conversions|ToVector(Float) (Variables|DancerSetup|GetDancer%20Scale)))))

(fn ConstructionScript
  (|CallParentFunction)
  (CallFunction|CharacterPartSetup)
  (Components|Animation|OverrideAnimationData (Variables|SkeletalMeshActor|GetSkeletalMeshComponent) (Utilities|Array|Get(acopy) (Variables|Default|GetAnimation) (Variables|DancerSetup|GetEmoteSelect)) true true 0.0 1.0)
  (Rendering|Components|TextRender|SetText (Variables|Default|GetLocation%20Name) (Variables|DancerSetup|GetLocation%20ID)))
